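# Niagara skydiving contrail system that draws ribbon trails from the character's hand and foot bones, plus speed lines while diving.
# The ribbon emitter runs 2 or 4 trails by quality level, pinning locked, non-aging particles to offset bone positions so trails stay attached.
# Dive amount drives ribbon spread, inherited movement and opacity scale, while LOD distance scales alpha.
# GPU sprite streaks spawn around a bone at a dive-scaled rate.
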NiagaraSystem /Game/Effects/Fort_Effects/Effects/NS_Contrails_Basic_v2
dependencies {
 usedDataInterfaces [
  &NiagaraDataInterfaceCurve
  &NiagaraDataInterfaceVector2DCurve
  &NiagaraDataInterfaceSkeletalMesh
 ]
 usedDynamicInputs [&FN_FrontEndFloatModifier, &SetBoolByFloatComparison]
 usedModules [
  &self:SystemSpawnScript.SystemScriptSource.SystemScriptGraph.NiagaraNodeAssignment_1.SetVariables_#1
  &SystemState
  /Game/Effects/Niagara/Modules/Skydiving/FN_DivingAttributes/v5/FN_DivingAttributes
  &self:SystemSpawnScript.SystemScriptSource.SystemScriptGraph.NiagaraNodeAssignment_0.SetVariables_#2
  &FN_Integrator
 ]
}
emitters {
 FN_SpeedLines {
  simTarget=GPUComputeSim
  emitterUpdateScript [
   &EmitterState()
   &SpawnRate(
    Spawn Probability=HlslExpression() {hlslExpression="Emitter.Age > .2 ? 1 : 0"}
    SpawnRate=&Lerp_Float(A=0, B=50, Alpha=$System.Dive)
    Use Spawn Probability=true
   )
  ]
  particleSpawnScript [
   /Niagara/Modules/Spawn/Initialization/InitializeParticle(
    Color=&MakeLinearColor(
     R:NiagaraFloat=0.2292
     G:NiagaraFloat=0.1989
     B:NiagaraFloat=0.1838
     A:NiagaraFloat=&UniformRangedFloat(Minimum=0.1, Maximum=0.2)
    )
    Lifetime=&UniformRangedFloat(Minimum=0.3, Maximum=0.4)
    Module.Write Color=true
    Module.Write Lifetime=true
    Module.Write Mass=true
    Module.Write Material Random=true
    Module.Write Position=false
    Module.Write Scale=false
    Module.Write SpriteRotation=true
    Module.Write SpriteSize=true
    Position=&SimulationPosition()
    Sprite Rotation=&UniformRangedFloat(Minimum=0, Maximum=360)
    Sprite Size=&UniformRangedVector2D(Minimum=<1.25 100>, Maximum=<0.8 200>)
   )
   &SkeletalMeshLocation(
    Bone / Socket Index=4
    Bone Sampling Mode=::NewEnumerator4
    Inherit Velocity=false
    MaskSpawnGroup=false
    Sampled Orientation Coordinate Space=::Local
    Skeletal Mesh=$System.Skeletal Mesh
   )
   &TorusLocation(
    Align Torus With Axis=true
    Coordinate Space:ENiagaraCoordinateSpace=::World
    Handle Radius=250
    Large Radius=200
    MaskSpawnGroup=false
    Module.Override Randomness=false
    Offset=&RotateVectorByQuaternion(Vector To Rotate=<500 0 0>, Quaternion=$Output.SkeletalMeshLocation.SampledRotation)
    Torus Axis=&RotateVectorByQuaternion(Vector To Rotate=<1 0 0>, Quaternion=$Output.SkeletalMeshLocation.SampledRotation)
    Torus Origin:NiagaraPosition=$Output.SkeletalMeshLocation.SampledPosition
   ) {enabled=false}
   &SphereLocation(
    MaskSpawnGroup=false
    Non Uniform Scale=&RotateVectorByQuaternion(Vector To Rotate=<1 1 1>, Quaternion=$Output.SkeletalMeshLocation.SampledRotation)
    Offset=&Add_Vector(
     A=&RotateVectorByQuaternion(Vector To Rotate=<100 0 0>, Quaternion=$Output.SkeletalMeshLocation.SampledRotation)
     B=&TransformVector(Vector=<0 0 -200>, Source Space=::World, Destination Space=::World)
    )
    Override Seed=false
    Sphere Orientation Axis:Vector3f=&RotateVectorByQuaternion(Vector To Rotate=<100 0 0>, Quaternion=$Output.SkeletalMeshLocation.SampledRotation)
    Sphere Origin:NiagaraPosition=$Output.SkeletalMeshLocation.SampledPosition
    Sphere Radius=80
    Surface Only=false
    UseNonUniformScale=true
    UseOffset=true
    UseSphereOrientation=true
   ) {enabled=false}
   &ShapeLocation(
    Cylinder Height=10
    Cylinder Radius=85
    MaskSpawnGroup=false
    Offset=<0 0 -100>
    Offset Coordinate Space=::Local
    Offset Mode=::NewEnumerator0
    Override Seed=false
    Shape Primitive=::NewEnumerator1
    Surface Only=true
    Use Endcaps In Surface Only Mode=false
    UseOwnerScale=true
   )
   &InheritVelocity(Module.Limit Speed=true)
  ]
  particleUpdateScript [
   &ParticleState()
   &Color(Scale Alpha=&FloatFromCurve(
     FloatCurve=DataInterface() {propertyValues {Curve {keys [<at 0 0 RCIM_Cubic>,<at 0.1855 0.9473 RCIM_Cubic>,<at 1 0 RCIM_Cubic>]}}}
     Scale Curve=0.1
    ))
   &SolveForcesAndVelocity(Clamp Velocity=false, Limit Acceleration=false)
   &ScaleSpriteSize(Scale Factor=&Vector2DFromCurve(
     Vector2Curve=DataInterface() {propertyValues {
       XCurve {keys [<at -0.0069 0.5 RCIM_Cubic>,<at 0.4918 1 RCIM_Cubic>,<at 1 0.5 RCIM_Cubic>]}
       YCurve {keys [<at 0 0.0245 RCIM_Cubic>, {arriveTangent=-0.0245, interpMode=RCIM_Cubic, leaveTangent=-0.0245, time=0.4975, value=1}, <at 1 0 RCIM_Cubic>]}
      }}
     ScaleCurve=<5 3>
    ))
  ]
  rendererClasses [&NiagaraSpriteRendererProperties]
  renderers [{rendererClass=&NiagaraSpriteRendererProperties}]
 }
 Local_Space_Skydive {
  emitterSpawnScript [/SetVariables(Emitter.Number of Trails=HlslExpression() {hlslExpression="Engine.QualityLevel > 2 ? 4 : 2"}) {isSetParameters=true}]
  emitterUpdateScript [
   &EmitterState(Life Cycle Mode=::NewEnumerator1, Loop Duration=1, UseLoopDelay=false)
   &SpawnBurst_Instantaneous(
    Spawn Count=$Emitter.Initial.Number of Trails
    Spawn Group=-1
    Spawn Probability=HlslExpression() {hlslExpression="Emitter.LoopCount == 0"}
    Use Spawn Probability=true
   )
   &SpawnRate(SpawnRate=60, Use Spawn Probability=false)
  ]
  particleSpawnScript [
   &FN_SetSpawnGroup(Number of Spawn Groups:NiagaraInt32=$Emitter.Initial.Number of Trails)
   &self:LockParticlesToBones() {module=LockParticlesToBones}
   &InitializeRibbon(
    Lifetime=&MakeCustomFloatFromBool(Bool=$Particles.Locked, True Float=100, False Float=/Niagara/DynamicInputs/Math/Min_Float(A:NiagaraFloat=$Emitter.Age, B:NiagaraFloat=0.3))
    Module.Write Color=true
    Module.Write Lifetime=true
    Module.Write Mass=true
    Module.Write Material Random=true
    Module.Write Position=false
    Module.Write RibbonTwist=false
    Module.Write RibbonWidth=true
    Ribbon Width=15.4638
   )
   &FN_FlipRibbonV() {enabled=false}
   &SkeletalMeshLocation(
    Bone / Socket Index=$Particles.SpawnGroup
    Bone Sampling Mode=::NewEnumerator4
    Inherit Velocity=false
    MaskSpawnGroup=false
    Skeletal Mesh=$System.Skeletal Mesh
   )
   /Game/Effects/Niagara/Modules/Skydiving/FN_Skydive_Velocity_v4/FN_Skydive_Velocity(
    Bone Orientation:Quat4f=$Output.SkeletalMeshLocation.SampledRotation
    Bone Position:Vector3f=$Output.SkeletalMeshLocation.SampledPosition
    FE Speed=1103.7715
    FE Spread:NiagaraFloat=/Game/Effects/Niagara/DynamicInputs/Skydiving/SkydiveHandsFeetFloatModifier(Hands:NiagaraFloat=&Lerp_Float(A=0, B=200, Alpha=$System.Dive), Feet:NiagaraFloat=0)
    SpawnGroup:NiagaraInt32=$Particles.SpawnGroup
    Speed=221.3639
    Spread=148.0048
   ) {module=FN_Skydive_Velocity001}
   /SetVariables(Particles.OffsetAmount:NiagaraFloat=&SkydiveHandsFeetFloatModifier(Hands:NiagaraFloat=27.1231, Feet:NiagaraFloat=31.0977)) {isSetParameters=true}
   /Game/Effects/Niagara/Modules/Skydiving/FN_SkydivePositionOffset_v1/FN_SkydivePositionOffset(
    Bone Orientation:Quat4f=$Output.SkeletalMeshLocation.SampledRotation
    Bone Position:Vector3f=$Output.SkeletalMeshLocation.SampledPosition
    Offset:NiagaraFloat=$Particles.OffsetAmount
   )
  ]
  particleUpdateScript [
   &ParticleState()
   /SetVariables(Particles.Age:NiagaraFloat=&MakeCustomFloatFromBool(Bool=$Particles.Locked, True Float=0, False Float=$Particles.Age)) {isSetParameters=true}
   /SetVariables(Particles.NormalizedAge:NiagaraFloat=&MakeCustomFloatFromBool(Bool=$Particles.Locked, True Float=0, False Float=$Particles.NormalizedAge)) {isSetParameters=true}
   &SkeletalMeshLocation(
    Apply To Particle Position=false
    Bone / Socket Index=$Particles.SpawnGroup
    Bone Sampling Mode=::NewEnumerator4
    Inherit Velocity=false
    Interpolate Mesh Sampling=false
    MaskSpawnGroup=false
    Skeletal Mesh=$System.Skeletal Mesh
   ) {module=SkeletalMeshLocation001}
   &FN_SkydivePositionOffset(
    Bone Orientation:Quat4f=$Output.SkeletalMeshLocation001.SampledRotation
    Bone Position:Vector3f=$Output.SkeletalMeshLocation001.SampledPosition
    Offset:NiagaraFloat=$Particles.OffsetAmount
    Write to Intrinsic Properties=false
   ) {module=FN_SkydivePositionOffset001}
   &CurlNoiseForce(
    Noise Strength=&Lerp_Float(A=0, B=1000, Alpha=$System.Dive)
    Pan Noise=true
    Pan Noise Field=<0 0 2>
    Sample Position:Vector3f=&Add_Vector(
     A=Linked() {linkedVariable {name=Particles.Position, type=NiagaraPosition}}
     B=&RandomVector(Vector Scale=50, Randomness Mode=::NewEnumerator1, Random Seed=&SpawnGroup(), Fixed Random Seed=true)
    )
   ) {enabled=false}
   &InheritSourceMovement(
    Applied Position Scale=&VectorFromFloat(Value=&Lerp_Float(A=0.6, B=0.9, Alpha=$System.Dive))
    ApplyForce=false
    ApplyPosition=true
    ApplyPositionFalloff=false
   )
   &SolveForcesAndVelocity(Clamp Velocity=false, Limit Acceleration=false)
   /SetVariables(Particles.RenderPosition:Vector3f=&MakeCustomVectorFromBool(
     Bool=$Particles.Locked
     True Vector=&Add_Vector(A=$Output.FN_SkydivePositionOffset001.OffsetDirection, B=$Output.SkeletalMeshLocation001.SampledPosition)
     False Vector=Linked() {linkedVariable {name=Particles.Position, type=NiagaraPosition}}
    )) {isSetParameters=true}
   &FN_RibbonMaterial(
    Opacity Pan U Offset:NiagaraFloat=/Game/Effects/Niagara/DynamicInputs/Random/FN_SeededRandomRangeFloat(Minimum:NiagaraFloat=0, Maximum:NiagaraFloat=1, Random Seed:NiagaraInt32=$Particles.SpawnGroup)
    Opacity Pan U Speed=1.6096
    Opacity Power=0.3455
    Opacity Scale U:NiagaraFloat=&Multiply_Float(
     A=&FN_SeededRandomRangeFloat(Minimum:NiagaraFloat=0.7182, Maximum:NiagaraFloat=1.0493, Random Seed:NiagaraInt32=$Particles.SpawnGroup)
     B=&Lerp_Float(A=1, B=0.2, Alpha=$System.Dive)
    )
    Scale Alpha:NiagaraFloat=&FloatFromCurve(
     FloatCurve=DataInterface() {propertyValues {Curve {keys [<at 0 1>,<at 0.7252 -0.0057>]}}}
     Scale Curve=/Game/Effects/Niagara/DynamicInputs/Math/FN_Fit(
      Clamp:NiagaraBool=true
      Value:NiagaraFloat=$Engine.Owner.LODDistance
      SourceMin:NiagaraFloat=400
      SourceMax:NiagaraFloat=2000
      DestMin:NiagaraFloat=0.03
      DestMax:NiagaraFloat=0.2
     )
    )
    TIme:NiagaraFloat=$System.FN_Integrator.Value
   )
  ]
  rendererClasses [&NiagaraRibbonRendererProperties]
  renderers [{rendererClass=&NiagaraRibbonRendererProperties}]
 }
}
summary {
 systemName=NS_Contrails_Basic_v2
 emitters [
  {emitterName=Local_Space_Skydive, rendererClasses [&NiagaraRibbonRendererProperties]}
  {emitterName=FN_SpeedLines, simTarget=GPUComputeSim, rendererClasses [&NiagaraSpriteRendererProperties]}
 ]
 userVariables [
  User.bIsFrontEnd {type=NiagaraBool}
  User.bIsFrontEndPreview {type=NiagaraBool}
  User.ParaGlideLeanAlpha {type=NiagaraFloat, defaultValue=0, description=""}
  User.RibbonIDOffset {type=NiagaraInt32}
 ]
}
