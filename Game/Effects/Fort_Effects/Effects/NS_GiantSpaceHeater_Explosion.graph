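# Niagara system for the giant space heater explosion, layering flash meshes, shockwave, glow, component-rendered light, smoke ring and colliding GPU sparks.
# Burst debris meshes fly out under gravity and drag, with ribbon trails following ExtraMeshes through particle attribute reads.
# A camera-facing Grid2D fluid sim, bound to a SourceParticle emitter, adds smoke and fire with culling limits.
# User.ExplosionSource offsets many layers, while System.ScaleDistance scales spark and debris spawning.
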NiagaraSystem /Game/Effects/Fort_Effects/Effects/NS_GiantSpaceHeater_Explosion
dependencies {
 usedDataInterfaces [
  &NiagaraDataInterfaceCurve
  &NiagaraDataInterfaceVectorCurve
  &NiagaraDataInterfaceColorCurve
  &NiagaraDataInterfaceParticleRead
  /Script/Niagara.NiagaraDataInterfaceGrid2DCollection
  &NiagaraDataInterfacePlatformSet
  &NiagaraDataInterfaceCamera
  /Script/Niagara.NiagaraDataInterfaceRenderTarget2D
  /Script/Niagara.NiagaraDataInterfaceEmitterProperties
 ]
 usedDynamicInputs []
 usedModules [
  &self:SystemSpawnScript.SystemScriptSource.SystemScriptGraph.NiagaraNodeAssignment_0.SetVariables_#1
  &SystemState
 ]
}
emitters {
 ExtraMeshes {
  emitterUpdateScript [
   &EmitterState(Life Cycle Mode=::NewEnumerator1, Loop Behavior=::NewEnumerator1, Loop Duration=0.5, UseLoopDelay=false)
   &SpawnBurst_Instantaneous(Spawn Count=5, Spawn Group=-1, Spawn Probability=HlslExpression() {hlslExpression="Emitter.LoopCount == 0"}, Use Spawn Probability=true)
  ]
  particleSpawnScript [
   /SetVariables(Particles.RandonInt2:NiagaraInt32=&UniformRangedInt(Minimum=0, Maximum=1)) {isSetParameters=true, module=SetVariables_#2}
   /Niagara/Modules/Spawn/Initialization/InitializeParticle(
    Lifetime=&MaskFloatBySpawnGroup(Masked Float=&UniformRangedFloat(Minimum=2, Maximum=3), Passthrough Float=2)
    Mesh Scale=&RandomRangeVector(Minimum=<0.2 0.2 0.2>, Maximum=<0.6 0.6 0.6>)
    Module.Write Color=true
    Module.Write Lifetime=true
    Module.Write Mass=true
    Module.Write Material Random=true
    Module.Write Position=true
    Module.Write Scale=true
    Module.Write SpriteRotation=false
    Module.Write SpriteSize=false
    Position=&AddVectorToPosition(Position=$Engine.Emitter.SimulationPosition, Vector=$User.ExplosionSource)
   )
   /Niagara/Modules/Spawn/Location/SphereLocation(
    Hemisphere Z=true
    MaskSpawnGroup=false
    Module.Override Randomness=false
    Sphere Radius=&Multiply_Float(A=150, B=$System.ScaleDistance)
    Surface Only=true
    Surface Only Band Thickness=2
   )
   &InitialMeshOrientation(
    Rotation=&UniformRangedVector(Minimum=<0 0 0>, Maximum=<1 1 1>)
    Use Orientation Vector=true
    Use Rotation Vector=true
   ) {module=InitialMeshOrientation001}
   &AddVelocity(
    Velocity=$Output.SphereLocation.SphereNormal
    Velocity Speed Scale=&RandomRangeFloat(Minimum=1500, Maximum=2500)
   )
   &AddRotationalVelocity(Rotation Rate=&UniformRangedVector(Minimum=<-1 -1 -1>, Maximum=<1 1 1>))
   &AddVelocityInCone(
    Cone Angle=&RandomRangeFloat(Minimum=0, Maximum=120)
    Cone Axis=<0 0 1>
    Module.Use Velocity Falloff On Cone Axis=false
    Velocity Strength=&RandomRangeFloat(Minimum=1500, Maximum=2500)
   ) {enabled=false, module=AddVelocityInCone001}
  ]
  particleUpdateScript [
   &ParticleState()
   &GravityForce(Gravity=<0 0 -1200>)
   &Drag(
    Drag=&FloatFromCurve(FloatCurve=DataInterface() {propertyValues {Curve {keys [
         <at 0 0>
         {arriveTangent=-4.1608, interpMode=RCIM_Cubic, leaveTangent=-4.1608, tangentMode=RCTM_User, time=0.25, value=1}
         {arriveTangent=-0.1083, interpMode=RCIM_Cubic, leaveTangent=-0.1083, tangentMode=RCTM_User, time=1, value=0}
        ]}}})
    Use Linear Drag=true
    Use Rotational Drag=false
   )
   &ScaleColor(
    Scale Alpha=&FloatFromCurve(FloatCurve=DataInterface() {propertyValues {Curve {keys [<at -0.0046 1.0113 RCIM_Cubic>,<at 1.0011 0.002 RCIM_Cubic>]}}})
    ScaleA=true
    ScaleRGB=true
   )
   &SolveForcesAndVelocity(Clamp Velocity=false, Limit Acceleration=false)
   &SolveRotationalForcesAndVelocity()
   &ScaleMeshSize(Scale Factor=&VectorFromFloat(Value=&FloatFromCurve(FloatCurve=DataInterface() {propertyValues {Curve {keys [<at 0 1>,<at 0.7956 1.0207>,<at 1 0>]}}})))
   /Game/Effects/Niagara/Modules/Attributes/FN_SetChildPosition(Child Position:NiagaraPosition=$Particles.Position)
  ]
  rendererClasses [&NiagaraMeshRendererProperties]
  renderers [
   {rendererClass=&NiagaraMeshRendererProperties, rendererIndex=0}
   {rendererClass=&NiagaraMeshRendererProperties, rendererIndex=1}
  ]
 }
 ExtraMeshes_Gen5 {
  emitterUpdateScript [
   &EmitterState(Life Cycle Mode=::NewEnumerator1, Loop Behavior=::NewEnumerator1, Loop Duration=0.5, UseLoopDelay=false)
   &SpawnBurst_Instantaneous(Spawn Count=5, Spawn Group=-1, Spawn Probability=HlslExpression() {hlslExpression="Emitter.LoopCount == 0"}, Use Spawn Probability=true)
  ]
  particleSpawnScript [
   /SetVariables(Particles.RandonInt2:NiagaraInt32=&UniformRangedInt(Minimum=0, Maximum=1)) {isSetParameters=true, module=SetVariables_#2}
   &InitializeParticle(
    Lifetime=&MaskFloatBySpawnGroup(Masked Float=&UniformRangedFloat(Minimum=2, Maximum=3), Passthrough Float=2)
    Mesh Scale=&RandomRangeVector(Minimum=<0.2 0.2 0.2>, Maximum=<0.6 0.6 0.6>)
    Module.Write Color=true
    Module.Write Lifetime=true
    Module.Write Mass=true
    Module.Write Material Random=true
    Module.Write Position=false
    Module.Write Scale=true
    Module.Write SpriteRotation=false
    Module.Write SpriteSize=false
    Position=&SimulationPosition()
   )
   &SphereLocation(
    Hemisphere Z=true
    MaskSpawnGroup=false
    Module.Override Randomness=false
    Sphere Radius=&Multiply_Float(A=150, B=$System.ScaleDistance)
    Surface Only=true
    Surface Only Band Thickness=2
   )
   &InitialMeshOrientation(
    Rotation=&UniformRangedVector(Minimum=<0 0 0>, Maximum=<1 1 1>)
    Use Orientation Vector=true
    Use Rotation Vector=true
   ) {module=InitialMeshOrientation001}
   &AddRotationalVelocity(Rotation Rate=&UniformRangedVector(Minimum=<-1 -1 -1>, Maximum=<1 1 1>))
   &AddVelocityInCone(
    Cone Angle=&RandomRangeFloat(Minimum=0, Maximum=120)
    Cone Axis=<0 0 1>
    Module.Use Velocity Falloff On Cone Axis=false
    Velocity Strength=&RandomRangeFloat(Minimum=150, Maximum=2500)
   ) {module=AddVelocityInCone001}
  ]
  particleUpdateScript [
   &ParticleState()
   &GravityForce(Gravity=<0 0 -1200>)
   &Drag(
    Drag=&FloatFromCurve(FloatCurve=DataInterface() {propertyValues {Curve {keys [
         <at 0 0>
         {arriveTangent=-4.1608, interpMode=RCIM_Cubic, leaveTangent=-4.1608, tangentMode=RCTM_User, time=0.25, value=1}
         {arriveTangent=-0.1083, interpMode=RCIM_Cubic, leaveTangent=-0.1083, tangentMode=RCTM_User, time=1, value=0}
        ]}}})
    Use Linear Drag=true
    Use Rotational Drag=false
   )
   &ScaleColor(
    Scale Alpha=&FloatFromCurve(FloatCurve=DataInterface() {propertyValues {Curve {keys [<at -0.0046 1.0113 RCIM_Cubic>,<at 1.0011 0.002 RCIM_Cubic>]}}})
    ScaleA=true
    ScaleRGB=true
   )
   &SolveForcesAndVelocity(Clamp Velocity=false, Limit Acceleration=false)
   &SolveRotationalForcesAndVelocity()
   &ScaleMeshSize(Scale Factor=&VectorFromFloat(Value=&FloatFromCurve(FloatCurve=DataInterface() {propertyValues {Curve {keys [<at 0 1>,<at 0.7956 1.0207>,<at 1 0>]}}})))
   &FN_SetChildPosition(Child Position:NiagaraPosition=$Particles.Position)
  ]
  rendererClasses [&NiagaraMeshRendererProperties]
  renderers [
   {rendererClass=&NiagaraMeshRendererProperties, rendererIndex=0}
   {rendererClass=&NiagaraMeshRendererProperties, rendererIndex=1}
  ]
 }
 FN_ParticleRibbon {
  emitterSpawnScript [/Game/Effects/Niagara/Modules/Attributes/FN_Source_v2(AttributeReader=DataInterface() {propertyValues {EmitterBinding {bindingMode=Other, emitterName=ExtraMeshes}}})]
  emitterUpdateScript [
   &EmitterState(Life Cycle Mode=::NewEnumerator1, Loop Behavior=::NewEnumerator1, Loop Duration=1.5, UseLoopDelay=false)
   &SpawnBurst_Instantaneous(Spawn Count=5, Spawn Group=-1, Spawn Probability=HlslExpression() {hlslExpression="Emitter.LoopCount == 0"}, Use Spawn Probability=true)
   &SpawnRate(
    SpawnRate=/Game/Effects/Niagara/DynamicInputs/Attributes/FN_GetEmitterParticleCount(AttributeReader:NiagaraDataInterfaceParticleRead=$Emitter.AttributeReader, Multiplier:NiagaraFloat=25)
    Use Spawn Probability=false
   )
  ]
  particleSpawnScript [
   /Game/Effects/Niagara/Modules/Attributes/FN_SourceParentID()
   /Game/Effects/Niagara/Modules/Attributes/FN_SourceLocation(Position:Vector3f=/Game/Effects/Niagara/DynamicInputs/Attributes/FN_GetVectorByID(Attrib name:ENiagaraCommonVectors=::NewEnumerator0, Default on Invalid001:Vector3f=<0 0 0>))
   &InitializeRibbon(
    Module.Write Color=true
    Module.Write Lifetime=true
    Module.Write Mass=true
    Module.Write Material Random=true
    Module.Write Position=false
    Module.Write RibbonTwist=false
    Module.Write RibbonWidth=true
    Ribbon Width=&Lerp_Float(
     A=400
     B=250
     Alpha=/Game/Effects/Niagara/DynamicInputs/Attributes/FN_GetFloatByID(Attribute:ENiagaraCommonFloats=::NewEnumerator1, Default on Invalid:NiagaraFloat=0)
    )
   )
  ]
  particleUpdateScript [
   /Game/Effects/Niagara/Modules/Attributes/FN_UpdateSourcedRibbbons(
    Locked Particles Position:Vector3f=&FN_GetVectorByID(
     Attrib name:ENiagaraCommonVectors=::NewEnumerator0
     Default on Invalid001:Vector3f=Linked() {linkedVariable {name=Particles.Position, type=NiagaraPosition}}
    )
    Ribbon Width:NiagaraFloat=/Game/Effects/Niagara/DynamicInputs/Math/FN_Fit(
     Clamp:NiagaraBool=true
     Value:NiagaraFloat=&FN_GetFloatByID(Attribute:ENiagaraCommonFloats=::NewEnumerator0, Default on Invalid:NiagaraFloat=0)
     SourceMin:NiagaraFloat=0
     SourceMax:NiagaraFloat=1
     DestMin:NiagaraFloat=200
     DestMax:NiagaraFloat=0
    )
   ) {module=FN_UpdateLockedParticles}
   &ParticleState() {enabled=false}
   /SetVariables(Particles.RenderRibbonWidth:NiagaraFloat=&Multiply_Float(
     A=&Multiply_Float(A=$Particles.RibbonWidth, B=$Particles.NormalizedAge)
     B=&Lerp_Float(
      A=1
      B=1
      Alpha=&FN_GetFloatByID(Attribute:ENiagaraCommonFloats=::NewEnumerator0, Default on Invalid:NiagaraFloat=1)
     )
    )) {isSetParameters=true, module=SetVariables_#3}
   /SetVariables(Particles.RenderRibbonWidth:NiagaraFloat=&FloatFromCurve(FloatCurve=DataInterface() {propertyValues {Curve {keys [<at 0 1>,<at 1 0>]}}}, Scale Curve=80)) {enabled=false, isSetParameters=true}
   &DynamicMaterialParameters(
    Index 0 Param 1=-1.5
    Index 0 Param 2=2
    Index 0 Param 3=1
    Index 0 Param 4=&FloatFromCurve(
     FloatCurve=DataInterface() {propertyValues {Curve {keys [<at 0 1>,<at 0.9 -0.4>]}}}
     CurveIndex=&FN_GetFloatByID(Attribute:ENiagaraCommonFloats=::NewEnumerator1, Default on Invalid:NiagaraFloat=0)
    )
    Module.Param0WriteEnabled=true
    Module.Param1WriteEnabled=true
    Module.Param2WriteEnabled=true
    Module.Param3WriteEnabled=true
   )
   &ScaleColor(
    Scale Alpha=&FloatFromCurve(FloatCurve=DataInterface() {propertyValues {Curve {keys [{arriveTangent=-3.0858, interpMode=RCIM_Cubic, leaveTangent=-3.0858, tangentMode=RCTM_User, time=0.25, value=0.9509}, <at 1 0>]}}})
    Scale RGB=&VectorFromCurve(
     VectorCurve=DataInterface() {propertyValues {
       XCurve {keys [{arriveTangent=-38.9035, interpMode=RCIM_Cubic, leaveTangent=-38.9035, tangentMode=RCTM_User, time=0.15, value=15}, <at 0.4 0.1264>]}
       YCurve {keys [{arriveTangent=-32.8642, interpMode=RCIM_Cubic, leaveTangent=-32.8642, tangentMode=RCTM_User, time=0.15, value=15}, <at 0.4 0.1238>]}
       ZCurve {keys [{arriveTangent=-34.4245, interpMode=RCIM_Cubic, leaveTangent=-34.4245, tangentMode=RCTM_User, time=0.15, value=15}, <at 4 0.1266>]}
      }}
     CurveIndex=$Emitter.NormalizedLoopAge
    )
    ScaleA=true
    ScaleRGB=true
   )
  ]
  rendererClasses [&NiagaraRibbonRendererProperties]
  renderers [{rendererClass=&NiagaraRibbonRendererProperties}]
 }
 FN_Sprite {
  simTarget=GPUComputeSim
  emitterUpdateScript [
   &EmitterState(Life Cycle Mode=::NewEnumerator1, Loop Behavior=::NewEnumerator1, Loop Duration=0.2, UseLoopDelay=false)
   &SpawnBurst_Instantaneous(Spawn Count=250, Use Spawn Probability=false)
   &SpawnRate(SpawnRate=&Multiply_Float(A=5000, B=$System.ScaleDistance), Use Spawn Probability=false)
  ]
  particleSpawnScript [
   &InitializeParticle(
    Lifetime=&UniformRangedFloat(Minimum=1, Maximum=3)
    Module.Write Color=true
    Module.Write Lifetime=true
    Module.Write Mass=true
    Module.Write Material Random=true
    Module.Write Position=true
    Module.Write Scale=false
    Module.Write SpriteRotation=true
    Module.Write SpriteSize=true
    Position=&AddVectorToPosition(Position=$Engine.Emitter.SimulationPosition, Vector=$User.ExplosionSource)
    Sprite Rotation=&UniformRangedFloat(Minimum=0, Maximum=360)
    Sprite Size=&Vector2DFromFloat(Value=&Multiply_Float(A=&UniformRangedFloat(Minimum=6, Maximum=12), B=$System.ScaleDistance))
   )
   &SphereLocation(
    Hemisphere Z=true
    MaskSpawnGroup=false
    Module.Override Randomness=false
    Sphere Radius=&Multiply_Float(A=350, B=$System.ScaleDistance)
    Surface Only=true
    Surface Only Band Thickness=2
   )
   &AddVelocityFromPoint(
    Module.Use Falloff Distance=false
    Origin Offset=&UniformRangedVector(Minimum=<-1 -1 -100>, Maximum=<1 1 -100>)
    Origin Offset Coordinate Space=::Simulation
    Velocity Strength=&UniformRangedFloat(Minimum=800, Maximum=1500)
   ) {enabled=false}
   &AddVelocity(
    Velocity=$Output.SphereLocation.SphereNormal
    Velocity Speed Scale=&RandomRangeFloat(Minimum=500, Maximum=1250)
   )
   &Color(Color=<rgba 1 0.25 0.1 0.5>)
  ]
  particleUpdateScript [
   &ParticleState()
   &Collision(
    Advanced Aging Rate=1
    Age Colliding Particles=false
    Collision Enabled=&SetBoolByFloatComparison(A=0.5, B=$Particles.Age, Comparison Type=::NewEnumerator6)
    Kill Offscreen Particles=true
    Module.Control Rest State Via Surface Normal=false
    Module.EnableMaxCollisionCount=false
    Module.Randomize Collision Normal Vector=true
    Randomize Collision Normal=0.01
    Restitution=0.2
   )
   &AccelerationForce(Acceleration=<0 0 -1600>)
   &SolveForcesAndVelocity(Clamp Velocity=false, Limit Acceleration=false)
   &ScaleColor(
    Scale Mode=::NewEnumerator1
    Scale RGBA=&Vector4FromCurve(Vector4Curve=DataInterface() {propertyValues {
       AlphaCurve {keys [<at 0.0024 0>,<at 0.1271 15>,<at 0.7369 0.9976>,<at 0.9368 0>]}
       BlueCurve {keys [<at 0.1075 26.0417>,<at 0.4233 0>]}
       GreenCurve {keys [<at 0.1075 82.0312>,<at 0.4233 0>]}
       RedCurve {keys [<at 0.1075 250>,<at 0.4233 0>]}
      }})
    ScaleRGBA=true
   ) {module=ScaleColor001}
   &SpriteRotationRate(Rotation Rate=2)
  ]
  rendererClasses [&NiagaraSpriteRendererProperties]
  renderers [{rendererClass=&NiagaraSpriteRendererProperties}]
 }
 Fireworks_Mesh001 {
  emitterUpdateScript [
   &EmitterState(Life Cycle Mode=::NewEnumerator1, Loop Behavior=::NewEnumerator1, Loop Duration=0.25, UseLoopDelay=true)
   &SpawnBurst_Instantaneous(Spawn Count=1, Spawn Time=0.05, Use Spawn Probability=false)
  ]
  particleSpawnScript [
   &InitializeParticle(
    Color=&UniformRangedLinearColor(Minimum=<rgba 1 1 1 0.7>, Maximum=<rgba 1 0.5 0 0.7>)
    Lifetime=&UniformRangedFloat(Minimum=1, Maximum=1.25)
    Mesh Scale=&Multiply_VectorByFloat(Vector=<0.3 0.3 0.3>, Float=1.2)
    Module.Write Color=true
    Module.Write Lifetime=true
    Module.Write Mass=true
    Module.Write Material Random=true
    Module.Write Position=true
    Module.Write Scale=true
    Module.Write SpriteRotation=false
    Module.Write SpriteSize=false
    Position=&AddVectorToPosition(Position=$Engine.Emitter.SimulationPosition, Vector=$User.ExplosionSource)
    Sprite Size=<25 25>
   )
   &InitialMeshOrientation(
    Orientation Vector=&UniformRangedVector(Minimum=<0 0 0>, Maximum=<1 1 1>)
    Use Orientation Vector=true
    Use Rotation Vector=false
   ) {module=InitialMeshOrientation001}
  ]
  particleUpdateScript [
   &ParticleState()
   &ScaleColor(
    Scale Alpha=&FloatFromCurve(FloatCurve=DataInterface() {propertyValues {Curve {keys [{arriveTangent=-3.2421, interpMode=RCIM_Cubic, leaveTangent=-3.2421, tangentMode=RCTM_User, time=0.1015, value=1.0322}, <at 1 0>]}}})
    Scale RGB=&VectorFromCurve(VectorCurve=DataInterface() {propertyValues {
       XCurve {keys [<at -0.0011 0.01>,<at 0.2555 0>,<at 0.4988 0>,<at 1 0.5>]}
       YCurve {keys [<at -0.0011 15>,<at 0.2492 9.0413>,<at 0.4988 5>,<at 1 1.649>]}
       ZCurve {keys [<at -0.0011 1.0067>,<at 0.2555 1>,<at 0.4988 1>,<at 0.9989 1>]}
      }})
    ScaleA=true
    ScaleRGB=true
   )
   &ScaleMeshSize(Scale Factor=&VectorFromFloat(Value=&FloatFromCurve(FloatCurve=DataInterface() {propertyValues {Curve {keys [
          {arriveTangent=9.455, interpMode=RCIM_Cubic, leaveTangent=9.455, tangentMode=RCTM_User, time=0, value=1}
          {arriveTangent=1.6472, interpMode=RCIM_Cubic, leaveTangent=1.6472, tangentMode=RCTM_User, time=0.2427, value=2.1701}
          <at 1.0062 2.5692>
         ]}}})))
   &DynamicMaterialParameters(
    Index 0 Param 1=1
    Index 0 Param 2=1
    Index 0 Param 3=&FloatFromCurve(FloatCurve=DataInterface() {propertyValues {Curve {keys [
         {arriveTangent=2.6532, interpMode=RCIM_Cubic, leaveTangent=2.6532, tangentMode=RCTM_User, time=0.0008, value=0.1517}
         {arriveTangent=0.491, interpMode=RCIM_Cubic, leaveTangent=0.491, tangentMode=RCTM_User, time=1, value=1.9}
        ]}}})
    Index 0 Param 4=&FloatFromCurve(FloatCurve=DataInterface() {propertyValues {Curve {keys [<at 0.0842 1.8976>,<at 1 1.2>]}}})
    Module.Param0WriteEnabled=true
    Module.Param1WriteEnabled=true
    Module.Param2WriteEnabled=true
    Module.Param3WriteEnabled=true
   )
  ]
  rendererClasses [&NiagaraMeshRendererProperties]
  renderers [{rendererClass=&NiagaraMeshRendererProperties, rendererIndex=0}]
 }
 GlowSphereMesh001 {
  emitterUpdateScript [
   &EmitterState()
   &SpawnBurst_Instantaneous(Spawn Count=1, Spawn Time=0.1, Use Spawn Probability=false)
  ]
  particleSpawnScript [&InitializeParticle(
    Color=<rgba 10 1 0.02 2>
    Lifetime=&UniformRangedFloat(Minimum=0.25, Maximum=0.25)
    Mesh Scale=<12 12 12>
    Module.Write Color=true
    Module.Write Lifetime=true
    Module.Write Mass=true
    Module.Write Material Random=true
    Module.Write Position=true
    Module.Write Scale=true
    Module.Write SpriteRotation=false
    Module.Write SpriteSize=false
    Position=&AddVectorToPosition(Position=$Engine.Emitter.SimulationPosition, Vector=$User.ExplosionSource)
   )]
  particleUpdateScript [
   &ParticleState()
   &DynamicMaterialParameters(
    Index 0 Param 1=&FloatFromCurve(FloatCurve=DataInterface() {propertyValues {Curve {keys [<at 0 1>,<at 0.3172 0.97>,<at 1 1>]}}})
    Index 0 Param 2=&FloatFromCurve(FloatCurve=DataInterface() {propertyValues {Curve {keys [<at 0 15>,<at 1 0>]}}})
    Module.Param0WriteEnabled=true
    Module.Param1WriteEnabled=true
    Module.Param2WriteEnabled=true
    Module.Param3WriteEnabled=true
   )
   &ScaleMeshSize(Scale Factor=/Niagara/DynamicInputs/ScaleByCurve/ScaleVectorByCurve(
     InitialValue:Vector3f=<1 1 1>
     ScaleCurve:NiagaraDataInterfaceVectorCurve=DataInterface() {propertyValues {
       XCurve {keys [
         {arriveTangent=4.2859, interpMode=RCIM_Cubic, leaveTangent=4.2859, tangentMode=RCTM_User, time=0, value=0}
         {arriveTangent=2, interpMode=RCIM_Cubic, leaveTangent=2, tangentMode=RCTM_User, time=0.25, value=1}
         {interpMode=RCIM_Cubic, tangentMode=RCTM_User, time=1, value=2}
        ]}
       YCurve {keys [
         {arriveTangent=5.8758, interpMode=RCIM_Cubic, leaveTangent=5.8759, tangentMode=RCTM_User, time=0, value=0}
         {arriveTangent=2, interpMode=RCIM_Cubic, leaveTangent=2, tangentMode=RCTM_User, time=0.25, value=1}
         {interpMode=RCIM_Cubic, tangentMode=RCTM_User, time=1, value=2}
        ]}
       ZCurve {keys [
         {arriveTangent=4.7007, interpMode=RCIM_Cubic, leaveTangent=4.7007, tangentMode=RCTM_User, time=0, value=0}
         {arriveTangent=2, interpMode=RCIM_Cubic, leaveTangent=2, tangentMode=RCTM_User, time=0.25, value=1}
         {interpMode=RCIM_Cubic, tangentMode=RCTM_User, time=1, value=2}
        ]}
      }}
     CurveIndex:NiagaraFloat=$Particles.NormalizedAge
    ))
   &ScaleColor(
    Scale Alpha=&FloatFromCurve(FloatCurve=DataInterface() {propertyValues {Curve {keys [
         {interpMode=RCIM_Cubic, tangentMode=RCTM_User, time=0.0045, value=0.9397}
         {arriveTangent=-1.0036, interpMode=RCIM_Cubic, leaveTangent=-1.0036, tangentMode=RCTM_User, time=0.4647, value=0.5233}
         {arriveTangent=-2.4196, interpMode=RCIM_Cubic, leaveTangent=-2.4196, tangentMode=RCTM_User, time=1.0011, value=0.002}
        ]}}})
    Scale RGB=&VectorFromCurve(VectorCurve=DataInterface() {propertyValues {XCurve {keys [<at -0.0011 0.25>,<at 0.5 0>]}, YCurve {keys [<at -0.0011 0.25>,<at 0.5 0>]}, ZCurve {keys [<at -0.0011 0.25>,<at 0.5 0>]}}})
    ScaleA=true
    ScaleRGB=true
   )
  ]
  rendererClasses [&NiagaraMeshRendererProperties]
  renderers [{rendererClass=&NiagaraMeshRendererProperties, rendererIndex=0}]
 }
 Grid2D_Gas_SmokeFire {
  simTarget=GPUComputeSim
  emitterSpawnScript [
   /SetVariables(Emitter.ResX=512, Emitter.ResY=512) {isSetParameters=true}
   /SetVariables(Emitter.SimGrid=DataInterface() {propertyValues {
      ClearBeforeNonIterationStage=true
      NumAttributes=0
      NumCellsMaxAxis=3
      NumCellsX=1
      NumCellsY=1
      OverrideBufferFormat=Float
      PreviewAttribute=None
      SetGridFromMaxAxis=false
      bOverrideFormat=false
      bPreviewGrid=false
      RenderTargetUserParameter {parameter {name=None, typeDefHandle {registeredTypeIndex=1}, varData []}}
      WorldBBoxSize=<1200 1200>
     }}) {isSetParameters=true}
   /SetVariables(Emitter.LightingGrid=DataInterface() {propertyValues {
      ClearBeforeNonIterationStage=true
      NumAttributes=3
      NumCellsMaxAxis=3
      NumCellsX=512
      NumCellsY=512
      OverrideBufferFormat=Float
      PreviewAttribute=None
      SetGridFromMaxAxis=false
      bOverrideFormat=false
      bPreviewGrid=false
      RenderTargetUserParameter {parameter {name=None, typeDefHandle {registeredTypeIndex=4}, varData []}}
      WorldBBoxSize=<100 100>
     }}) {isSetParameters=true}
   /SetVariables(Emitter.LightingGridDownsample=DataInterface() {propertyValues {
      ClearBeforeNonIterationStage=true
      NumAttributes=1
      NumCellsMaxAxis=3
      NumCellsX=64
      NumCellsY=64
      OverrideBufferFormat=Float
      PreviewAttribute=None
      SetGridFromMaxAxis=false
      bOverrideFormat=false
      bPreviewGrid=false
      RenderTargetUserParameter {parameter {name=None, typeDefHandle {registeredTypeIndex=4}, varData []}}
      WorldBBoxSize=<100 100>
     }}) {isSetParameters=true}
   /SetVariables() {isSetParameters=true}
   /NiagaraFluids/Modules/Grid2D/Grid2D_SetResolution(Grid:NiagaraDataInterfaceGrid2DCollection=$Emitter.SimGrid, NumCellsX=512, NumCellsY=512, WorldGridExtents:Vector2f=<3000 3000>)
   &Grid2D_SetResolution(
    Grid:NiagaraDataInterfaceGrid2DCollection=$Emitter.PressureGrid
    OtherGrid:NiagaraDataInterfaceGrid2DCollection=$Emitter.SimGrid
    Resolution Method:ENiagaraGrid2DResolution=::NewEnumerator3
    WorldGridExtents:Vector2f=$Emitter.Grid2D_SetResolution.WorldGridExtents
   ) {module=Grid2D_SetResolution001}
   &Grid2D_SetResolution(
    Grid:NiagaraDataInterfaceGrid2DCollection=$Emitter.LightingGridDownsample
    OtherGrid:NiagaraDataInterfaceGrid2DCollection=$Emitter.SimGrid
    Resolution Method:ENiagaraGrid2DResolution=::NewEnumerator3
    Resolution Mult=0.2
    WorldGridExtents:Vector2f=$Emitter.Grid2D_SetResolution.WorldGridExtents
   ) {module=Grid2D_SetResolution002}
   &Grid2D_SetResolution(
    Grid:NiagaraDataInterfaceGrid2DCollection=$Emitter.LightingGrid
    OtherGrid:NiagaraDataInterfaceGrid2DCollection=$Emitter.SimGrid
    Resolution Method:ENiagaraGrid2DResolution=::NewEnumerator3
    WorldGridExtents:Vector2f=$Emitter.Grid2D_SetResolution.WorldGridExtents
   ) {module=Grid2D_SetResolution003}
   /SetVariables(Emitter.dt=0.0167) {isSetParameters=true}
   /SetVariables(Emitter.g:Vector2f=<0 -9.8>) {isSetParameters=true}
   /SetVariables() {isSetParameters=true}
   /SetVariables(Emitter.density=1) {isSetParameters=true}
   /SetVariables(Emitter.ParticleSource=DataInterface() {propertyValues {EmitterBinding {bindingMode=Other, emitterName=SourceParticle}}}) {isSetParameters=true}
   /SetVariables() {isSetParameters=true}
   &self:SetLightingRT_01(LightingGrid:NiagaraDataInterfaceGrid2DCollection=$Emitter.LightingGrid, LightingRT:NiagaraDataInterfaceRenderTarget2D=$Emitter.LightingRT) {module=SetLightingRT_01}
   /NiagaraFluids/Modules/Grid2D/Grid2D_UpdateSimAttributes(Grid:NiagaraDataInterfaceGrid2DCollection=$Emitter.SimGrid, WorldGridSize:Vector2f=$Emitter.Grid2D_SetResolution.WorldGridExtents) {module=UpdateSimAttributes}
   /SetVariables() {isSetParameters=true}
   /SetVariables(Emitter.testCurve=DataInterface() {propertyValues {
      bExposeCurve=true
      AlphaCurve {keys []}
      BlueCurve {keys [<at 0.3032 1>,<at 0.6671 0>]}
      GreenCurve {keys [<at 0.3032 1>,<at 0.6671 0.0785>]}
      RedCurve {keys [<at 0.3032 1>,<at 0.6671 1>]}
     }}) {isSetParameters=true}
   /NiagaraFluids/Functions/General/Fluids_Gas_Manager(
    Density Color Curve=DataInterface() {propertyValues {
      ExposedName=Curve Density
      bExposeCurve=true
      AlphaCurve {keys [<at 0 0 RCIM_Cubic>,<at 2.7618 1>]}
      BlueCurve {keys [<at 0.0245 0.045>,<at 4.4467 0.0059>]}
      GreenCurve {keys [<at 0.0245 0.0307>,<at 4.4467 0.0112>]}
      RedCurve {keys [<at 0.0245 0.0266>,<at 4.4467 0.0208>]}
     }}
    Depth Fade=250
    Dissipation Density=0.2
    Dissipation Temperature=0.5
    Dissipation Velocity=1
    EnvironmentColor:LinearColor=<rgba 0.1427 0.2031 0.3542 1>
    Pressure Relaxation=1
    Shadowing=0.25
    SunColor:LinearColor=<rgba 1.2417 1.53 2 1>
    Temperature Bouyancy=-2
    Temperature Color Curve=DataInterface() {propertyValues {
      ExposedName=Curve Temperature
      bExposeCurve=true
      AlphaCurve {keys [<at 0 0>,<at 0.3256 1>]}
      BlueCurve {keys [<at 0.0696 0>,<at 0.6238 0>,<at 0.9917 0.0148>]}
      GreenCurve {keys [<at 0.0696 0>,<at 0.6238 0.1697>,<at 0.9917 1.5503>]}
      RedCurve {keys [<at 0.0696 0>,<at 0.6238 2>,<at 0.9917 3>]}
     }}
    Vorticity Confinement=0.1
   )
   /NiagaraFluids/Modules/Grid2D/Grid2D_InitRT(Grid:NiagaraDataInterfaceGrid2DCollection=$Emitter.SimGrid, SimRT:NiagaraDataInterfaceRenderTarget2D=$Emitter.SimRT) {module=Grid2D_Gas_InitRT}
   &self:Grid2D_Gas_SmokeFire_0.NiagaraScratchPadContainer_2.ScratchModule_0(Grid:NiagaraDataInterfaceGrid2DCollection=$Emitter.SimGrid, PressureGrid:NiagaraDataInterfaceGrid2DCollection=$Emitter.PressureGrid) {module=ScratchModule}
   /SetVariables(Emitter.GridsAllocated=true) {isSetParameters=true}
  ]
  emitterUpdateScript [
   &self:Emitter_Limitation_and_Culling__01(Max Num=3, Platform Set=DataInterface() {propertyValues {Platforms {qualityLevelMask=-17, cVarConditions [], deviceProfileStates []}}}) {module=Emitter_Limitation_and_Culling__01}
   &EmitterState(Life Cycle Mode=::NewEnumerator1, Loop Behavior=::NewEnumerator1, Scalability Mode=::NewEnumerator1, UseLoopDelay=false)
   /SetVariables(Emitter.dt:NiagaraFloat=$Engine.DeltaTime) {isSetParameters=true}
   /NiagaraFluids/Modules/Grid2D/Grid2D_CreateUnitToWorldTransform(
    Camera:NiagaraDataInterfaceCamera=$Emitter.CameraQuery
    Camera Facing=true
    Local Pivot:Vector2f=<0.5 0.15>
    Offset:NiagaraPosition=$Engine.Owner.Position
    Rotation:Quat4f=$Engine.Owner.Rotation
    WorldGridExtents:Vector2f=$Emitter.Grid2D_SetResolution.WorldGridExtents
   )
   /SetVariables(Emitter.Position:NiagaraPosition=&ConvertVectorToPosition(
     Input Position=$Emitter.Grid2D_CreateUnitToWorldTransform.LocalToWorldTranslation
     Conversion Operation:ENiagara_LWCConvertVecToPos=::NewEnumerator0
    )) {isSetParameters=true}
   /NiagaraFluids/Modules/Grid2D/Grid2D_SetBounds(
    CameraFacing:NiagaraBool=$Emitter.Grid2D_CreateUnitToWorldTransform.CameraFacing
    EmitterProperties=DataInterface() {propertyValues {EmitterBinding {bindingMode=Self, emitterName=None}}}
    LocalPivot:Vector2f=$Emitter.Grid2D_CreateUnitToWorldTransform.LocalPivot
    Rotate for Camera Facing=true
    WorldGridExtents:Vector2f=$Emitter.Grid2D_SetResolution.WorldGridExtents
   )
   /SetVariables(Emitter.ExecutionState:ENiagaraExecutionState=$Emitter.ENiagaraExecutionState) {isSetParameters=true}
   &self:ScratchModule_03(PerformDeallocation:NiagaraBool=/Niagara/DynamicInputs/Scalability/IsEmitterComplete()) {module=ScratchModule_03}
  ]
  rendererClasses [&NiagaraSpriteRendererProperties]
  renderers [
   {rendererClass=&NiagaraSpriteRendererProperties}
   {rendererClass=&NiagaraSpriteRendererProperties, rendererIndex=1}
  ]
 }
 NewLIght {
  emitterUpdateScript [
   &EmitterScalability(
    Module.bApplySpawnCountScale=false
    Module.bApplySpawnCountScaleByDistanceFraction=false
    Module.bUseMaxDistance=false
    Module.bUseMinDistance=false
   ) {enabled=false}
   &EmitterLifeCycle(NextLoopDuration=4.4208) {enabled=false}
   &EmitterState()
   &SpawnBurst_Instantaneous(Spawn Count=1, Spawn Time=0.05, Use Spawn Probability=false)
   &SpawnRate(SpawnRate=15, Use Spawn Probability=false) {enabled=false}
  ]
  particleSpawnScript [
   &InitializeParticle(
    Lifetime=0.7
    Mesh Scale=<5 5 5>
    Module.Write Color=true
    Module.Write Lifetime=true
    Module.Write Mass=true
    Module.Write Material Random=true
    Module.Write Position=false
    Module.Write Scale=true
    Module.Write SpriteRotation=false
    Module.Write SpriteSize=true
    Position=&SimulationPosition()
   )
   &SphereLocation(MaskSpawnGroup=false, Module.Override Randomness=false, Offset=<20 0 0>, Sphere Radius=0.01, Surface Only=false) {enabled=false}
   /SetVariables(
    Particles.NewLightRadious=250
    Particles.NewLinearColor:LinearColor=<rgba 25 5 0 1>
    Particles.NewVector:Vector3f=&Add_Vector(A=/Niagara/DynamicInputs/Helpers/SimulationPosition(), B=<0 0 500>)
   ) {isSetParameters=true}
   &SystemLocation(MaskSpawnGroup=false, Offset=<0 0 100>)
  ]
  particleUpdateScript [
   &ParticleState()
   &Color(Color=<rgba 1 0.4584 0.0999 1>)
   &ScaleMeshSize(Scale Factor=&VectorFromCurve(VectorCurve=DataInterface() {propertyValues {
       XCurve {keys [
         {arriveTangent=9.7102, interpMode=RCIM_Cubic, leaveTangent=9.7102, tangentMode=RCTM_User, time=-0.0005, value=3.4203}
         {arriveTangent=6.8631, interpMode=RCIM_Cubic, leaveTangent=6.8631, tangentMode=RCTM_User, time=0.2646, value=6.6389}
         <at 0.9996 7.8749>
         <at 1.61 7.8749>
        ]}
       YCurve {keys [
         {arriveTangent=25.6031, interpMode=RCIM_Cubic, leaveTangent=25.6031, tangentMode=RCTM_User, time=-0.0005, value=3.4203}
         {arriveTangent=6.8631, interpMode=RCIM_Cubic, leaveTangent=6.8631, tangentMode=RCTM_User, time=0.2633, value=6.6865}
         <at 0.9996 7.8749>
         <at 1.61 7.8749>
        ]}
       ZCurve {keys [
         {arriveTangent=25.8257, interpMode=RCIM_Cubic, leaveTangent=25.8257, tangentMode=RCTM_User, time=-0.0005, value=3.4203}
         {arriveTangent=6.8631, interpMode=RCIM_Cubic, leaveTangent=6.8631, tangentMode=RCTM_User, time=0.2666, value=6.7585}
         <at 0.9996 7.8749>
         <at 1.61 7.8749>
        ]}
      }})) {enabled=false}
   &ScaleColor(
    Scale Alpha=&FloatFromCurve(FloatCurve=DataInterface() {propertyValues {Curve {keys [
         {interpMode=RCIM_Cubic, tangentMode=RCTM_User, time=-0.0034, value=0.1203}
         {arriveTangent=-0.1606, interpMode=RCIM_Cubic, leaveTangent=-0.1606, tangentMode=RCTM_User, time=0.3345, value=0.0495}
         {interpMode=RCIM_Cubic, tangentMode=RCTM_User, time=0.9972, value=-0.0001}
        ]}}})
    Scale RGB=&ScaleFloatByCurve(
     InitialValue:Vector3f=<1 0.4 0.1>
     ScaleCurve:NiagaraDataInterfaceVectorCurve=DataInterface() {propertyValues {XCurve {keys [<at 0 1>,<at 1 0>]}, YCurve {keys [<at 0 1>,<at 1.0017 0>]}, ZCurve {keys [<at 0 1>,<at 1.0017 0>]}}}
     CurveIndex:NiagaraFloat=$Particles.NormalizedAge
    )
    ScaleA=true
    ScaleRGB=true
   )
  ]
  rendererClasses [&NiagaraComponentRendererProperties]
  renderers [{rendererClass=&NiagaraComponentRendererProperties, rendererIndex=0}]
 }
 S_Combo {
  emitterUpdateScript [
   &EmitterState(Life Cycle Mode=::NewEnumerator1, Loop Behavior=::NewEnumerator1, Loop Delay=0.01, Loop Duration=0.4, UseLoopDelay=true)
   &SpawnBurst_Instantaneous(Spawn Count=1, Use Spawn Probability=false)
   &SpawnBurst_Instantaneous(Spawn Count=1, Spawn Group=1, Use Spawn Probability=false) {module=SpawnBurst_Instantaneous001}
  ]
  particleSpawnScript [
   &InitializeParticle(
    Color=<rgba 12 2.087 0.6504 1>
    Lifetime=&Lerp_Float(A=0.12, B=0.13, Alpha=&MakeFloatFromInt(Int=&SpawnGroup()))
    Mesh Scale=&VectorFromFloat(Value=1.1)
    Module.Write Color=true
    Module.Write Lifetime=true
    Module.Write Mass=true
    Module.Write Material Random=true
    Module.Write Position=true
    Module.Write Scale=true
    Module.Write SpriteRotation=false
    Module.Write SpriteSize=false
    Position=&AddVectorToPosition(Position=$Engine.Emitter.SimulationPosition, Vector=$User.ExplosionSource)
    Sprite Size=<5 5>
   )
   &InitialMeshOrientation(Orientation Vector=<0 0 1>, Use Orientation Vector=true, Use Rotation Vector=false) {module=InitialMeshOrientation001}
   &DynamicMaterialParameters(
    Index 0 Param 1=&UniformRangedFloat(Minimum=0, Maximum=20)
    Module.Param0WriteEnabled=true
    Module.Param1WriteEnabled=false
    Module.Param2WriteEnabled=false
    Module.Param3WriteEnabled=false
   )
  ]
  particleUpdateScript [
   &ParticleState()
   &SolveForcesAndVelocity(Clamp Velocity=false, Limit Acceleration=false)
   &ScaleMeshSize(Scale Factor=&VectorFromFloat(Value=&FloatFromCurve(
      FloatCurve=DataInterface() {propertyValues {Curve {keys [
          {arriveTangent=9.901, interpMode=RCIM_Cubic, leaveTangent=9.901, tangentMode=RCTM_User, time=0, value=0.4}
          <at 0.3174 2.7277>
          <at 1 0 RCIM_Cubic>
         ]}}}
      Scale Curve=2.6
     )))
   &ScaleColor(
    Scale Alpha=&FloatFromCurve(FloatCurve=DataInterface() {propertyValues {Curve {keys [<at 0.4 1 RCIM_Cubic>,<at 1 0>]}}})
    ScaleA=true
    ScaleRGB=true
   )
  ]
  rendererClasses [&NiagaraMeshRendererProperties]
  renderers [{rendererClass=&NiagaraMeshRendererProperties, rendererIndex=0}]
 }
 Shockwave {
  emitterUpdateScript [
   &EmitterState(
    Enable Distance Culling=true
    Life Cycle Mode=::NewEnumerator1
    Loop Behavior=::NewEnumerator1
    Loop Duration=0.4
    MaxDistance=20000
    Module.bUseMaxDistance=true
    Module.bUseMinDistance=false
    Scalability Mode=::NewEnumerator1
    UseLoopDelay=false
   )
   &SpawnBurst_Instantaneous(Spawn Count=1, Use Spawn Probability=false)
  ]
  particleSpawnScript [&InitializeParticle(
    Color=<rgba 350 189.8751 0 0.5>
    Lifetime=0.2
    Mesh Scale=&VectorFromFloat(Value=5)
    Module.Write Color=true
    Module.Write Lifetime=true
    Module.Write Mass=true
    Module.Write Material Random=true
    Module.Write Position=true
    Module.Write Scale=true
    Module.Write SpriteRotation=false
    Module.Write SpriteSize=true
    Position=&AddVectorToPosition(Position=$Engine.Emitter.SimulationPosition, Vector=$User.ExplosionSource)
    Sprite Size=<5 5>
   )]
  particleUpdateScript [
   &ParticleState()
   &SolveForcesAndVelocity(Clamp Velocity=false, Limit Acceleration=false)
   &ScaleColor(
    Scale Mode=::NewEnumerator1
    Scale RGBA=&Vector4FromCurve(Vector4Curve=DataInterface() {propertyValues {
       AlphaCurve {keys [<at 0 1>,<at 0.7084 0.1>,<at 1 0>]}
       BlueCurve {keys [<at 0 0.0111>,<at 1 1>]}
       GreenCurve {keys [<at 0 0.4>,<at 1 1>]}
       RedCurve {keys [<at 0 1>,<at 1 1>]}
      }})
    ScaleRGBA=true
   )
   &ScaleMeshSize(Scale Factor=&VectorFromFloat(Value=&FloatFromCurve(FloatCurve=DataInterface() {propertyValues {Curve {keys [<at 0 0 RCIM_Cubic>, {arriveTangent=1.8378, interpMode=RCIM_Cubic, leaveTangent=1.8378, time=0.305, value=1.3181}, <at 0.9889 1.8174>]}}})))
  ]
  rendererClasses [&NiagaraMeshRendererProperties]
  renderers [{rendererClass=&NiagaraMeshRendererProperties, rendererIndex=0}]
 }
 Smoke_Ring {
  emitterUpdateScript [
   &EmitterState(Life Cycle Mode=::NewEnumerator1, Loop Behavior=::NewEnumerator1, Loop Delay=0.01, Loop Duration=0.125, UseLoopDelay=true)
   &SpawnBurst_Instantaneous(Spawn Count=30, Spawn Time=0.03, Use Spawn Probability=false)
  ]
  particleSpawnScript [
   &InitializeParticle(
    Color=&UniformRangedLinearColor(Minimum=<rgba 0.0573 0.0474 0.041 1>, Maximum=<rgba 0.0729 0.0547 0.0396 1>)
    Lifetime=&UniformRangedFloat(Minimum=0.5, Maximum=1.7)
    Module.Write Color=true
    Module.Write Lifetime=true
    Module.Write Mass=true
    Module.Write Material Random=true
    Module.Write Position=true
    Module.Write Scale=false
    Module.Write SpriteRotation=true
    Module.Write SpriteSize=true
    Sprite Rotation=&NormalizedAngleToDegrees(NormalizedAngle=&UniformRangedFloat(Minimum=0, Maximum=360))
    Sprite Size=&Vector2DFromFloat(Value=&RandomRangeFloat(Minimum=320, Maximum=480))
   )
   /SetVariables(Particles.InitSpriteRotRate=&UniformRangedFloat(Minimum=-3, Maximum=3)) {isSetParameters=true, module=SetVariables_#4}
   &TorusLocation(
    Align Torus With Axis=true
    Large Radius=15
    MaskSpawnGroup=false
    Torus Distribution Mode:ENiagaraTorusDistributionMode=::NewEnumerator1
    Torus Mode=::NewEnumerator2
    U Position:NiagaraFloat=&ReturnNormalizedExecIndex(Normalized Index Includes One=false)
   )
   &AddVelocityFromPoint(
    Module.Use Falloff Distance=false
    Origin Offset Coordinate Space=::Simulation
    Velocity Strength=&RandomRangeFloat(Minimum=2000, Maximum=4500)
   )
   &AddVelocity(Velocity=&MakeVector(X=0, Y=0, Z=&RandomRangeFloat(Minimum=-150, Maximum=500)))
  ]
  particleUpdateScript [
   &ParticleState()
   &SpriteRotationRate(Rotation Rate=&Multiply_Float(
     A=$Particles.InitSpriteRotRate
     B=&FloatFromCurve(
      FloatCurve=DataInterface() {propertyValues {Curve {keys [
          {arriveTangent=-4.0107, interpMode=RCIM_Cubic, leaveTangent=-4.0107, tangentMode=RCTM_User, time=-0.0046, value=1.0113}
          {arriveTangent=-0.2496, interpMode=RCIM_Cubic, leaveTangent=-0.2496, tangentMode=RCTM_User, time=0.5, value=0.16}
         ]}}}
      CurveIndex=$Emitter.Age
      Scale Curve=150
     )
    ))
   /Niagara/Modules/Update/Forces/CurlNoiseForce(Noise Frequency=0.0012, Noise Strength=650, Pan Noise=true, Pan Noise Field=<0 0 0.25>)
   &ScaleColor(
    Scale Alpha=&FloatFromCurve(FloatCurve=DataInterface() {propertyValues {Curve {keys [<at 0.1219 4>,<at 0.194 1.5>,<at 0.4882 1.5>,<at 1 0>]}}})
    Scale RGB=&VectorFromFloat(Value=&FloatFromCurve(FloatCurve=DataInterface() {propertyValues {Curve {keys [{arriveTangent=1.8145, interpMode=RCIM_Cubic, leaveTangent=4.5795, tangentMode=RCTM_Break, time=0, value=0.4}, <at 0.3387 0.9565>, <at 1 1>]}}}))
    ScaleA=true
    ScaleRGB=true
   )
   &Drag(
    Drag=&FloatFromCurve(
     FloatCurve=DataInterface() {propertyValues {Curve {keys [
         {interpMode=RCIM_Cubic, leaveTangent=0.0927, tangentMode=RCTM_Break, time=0, value=1}
         {arriveTangent=-2.586, interpMode=RCIM_Cubic, leaveTangent=-2.586, tangentMode=RCTM_User, time=1, value=0}
        ]}}}
     Scale Curve=&Lerp_Float(A=7, B=9, Alpha=$Particles.Initial.MaterialRandom)
    )
    Use Linear Drag=true
    Use Rotational Drag=false
   )
   &AccelerationForce(Acceleration=<250 0 500>)
   &SolveForcesAndVelocity(Clamp Velocity=false, Limit Acceleration=false)
   &ScaleSpriteSize(Scale Factor=&Vector2DFromFloat(Value=&FloatFromCurve(FloatCurve=DataInterface() {propertyValues {Curve {keys [
          {arriveTangent=7.9207, interpMode=RCIM_Cubic, leaveTangent=5.2162, tangentMode=RCTM_User, time=0, value=0.15}
          {arriveTangent=1.0025, interpMode=RCIM_Cubic, leaveTangent=1.0025, tangentMode=RCTM_User, time=0.2012, value=0.7293}
          {arriveTangent=0.2754, interpMode=RCIM_Cubic, leaveTangent=0.2754, tangentMode=RCTM_User, time=1.0463, value=1}
         ]}}})))
   &DynamicMaterialParameters(
    Index 0 Param 1=&Add_Float(
     A=$Particles.RandomFloat
     B=&FloatFromCurve(
      FloatCurve=DataInterface() {propertyValues {Curve {keys [
          {arriveTangent=2.7533, interpMode=RCIM_Cubic, leaveTangent=1.8958, tangentMode=RCTM_User, time=0, value=1}
          {arriveTangent=0.8292, interpMode=RCIM_Cubic, leaveTangent=0.8292, tangentMode=RCTM_User, time=0.321, value=1.3453}
          {arriveTangent=0.0529, interpMode=RCIM_Cubic, leaveTangent=0.2021, tangentMode=RCTM_User, time=1, value=1.5}
         ]}}}
      CurveIndex=&Add_Float(A=&Multiply_Float(A=$Particles.Initial.MaterialRandom, B=0.5), B=$Particles.NormalizedAge)
     )
    )
    Index 0 Param 2=&FloatFromCurve(FloatCurve=DataInterface() {propertyValues {Curve {keys [
         {arriveTangent=0.0453, interpMode=RCIM_Cubic, tangentMode=RCTM_Break, time=0, value=-1}
         {arriveTangent=0.0604, interpMode=RCIM_Cubic, leaveTangent=0.0604, tangentMode=RCTM_User, time=1, value=-1}
        ]}}})
    Index 0 Param 3=&FloatFromCurve(FloatCurve=DataInterface() {propertyValues {Curve {keys [
         {arriveTangent=-14.1037, interpMode=RCIM_Cubic, leaveTangent=-5.4872, tangentMode=RCTM_User, time=0, value=1}
         {arriveTangent=-1.8087, interpMode=RCIM_Cubic, leaveTangent=-1.8087, tangentMode=RCTM_User, time=0.0464, value=0.214}
         {arriveTangent=0.2347, interpMode=RCIM_Cubic, leaveTangent=0.2723, tangentMode=RCTM_Break, time=0.4, value=0}
        ]}}})
    Index 0 Param 4=&FloatFromCurve(FloatCurve=DataInterface() {propertyValues {Curve {keys [<at 0.5 1>,<at 1 0>]}}})
    Index 1 Param 1=&FloatFromCurve(
     FloatCurve=DataInterface() {propertyValues {Curve {keys [
         {interpMode=RCIM_Cubic, leaveTangent=4.7088, tangentMode=RCTM_Break, time=0, value=0}
         {arriveTangent=4.2257, interpMode=RCIM_Cubic, leaveTangent=4.2257, tangentMode=RCTM_User, time=0.25, value=1}
        ]}}}
     Scale Curve=&RandomRangeFloat(Minimum=1.5, Maximum=1.5)
    )
    Index 1 Param 2=&FloatFromCurve(
     FloatCurve=DataInterface() {propertyValues {Curve {keys [<at 0 0.2>, {leaveTangent=0.9066, time=1, value=2}]}}}
     Scale Curve=&RandomRangeFloat(Minimum=0.3, Maximum=0.4)
    )
    Index 1 Param 3=&FloatFromCurve(FloatCurve=DataInterface() {propertyValues {Curve {keys [
         {interpMode=RCIM_Cubic, leaveTangent=6.3917, tangentMode=RCTM_Break, time=0, value=0.2}
         {arriveTangent=4.8433, interpMode=RCIM_Cubic, leaveTangent=19.0929, tangentMode=RCTM_User, time=0.25, value=1}
        ]}}})
    Index 1 Param 4=&FloatFromCurve(
     FloatCurve=DataInterface() {propertyValues {Curve {keys [{arriveTangent=-0.3224, interpMode=RCIM_Cubic, leaveTangent=-0.3224, tangentMode=RCTM_User, time=0, value=1}, <at 1 0.9>]}}}
     Scale Curve=&RandomRangeFloat(Minimum=1.5, Maximum=2)
    )
    Index 2 Param 1=&FloatFromCurve(FloatCurve=DataInterface() {propertyValues {Curve {keys [
         {interpMode=RCIM_Cubic, leaveTangent=4.2485, tangentMode=RCTM_Break, time=0, value=0}
         {arriveTangent=1.0548, interpMode=RCIM_Cubic, leaveTangent=1.0548, tangentMode=RCTM_User, time=0.17, value=0.5715}
         <at 1 1>
        ]}}})
    Index 2 Param 2=&FloatFromCurve(
     FloatCurve=DataInterface() {propertyValues {Curve {keys [
         {arriveTangent=1.9056, interpMode=RCIM_Cubic, leaveTangent=1.9056, tangentMode=RCTM_User, time=0, value=0.2}
         {arriveTangent=0.0259, interpMode=RCIM_Cubic, leaveTangent=0.0259, tangentMode=RCTM_User, time=1, value=1.4}
        ]}}}
     Scale Curve=&RandomRangeFloat(Minimum=0.2, Maximum=0.7)
    )
    Index 2 Param 3=&FloatFromCurve(FloatCurve=DataInterface() {propertyValues {Curve {keys [
         {interpMode=RCIM_Cubic, leaveTangent=2.3064, tangentMode=RCTM_Break, time=0, value=1}
         {interpMode=RCIM_Cubic, leaveTangent=11.1233, tangentMode=RCTM_Break, time=1, value=2}
        ]}}})
    Index 2 Param 4=&FloatFromCurve(
     FloatCurve=DataInterface() {propertyValues {Curve {keys [
         {arriveTangent=-0.6867, interpMode=RCIM_Cubic, leaveTangent=-0.6867, tangentMode=RCTM_User, time=0, value=1}
         {interpMode=RCIM_Cubic, leaveTangent=-1.2956, tangentMode=RCTM_Break, time=1, value=0.75}
        ]}}}
     Scale Curve=1.6
    )
    Module.Param0WriteEnabled=true
    Module.Param0WriteEnabled1=true
    Module.Param0WriteEnabled2=true
    Module.Param1WriteEnabled=false
    Module.Param1WriteEnabled1=true
    Module.Param1WriteEnabled2=true
    Module.Param2WriteEnabled=true
    Module.Param2WriteEnabled1=true
    Module.Param2WriteEnabled2=true
    Module.Param3WriteEnabled=true
    Module.Param3WriteEnabled1=true
    Module.Param3WriteEnabled2=true
    Write Parameter Index 1=true
    Write Parameter Index 2=true
   )
  ]
  rendererClasses [&NiagaraSpriteRendererProperties]
  renderers [
   {rendererClass=&NiagaraSpriteRendererProperties}
   {rendererClass=&NiagaraSpriteRendererProperties, rendererIndex=1}
  ]
 }
 Smoke_Sphere {
  emitterUpdateScript [
   &EmitterState(Life Cycle Mode=::NewEnumerator1, Loop Behavior=::NewEnumerator1, Loop Delay=0.05, Loop Duration=0.1, UseLoopDelay=true)
   &SpawnBurst_Instantaneous(Spawn Group=1, Use Spawn Probability=false, UseLoopCountLimit=false) {module=SpawnBurst_Instantaneous001}
   &SpawnBurst_Instantaneous(Spawn Count=25, Use Spawn Probability=false)
  ]
  particleSpawnScript [
   &InitializeParticle(
    Color=&UniformRangedLinearColor(Minimum=<rgba 0.0573 0.0474 0.0428 1>, Maximum=<rgba 0.0729 0.0613 0.0491 1>)
    Lifetime=&UniformRangedFloat(Minimum=0.5, Maximum=2)
    Module.Write Color=true
    Module.Write Lifetime=true
    Module.Write Mass=true
    Module.Write Material Random=true
    Module.Write Position=true
    Module.Write Scale=false
    Module.Write SpriteRotation=true
    Module.Write SpriteSize=true
    Position=&AddVectorToPosition(Position=$Engine.Emitter.SimulationPosition, Vector=$User.ExplosionSource)
    Sprite Rotation=&NormalizedAngleToDegrees(NormalizedAngle=&UniformRangedFloat(Minimum=0, Maximum=360))
    Sprite Size=&Vector2DFromFloat(Value=&MaskFloatBySpawnGroup(Masked Float=750, Passthrough Float=&RandomRangeFloat(Minimum=550, Maximum=700), Spawn Group Mask=1))
   )
   /SetVariables(Particles.InitSpriteRotRate=&UniformRangedFloat(Minimum=-3, Maximum=3)) {isSetParameters=true, module=SetVariables_#4}
   /Niagara/Modules/Spawn/Location/V2/SphereLocation(
    MaskSpawnGroup=false
    Offset=<0 0 150>
    Override Seed=true
    Random Seed=20
    Sphere Radius=&MakeCustomFloatFromBool(Bool=true, True Float=&Multiply_Float(A=$User.ExplosionRadius, B=0.5), False Float=260)
    Surface Only=true
    UseNonUniformScale=false
    UseOffset=false
    UseSphereOrientation=false
   )
   /SetVariables(Particles.Position=&MaskVectorBySpawnGroup(
     Masked Vector=&Add_Vector(A=Linked() {linkedVariable {name=Engine.Owner.Position, type=NiagaraPosition}}, B=$User.ExplosionSource)
     Passthrough Vector=Linked() {linkedVariable {name=Particles.Position, type=NiagaraPosition}}
     Spawn Group Mask=1
    )) {isSetParameters=true}
   &AddVelocityFromPoint(Module.Use Falloff Distance=false, Velocity Strength=&RandomRangeFloat(Minimum=800, Maximum=2500)) {enabled=false}
   &AddVelocity() {enabled=false}
   &AddVelocity(
    Velocity=$Output.SphereLocation.SphereNormal
    Velocity Speed Scale=&RandomRangeFloat(Minimum=100, Maximum=500)
   ) {module=AddVelocity001}
  ]
  particleUpdateScript [
   &ParticleState()
   &SpriteRotationRate(Rotation Rate=&Multiply_Float(
     A=$Particles.InitSpriteRotRate
     B=&FloatFromCurve(
      FloatCurve=DataInterface() {propertyValues {Curve {keys [
          {arriveTangent=-4.0107, interpMode=RCIM_Cubic, leaveTangent=-4.0107, tangentMode=RCTM_User, time=-0.0046, value=1.0113}
          {arriveTangent=-0.2496, interpMode=RCIM_Cubic, leaveTangent=-0.2496, tangentMode=RCTM_User, time=0.5, value=0.16}
         ]}}}
      CurveIndex=$Emitter.Age
      Scale Curve=150
     )
    ))
   &ScaleColor(
    Scale Alpha=&FloatFromCurve(FloatCurve=DataInterface() {propertyValues {Curve {keys [<at 0.1165 1.8827>,<at 0.194 1.5>,<at 0.7612 1.5>,<at 1 0>]}}})
    Scale RGB=&VectorFromFloat(Value=&FloatFromCurve(FloatCurve=DataInterface() {propertyValues {Curve {keys [{arriveTangent=1.8145, interpMode=RCIM_Cubic, leaveTangent=4.5795, tangentMode=RCTM_Break, time=0, value=0.4}, <at 0.3387 0.9565>, <at 1 1>]}}}))
    ScaleA=true
    ScaleRGB=true
   )
   &CurlNoiseForce(Noise Frequency=0.0012, Noise Strength=800, Pan Noise=true, Pan Noise Field=<0 0 1>)
   &Drag(Drag=15, Use Linear Drag=true, Use Rotational Drag=false)
   &AccelerationForce(Acceleration=<250 0 1500>)
   &SolveForcesAndVelocity(Clamp Velocity=false, Limit Acceleration=false)
   &ScaleSpriteSize(Scale Factor=&Vector2DFromFloat(Value=&FloatFromCurve(FloatCurve=DataInterface() {propertyValues {Curve {keys [
          {arriveTangent=5.2865, interpMode=RCIM_Cubic, leaveTangent=3.7568, tangentMode=RCTM_User, time=0.0053, value=0.3786}
          {arriveTangent=1.0025, interpMode=RCIM_Cubic, leaveTangent=1.0025, tangentMode=RCTM_User, time=0.166, value=0.7682}
          {arriveTangent=0.2754, interpMode=RCIM_Cubic, leaveTangent=0.2754, tangentMode=RCTM_User, time=1.0463, value=1}
         ]}}})))
   &DynamicMaterialParameters(
    Index 0 Param 1=&Add_Float(
     A=$Particles.RandomFloat
     B=&FloatFromCurve(
      FloatCurve=DataInterface() {propertyValues {Curve {keys [
          {arriveTangent=2.7533, interpMode=RCIM_Cubic, leaveTangent=1.8958, tangentMode=RCTM_User, time=0, value=1}
          {arriveTangent=0.8292, interpMode=RCIM_Cubic, leaveTangent=0.8292, tangentMode=RCTM_User, time=0.321, value=1.3453}
          {arriveTangent=0.0529, interpMode=RCIM_Cubic, leaveTangent=0.2021, tangentMode=RCTM_User, time=1, value=1.5}
         ]}}}
      CurveIndex=&Add_Float(A=&Multiply_Float(A=$Particles.Initial.MaterialRandom, B=0.5), B=$Particles.NormalizedAge)
     )
    )
    Index 0 Param 2=&FloatFromCurve(FloatCurve=DataInterface() {propertyValues {Curve {keys [
         {arriveTangent=0.0453, interpMode=RCIM_Cubic, tangentMode=RCTM_Break, time=0, value=-1}
         {arriveTangent=0.0604, interpMode=RCIM_Cubic, leaveTangent=0.0604, tangentMode=RCTM_User, time=1, value=-1}
        ]}}})
    Index 0 Param 3=&FloatFromCurve(FloatCurve=DataInterface() {propertyValues {Curve {keys [
         {arriveTangent=-53.9199, interpMode=RCIM_Cubic, leaveTangent=-8.9727, tangentMode=RCTM_User, time=0, value=2}
         {arriveTangent=-1.6741, interpMode=RCIM_Cubic, leaveTangent=-1.6741, tangentMode=RCTM_User, time=0.0514, value=0.1865}
         {arriveTangent=0.2347, interpMode=RCIM_Cubic, leaveTangent=0.2723, tangentMode=RCTM_Break, time=0.4, value=0}
        ]}}})
    Index 0 Param 4=&FloatFromCurve(FloatCurve=DataInterface() {propertyValues {Curve {keys [<at 0.5 1>,<at 1 0>]}}})
    Index 1 Param 1=&FloatFromCurve(
     FloatCurve=DataInterface() {propertyValues {Curve {keys [
         {interpMode=RCIM_Cubic, leaveTangent=4.7088, tangentMode=RCTM_Break, time=0, value=0}
         {arriveTangent=4.2257, interpMode=RCIM_Cubic, leaveTangent=4.2257, tangentMode=RCTM_User, time=0.25, value=1}
        ]}}}
     Scale Curve=&RandomRangeFloat(Minimum=1.2, Maximum=2)
    )
    Index 1 Param 2=&FloatFromCurve(
     FloatCurve=DataInterface() {propertyValues {Curve {keys [<at 0 0.2>, {leaveTangent=0.9066, time=1, value=2}]}}}
     Scale Curve=&RandomRangeFloat(Minimum=0.1, Maximum=0.2)
    )
    Index 1 Param 3=&FloatFromCurve(FloatCurve=DataInterface() {propertyValues {Curve {keys [
         {interpMode=RCIM_Cubic, leaveTangent=6.3917, tangentMode=RCTM_Break, time=0, value=0.2}
         {arriveTangent=1.7549, interpMode=RCIM_Cubic, leaveTangent=10.6767, tangentMode=RCTM_User, time=0.25, value=1}
        ]}}})
    Index 1 Param 4=&FloatFromCurve(
     FloatCurve=DataInterface() {propertyValues {Curve {keys [{arriveTangent=-0.136, interpMode=RCIM_Cubic, leaveTangent=-0.136, tangentMode=RCTM_User, time=0, value=1}, <at 1 0.95>]}}}
     Scale Curve=&RandomRangeFloat(Minimum=1.5, Maximum=3)
    )
    Index 2 Param 1=&FloatFromCurve(FloatCurve=DataInterface() {propertyValues {Curve {keys [
         {interpMode=RCIM_Cubic, leaveTangent=4.2485, tangentMode=RCTM_Break, time=0, value=0}
         {arriveTangent=1.0548, interpMode=RCIM_Cubic, leaveTangent=1.0548, tangentMode=RCTM_User, time=0.17, value=0.5715}
         <at 1 1>
        ]}}})
    Index 2 Param 2=&FloatFromCurve(
     FloatCurve=DataInterface() {propertyValues {Curve {keys [
         {arriveTangent=1.9056, interpMode=RCIM_Cubic, leaveTangent=1.9056, tangentMode=RCTM_User, time=0, value=0.2}
         {arriveTangent=0.0259, interpMode=RCIM_Cubic, leaveTangent=0.0259, tangentMode=RCTM_User, time=1, value=1.4}
        ]}}}
     Scale Curve=&RandomRangeFloat(Minimum=0.2, Maximum=0.7)
    )
    Index 2 Param 3=&FloatFromCurve(FloatCurve=DataInterface() {propertyValues {Curve {keys [
         {interpMode=RCIM_Cubic, leaveTangent=2.3064, tangentMode=RCTM_Break, time=0, value=1}
         {interpMode=RCIM_Cubic, leaveTangent=11.1233, tangentMode=RCTM_Break, time=1, value=2}
        ]}}})
    Index 2 Param 4=&FloatFromCurve(
     FloatCurve=DataInterface() {propertyValues {Curve {keys [
         {arriveTangent=-0.6867, interpMode=RCIM_Cubic, leaveTangent=-0.6867, tangentMode=RCTM_User, time=0, value=1}
         {interpMode=RCIM_Cubic, leaveTangent=-1.2956, tangentMode=RCTM_Break, time=1, value=0.75}
        ]}}}
     Scale Curve=1.6
    )
    Module.Param0WriteEnabled=true
    Module.Param0WriteEnabled1=true
    Module.Param0WriteEnabled2=true
    Module.Param1WriteEnabled=false
    Module.Param1WriteEnabled1=true
    Module.Param1WriteEnabled2=true
    Module.Param2WriteEnabled=true
    Module.Param2WriteEnabled1=true
    Module.Param2WriteEnabled2=true
    Module.Param3WriteEnabled=true
    Module.Param3WriteEnabled1=true
    Module.Param3WriteEnabled2=true
    Write Parameter Index 1=true
    Write Parameter Index 2=true
   )
   &CameraOffset(Camera Offset Amount=35)
  ]
  rendererClasses [&NiagaraSpriteRendererProperties]
  renderers [
   {rendererClass=&NiagaraSpriteRendererProperties}
   {rendererClass=&NiagaraSpriteRendererProperties, rendererIndex=1}
  ]
 }
 SmokeyFingers001 {
  emitterUpdateScript [
   &EmitterState(Life Cycle Mode=::NewEnumerator1, Loop Behavior=::NewEnumerator1, Loop Delay=0.065, Loop Duration=0.15, UseLoopDelay=true)
   &SpawnRate(SpawnRate=50, Use Spawn Probability=false)
  ]
  particleSpawnScript [
   &InitializeParticle(
    Color=<rgba 0 1 1 0.85>
    Lifetime=&UniformRangedFloat(Minimum=0.2, Maximum=0.25)
    Mesh Scale=&Multiply_VectorByFloat(Vector=&UniformRangedVector(Minimum=<5 25 25>, Maximum=<8 25 25>), Float=$System.ScaleDistance)
    Module.Write Color=true
    Module.Write Lifetime=true
    Module.Write Mass=true
    Module.Write Material Random=true
    Module.Write Position=true
    Module.Write Scale=true
    Module.Write SpriteRotation=false
    Module.Write SpriteSize=false
    Position=&AddVectorToPosition(Position=$Engine.Emitter.SimulationPosition, Vector=$User.ExplosionSource)
    Sprite Size=<25 25>
   )
   &InitialMeshOrientation(
    Orientation Axis=<-1 0 0>
    Orientation Vector=<0 0 -1>
    Rotation=&UniformRangedVector(Minimum=<0 0.18 0.18>, Maximum=<0 -0.15 -0.18>)
    Use Orientation Vector=true
    Use Rotation Vector=true
   ) {enabled=false, module=InitialMeshOrientation001}
   &InitialMeshOrientation(Use Rotation Vector=false)
  ]
  particleUpdateScript [
   &GravityForce(Gravity=<0 0 -0.01>)
   &SolveForcesAndVelocity(Clamp Velocity=false, Limit Acceleration=false)
   &ParticleState()
   &ScaleColor(
    Scale Alpha=&FloatFromCurve(FloatCurve=DataInterface() {propertyValues {Curve {keys [
         {arriveTangent=200.5769, interpMode=RCIM_Cubic, leaveTangent=200.5752, tangentMode=RCTM_User, time=0, value=0}
         {arriveTangent=0.0668, interpMode=RCIM_Cubic, leaveTangent=0.0668, tangentMode=RCTM_User, time=0.01, value=1}
         <at 1 0>
        ]}}})
    Scale RGB=&VectorFromCurve(VectorCurve=DataInterface() {propertyValues {
       bUseLUT=false
       XCurve {keys [
         <at -0.0009 0.9906 RCIM_Cubic>
         {arriveTangent=-1.9223, interpMode=RCIM_Cubic, leaveTangent=-1.0067, tangentMode=RCTM_Break, time=0.3779, value=0.4}
         <at 1 0 RCIM_Cubic>
        ]}
       YCurve {keys [
         {arriveTangent=-3.6896, interpMode=RCIM_Cubic, leaveTangent=-3.6896, tangentMode=RCTM_User, time=-0.0011, value=1.0067}
         {arriveTangent=-0.3437, interpMode=RCIM_Cubic, leaveTangent=-0.3437, tangentMode=RCTM_User, time=0.3551, value=0.2247}
         <at 1 0 RCIM_Cubic>
        ]}
       ZCurve {keys [
         <at -0.0011 0 RCIM_Cubic>
         {arriveTangent=-0.223, interpMode=RCIM_Cubic, leaveTangent=-0.223, tangentMode=RCTM_User, time=0.9965, value=0.7415}
        ]}
      }})
    ScaleA=true
    ScaleRGB=true
   )
   &ScaleMeshSize(Scale Factor=&VectorFromFloat(Value=&FloatFromCurve(
      FloatCurve=DataInterface() {propertyValues {Curve {keys [
          {arriveTangent=4.641, interpMode=RCIM_Cubic, leaveTangent=4.641, tangentMode=RCTM_User, time=0, value=0}
          {arriveTangent=0.8503, interpMode=RCIM_Cubic, leaveTangent=0.8503, time=0.1368, value=0.6768}
          <at 1 0.8503 RCIM_Cubic>
         ]}}}
      Scale Curve=1.4
     )))
   &DynamicMaterialParameters(
    Index 0 Param 1=&FloatFromCurve(FloatCurve=DataInterface() {propertyValues {Curve {keys [
         {arriveTangent=3.1529, interpMode=RCIM_Cubic, leaveTangent=3.1529, tangentMode=RCTM_User, time=0, value=0.5}
         {arriveTangent=0.4946, interpMode=RCIM_Cubic, leaveTangent=0.4946, tangentMode=RCTM_User, time=0.1931, value=0.9256}
         <at 1.001 1.1573>
        ]}}})
    Index 0 Param 2=&FloatFromCurve(FloatCurve=DataInterface() {propertyValues {Curve {keys [<at 0.505 0.9841>,<at 0.7325 0.0093>,<at 1.0011 0.002>]}}})
    Index 0 Param 3=&FloatFromCurve(
     FloatCurve=DataInterface() {propertyValues {Curve {keys [
         {arriveTangent=3.7271, interpMode=RCIM_Cubic, leaveTangent=3.7271, tangentMode=RCTM_User, time=-0.002, value=1}
         {arriveTangent=2.1936, interpMode=RCIM_Cubic, leaveTangent=2.1936, time=0.4642, value=2.7053}
         <at 0.9979 3.1936>
        ]}}}
     Scale Curve=&Multiply_Float(A=$System.ScaleDistance, B=1.5)
    )
    Module.Param0WriteEnabled=true
    Module.Param1WriteEnabled=true
    Module.Param2WriteEnabled=true
    Module.Param3WriteEnabled=true
   )
   &CameraOffset(Camera Offset Amount=-25)
  ]
  rendererClasses [&NiagaraMeshRendererProperties]
  renderers [{rendererClass=&NiagaraMeshRendererProperties, rendererIndex=0}]
 }
 SourceParticle {
  simTarget=GPUComputeSim
  emitterSpawnScript [
   /SetVariables(Emitter.RasterizationGrid=DataInterface() {propertyValues {
      ClearBeforeNonIterationStage=true
      NumAttributes=4
      NumCellsMaxAxis=3
      NumCellsX=3
      NumCellsY=3
      OverrideBufferFormat=Float
      PreviewAttribute=None
      SetGridFromMaxAxis=false
      bOverrideFormat=false
      bPreviewGrid=false
      RenderTargetUserParameter {parameter {name=None, typeDefHandle {registeredTypeIndex=4}, varData []}}
      WorldBBoxSize=<100 100>
     }}) {isSetParameters=true}
   &Grid2D_SetResolution(Grid:NiagaraDataInterfaceGrid2DCollection=$Emitter.RasterizationGrid, NumCellsX=256, NumCellsY=256, WorldGridExtents:Vector2f=<1200 1200>)
   &Grid2D_UpdateSimAttributes(Grid:NiagaraDataInterfaceGrid2DCollection=$Emitter.RasterizationGrid, WorldGridSize:Vector2f=$Emitter.Grid2D_SetResolution.WorldGridExtents) {module=UpdateSimAttributes}
  ]
  emitterUpdateScript [
   /NiagaraFN/Modules/Scalability/SignificantInstancesCull(
    Cull Response:ENiagaraExecutionStateManagement=::KillImmediately
    First Frame Only=false
    Num Instances=3
    Platform Set=DataInterface() {propertyValues {Platforms {qualityLevelMask=-17, cVarConditions [], deviceProfileStates []}}}
   )
   &EmitterState(Life Cycle Mode=::NewEnumerator1, Loop Behavior=::NewEnumerator1, Loop Duration=0.2, Scalability Mode=::NewEnumerator1, UseLoopDelay=false)
   &SpawnBurst_Instantaneous(
    Spawn Count=&RandomRangeInt(Minimum=20, Maximum=10, Evaluation Type=::NewEnumerator1)
    Use Spawn Probability=false
   )
   &SpawnRate(SpawnRate=125, Use Spawn Probability=false) {enabled=false}
   &Grid2D_CreateUnitToWorldTransform(
    Camera Facing=true
    Local Pivot:Vector2f=<0.5 0.1>
    Offset:NiagaraPosition=$Engine.Owner.Position
    Rotation:Quat4f=$Engine.Owner.Rotation
    WorldGridExtents:Vector2f=$Emitter.Grid2D_SetResolution.WorldGridExtents
   )
  ]
  particleSpawnScript [
   &InitializeParticle(
    Color=<rgba 1111 13.6009 0 1>
    Lifetime=&RandomRangeFloat(Minimum=0.5, Maximum=2)
    Mass=&RandomRangeFloat(Minimum=0.75, Maximum=1)
    Module.Write Color=true
    Module.Write Lifetime=true
    Module.Write Mass=true
    Module.Write Material Random=true
    Module.Write Position=false
    Module.Write Scale=false
    Module.Write SpriteRotation=false
    Module.Write SpriteSize=true
    Sprite Size=&Vector2DFromFloat(Value=&RandomRangeFloat(Minimum=500, Maximum=500))
   )
   &AddVelocityInCone(
    Cone Angle=&Cosine(Normalized Angle=$Emitter.Age, Period=0.4, Scale=0, Bias=88)
    Cone Axis=<0 0 1>
    Cone Axis Coordinate Space=::World
    Module.Use Velocity Falloff On Cone Axis=true
    Velocity Falloff Away From Cone Axis=0.333
    Velocity Strength=&RandomRangeFloat(Minimum=1200, Maximum=2500)
   )
   &SphereLocation(
    MaskSpawnGroup=false
    Non Uniform Scale=<1 1 0.4>
    Override Seed=false
    Sphere Radius=&Lerp_Float(A=25, B=50, Alpha=&Sine(Normalized Angle=$Emitter.Age, Period=4, Scale=0.5, Bias=1))
    Surface Only=false
    UseNonUniformScale=true
    UseOffset=true
    UseSphereOrientation=false
   )
   &CurlNoiseForce(Noise Frequency=0.04, Noise Strength=250, Pan Noise=true, Pan Noise Field=<0.3 0.5 0.3>)
   &InheritVelocity(Module.Limit Speed=true, Velocity Limit=1500, Velocity Scale=<0 0 0>)
   &ApplyInitialForces()
   &CameraOffset()
   &AddVelocityFromPoint(Module.Use Falloff Distance=false, Origin Offset=<0 0 -50>, Velocity Strength=2000)
  ]
  particleUpdateScript [
   &ParticleState()
   &CurlNoiseForce(Noise Frequency=0.03, Noise Strength=100, Pan Noise=true, Pan Noise Field=<0.1 0.5 0.2>) {enabled=false, module=CurlNoiseForce001}
   &GravityForce(Gravity=<0 0 -1600>)
   &Drag(Drag=1.5, Use Linear Drag=true, Use Rotational Drag=false)
   &ScaleColor(
    Scale Alpha=&FloatFromCurve(FloatCurve=DataInterface() {propertyValues {Curve {keys [<at -0.0046 1.0113 RCIM_Cubic>,<at 1.0011 0.002 RCIM_Cubic>]}}})
    ScaleA=true
    ScaleRGB=true
   )
   &Collision(
    Age Colliding Particles=false
    Collision Enabled=&SetBoolByFloatComparison(A=$Particles.Age, B=0.25)
    Enable Rest State=false
    Friction=0
    Kill Occluded Particles=false
    Kill Offscreen Particles=true
    Module.EnableMaxCollisionCount=false
    Module.Randomize Collision Normal Vector=false
    Restitution=0.7
   ) {enabled=false, module=Collision001}
   &SolveForcesAndVelocity(Clamp Velocity=false, Limit Acceleration=false)
   /NiagaraFluids/Modules/Particles/Fluids_Gas_Source(
    Falloff=0.1
    NoiseIntensity=0.02
    NoiseScale=0.1
    Radius:NiagaraFloat=&FloatFromCurve(
     FloatCurve=DataInterface() {propertyValues {Curve {keys [
         {arriveTangent=-4.9265, interpMode=RCIM_Cubic, leaveTangent=1.4823, tangentMode=RCTM_Break, time=-0.0007, value=0.2746}
         {arriveTangent=0.8901, interpMode=RCIM_Cubic, leaveTangent=0.8901, tangentMode=RCTM_User, time=0.1092, value=0.4826}
         <at 0.4821 0.5642>
         <at 0.9812 0.2949>
        ]}}}
     Scale Curve=&RandomRangeFloat(Minimum=45, Maximum=57)
    )
    Temperature:NiagaraFloat=&FloatFromCurve(
     FloatCurve=DataInterface() {propertyValues {Curve {keys [
         {interpMode=RCIM_Cubic, tangentMode=RCTM_User, time=-0.0059, value=0.3775}
         {arriveTangent=0.5019, interpMode=RCIM_Cubic, leaveTangent=0.5019, tangentMode=RCTM_User, time=0.2461, value=0.4907}
         {arriveTangent=-0.6509, interpMode=RCIM_Cubic, leaveTangent=-0.6509, tangentMode=RCTM_User, time=0.8508, value=0.427}
         <at 1 0>
        ]}}}
     Scale Curve=&RandomRangeFloat(Minimum=500, Maximum=2500)
    )
    VelocityScale:NiagaraFloat=&RandomRangeFloat(Minimum=0.8, Maximum=1.5)
   )
   /SetVariables(Particles.SpriteSize=&Vector2DFromFloat(Value=&Multiply_Float(A=$Particles.fluids_source_radius, B=2))) {isSetParameters=true}
  ]
  rendererClasses [&NiagaraSpriteRendererProperties]
  renderers [{rendererClass=&NiagaraSpriteRendererProperties}]
 }
 initSphere001 {
  emitterUpdateScript [
   &EmitterState(Life Cycle Mode=::NewEnumerator1, Loop Behavior=::NewEnumerator1, Loop Duration=0.4, UseLoopDelay=false)
   &SpawnBurst_Instantaneous(Spawn Count=1, Use Spawn Probability=false)
  ]
  particleSpawnScript [
   &InitializeParticle(
    Color=<rgba 1 0.4259 0.0521 1>
    Lifetime=0.15
    Mesh Scale=&VectorFromFloat(Value=20)
    Module.Write Color=true
    Module.Write Lifetime=true
    Module.Write Mass=true
    Module.Write Material Random=true
    Module.Write Position=true
    Module.Write Scale=true
    Module.Write SpriteRotation=false
    Module.Write SpriteSize=false
    Position=&AddVectorToPosition(Position=$Engine.Emitter.SimulationPosition, Vector=$User.ExplosionSource)
    Sprite Size=<5 5>
   )
   /SetVariables(Particles.LightPos:Vector3f=&Add_Vector(A=Linked() {linkedVariable {name=Particles.Position, type=NiagaraPosition}}, B=<0 0 150>)) {isSetParameters=true}
  ]
  particleUpdateScript [
   &ParticleState()
   &ScaleMeshSize(Scale Factor=&VectorFromFloat(Value=&FloatFromCurve(FloatCurve=DataInterface() {propertyValues {Curve {keys [{arriveTangent=5.2157, interpMode=RCIM_Cubic, leaveTangent=5.2157, tangentMode=RCTM_User, time=0, value=0}, <at 0.4 1 RCIM_Cubic>]}}})))
   &ScaleColor(
    Scale RGB=&VectorFromFloat(Value=&FloatFromCurve(FloatCurve=DataInterface() {propertyValues {Curve {keys [<at 0 0.7>,<at 0.7729 0.7>,<at 1 0>]}}}))
    ScaleA=true
    ScaleRGB=true
   )
   /SetVariables(Particles.LightRadiusV2:NiagaraFloat=&FloatFromCurve(FloatCurve=DataInterface() {propertyValues {Curve {keys [<at 0 1>,<at 1 0>]}}}, Scale Curve=100)) {isSetParameters=true}
  ]
  rendererClasses [&NiagaraMeshRendererProperties, &NiagaraLightRendererProperties]
  renderers [
   {rendererClass=&NiagaraMeshRendererProperties, rendererIndex=0}
   {rendererClass=&NiagaraLightRendererProperties, rendererIndex=1}
  ]
 }
}
summary {
 systemName=NS_GiantSpaceHeater_Explosion
 emitters [
  {emitterName=Smoke_Sphere, rendererClasses [&NiagaraSpriteRendererProperties]}
  {emitterName=FN_Sprite, simTarget=GPUComputeSim, rendererClasses [&NiagaraSpriteRendererProperties]}
  {emitterName=ExtraMeshes_Gen5, rendererClasses [&NiagaraMeshRendererProperties]}
  {emitterName=GlowSphereMesh001, rendererClasses [&NiagaraMeshRendererProperties]}
  {emitterName=Fireworks_Mesh001, rendererClasses [&NiagaraMeshRendererProperties]}
  {emitterName=SmokeyFingers001, rendererClasses [&NiagaraMeshRendererProperties]}
  {emitterName=Smoke_Ring, rendererClasses [&NiagaraSpriteRendererProperties]}
  {emitterName=initSphere001, rendererClasses [&NiagaraMeshRendererProperties, &NiagaraLightRendererProperties]}
  {emitterName=Shockwave, rendererClasses [&NiagaraMeshRendererProperties]}
  {emitterName=S_Combo, rendererClasses [&NiagaraMeshRendererProperties]}
  {emitterName=ExtraMeshes, rendererClasses [&NiagaraMeshRendererProperties]}
  {emitterName=FN_ParticleRibbon, rendererClasses [&NiagaraRibbonRendererProperties]}
  {emitterName=SourceParticle, simTarget=GPUComputeSim, rendererClasses [&NiagaraSpriteRendererProperties]}
  {emitterName=Grid2D_Gas_SmokeFire, simTarget=GPUComputeSim, rendererClasses [&NiagaraSpriteRendererProperties]}
  {emitterName=NewLIght, rendererClasses [&NiagaraComponentRendererProperties]}
 ]
 userVariables [
  User.ExplosionRadius {type=NiagaraFloat, defaultValue=750, description=""}
  User.ExplosionSource {type=Vector3f, description="", defaultValue {struct=Vector3f, value=<0 0 335>}}
  User.GridPivot {type=Vector2f, description="", defaultValue {struct=Vector2f, value=<0.5 0.3>}}
  User.GridSizeWorldSpace {type=Vector2f, description="", defaultValue {struct=Vector2f, value=<8000 8000>}}
  User.Resolution {type=NiagaraInt32}
 ]
}
