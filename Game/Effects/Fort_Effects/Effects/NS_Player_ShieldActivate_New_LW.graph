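# Niagara system for the player shield activation effect, showing brief blue sprites across the character's body.
# ShieldDecals_ bursts one sprite per filtered skeleton bone, using a sequential index to pick both the bone and its hand-tuned offset.
# Sprites follow the skinned mesh with a camera offset, while curves fade alpha and animate dynamic material parameters over life.
# Light, Squares and Squares001 list no renderers, and the older ShieldDecals emitter is disabled.
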NiagaraSystem /Game/Effects/Fort_Effects/Effects/NS_Player_ShieldActivate_New_LW
dependencies {
 usedDataInterfaces [&NiagaraDataInterfaceCurve, &NiagaraDataInterfaceSkeletalMesh, &NiagaraDataInterfaceArrayFloat3]
 usedDynamicInputs []
 usedModules [&SystemState]
}
emitters {
 Light {rendererClasses [], renderers []}
 ShieldDecals {bEnabled=false, rendererClasses [], renderers []}
 ShieldDecals_ {
  emitterUpdateScript [
   &EmitterState(Life Cycle Mode=::NewEnumerator1, Loop Behavior=::NewEnumerator1, Loop Duration=1, UseLoopDelay=true)
   &SpawnBurst_Instantaneous(Spawn Count=17, Use Spawn Probability=false, UseLoopCountLimit=false)
  ]
  particleSpawnScript [
   /SetVariables(Particles.SkeletonSamplingIndex:NiagaraInt32=/CascadeToNiagaraConverter/NiagaraScripts/DynamicInputScripts/SequentialInt(Use Array Data:NiagaraBool=false, First Number:NiagaraInt32=0, Sequence:NiagaraDataInterfaceArrayInt32=?, Last Number:NiagaraInt32=16)) {isSetParameters=true}
   &InitializeParticle(Lifetime=0.95, Sprite Size=<60 60>, Sprite Size Mode=::NewEnumerator4, UsePositionOffset=false, Write Material Random=true)
   &Color(Color=&MakeLinearColorFromVectorAndFloat("Vector (RGB)"=<0.026 0.377 1>, "Float (Alpha)"=1.5))
  ]
  particleUpdateScript [
   &ParticleState()
   &ScaleColor(
    Scale Alpha=&FloatFromCurve(FloatCurve=DataInterface() {propertyValues {Curve {keys [
         {arriveTangent=7.9503, interpMode=RCIM_Cubic, leaveTangent=7.9503, tangentMode=RCTM_User, time=0, value=0}
         {arriveTangent=-1.8291, interpMode=RCIM_Cubic, leaveTangent=-1.8291, tangentMode=RCTM_User, time=0.2, value=1}
         <at 1 0>
        ]}}})
    ScaleA=true
    ScaleRGB=true
   )
   &SkeletalMeshLocation(
    Bone / Socket Index=$Particles.SkeletonSamplingIndex
    Bone Sampling Mode=::NewEnumerator4
    MaskSpawnGroup=false
    Orientation Sampling=::NewEnumerator1
    Sampled Position Offset=&SelectVectorFromArray(
     Direct Array Index=$Particles.SkeletonSamplingIndex
     Vector Selection Array=DataInterface() {propertyValues {FloatData [
        <0 0 0>
        <5 0 0>
        <0 0 0>
        <0 0 0>
        <5 0 0>
        <0 0 0>
        <0 0 0>
        <-5 0 0>
        <25 0 0>
        <18 0 0>
        <-8 0 0>
        <-4 -50 0>
        <-80 10 0>
        <8 0 0>
        <4 50 0>
        <80 -10 0>
        <-10 -22 0>
       ]}}
    )
    Skeletal Mesh=DataInterface() {propertyValues {
      DefaultMesh=null
      ExcludeBoneName=root
      OptionalComponentClass=null
      PreviewMesh=null
      SkinningMode=SkinOnTheFly
      SoftSourceActor=null
      SourceMode=Default
      UvSetIndex=0
      WholeMeshLOD=-1
      bExcludeBone=false
      bReadDeformedGeometry=true
      bRequireCurrentFrameData=true
      ComponentTags []
      FilteredBones [pelvis,spine_02,upperarm_l,lowerarm_l,hand_l,upperarm_r,lowerarm_r,hand_r,neck_01,head,thigh_l,calf_l,foot_l,thigh_r,calf_r,foot_r,Spine_05]
      FilteredSockets []
      MeshUserParameter {parameter {name=None, typeDefHandle {registeredTypeIndex=1}, varData []}}
      SamplingRegions []
     }}
   )
   &CameraOffset(Camera Offset Amount=25)
   &DynamicMaterialParameters(
    Index 0 Param 1=&FloatFromCurve(FloatCurve=DataInterface() {propertyValues {Curve {keys [
         <at 0 -0.5 RCIM_Cubic>
         {arriveTangent=18.2487, interpMode=RCIM_Cubic, leaveTangent=18.2487, tangentMode=RCTM_User, time=0.0692, value=0.2738}
         {arriveTangent=4.8641, interpMode=RCIM_Cubic, leaveTangent=4.8641, tangentMode=RCTM_User, time=0.15, value=1}
         {arriveTangent=4.2839, interpMode=RCIM_Cubic, leaveTangent=4.2839, tangentMode=RCTM_User, time=1, value=5}
        ]}}})
    Index 0 Param 2=1
    Index 0 Param 3=&FloatFromCurve(FloatCurve=DataInterface() {propertyValues {Curve {keys [
         {arriveTangent=2.3217, interpMode=RCIM_Cubic, leaveTangent=2.3217, tangentMode=RCTM_User, time=0, value=-0.25}
         {arriveTangent=0.1844, interpMode=RCIM_Cubic, leaveTangent=0.1844, tangentMode=RCTM_User, time=1, value=1}
        ]}}})
    Index 0 Param 4=1
    Module.Param0WriteEnabled=true
    Module.Param1WriteEnabled=true
    Module.Param2WriteEnabled=true
    Module.Param3WriteEnabled=true
   )
  ]
  rendererClasses [&NiagaraSpriteRendererProperties]
  renderers [{rendererClass=&NiagaraSpriteRendererProperties}]
 }
 Squares {rendererClasses [], renderers []}
 Squares001 {rendererClasses [], renderers []}
}
summary {
 systemName=NS_Player_ShieldActivate_New_LW
 emitters [
  {emitterName=ShieldDecals_, rendererClasses [&NiagaraSpriteRendererProperties]}
  {emitterName=Light, rendererClasses []}
  {emitterName=Squares, rendererClasses []}
  {emitterName=Squares001, rendererClasses []}
  {bEnabled=false, emitterName=ShieldDecals, rendererClasses []}
 ]
 userVariables [User.SpawnShapeZ {type=NiagaraFloat, defaultValue=0, description=""}]
}
