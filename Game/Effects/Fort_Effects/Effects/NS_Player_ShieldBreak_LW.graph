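# Niagara system for the player shield-break effect, with an LW suffix that suggests a lightweight variant.
# It has seven emitters named for a large shield flash, corona, flare, shatter sprites, shatter meshes, light and LOD0 squares.
# SystemState is the only listed module, with no user variables, dynamic inputs or data interfaces.
# The emitters list no renderers, so their visual setup is not visible here.
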NiagaraSystem /Game/Effects/Fort_Effects/Effects/NS_Player_ShieldBreak_LW
dependencies {usedDataInterfaces [], usedDynamicInputs [], usedModules [&SystemState]}
emitters {
 Corona {rendererClasses [], renderers []}
 Flare {rendererClasses [], renderers []}
 Light {rendererClasses [], renderers []}
 ShatterMeshes {rendererClasses [], renderers []}
 ShatterSprites {rendererClasses [], renderers []}
 ShieldFlashLarge {rendererClasses [], renderers []}
 SquaresLOD0 {rendererClasses [], renderers []}
}
summary {
 systemName=NS_Player_ShieldBreak_LW
 emitters [
  {emitterName=ShieldFlashLarge, rendererClasses []}
  {emitterName=Corona, rendererClasses []}
  {emitterName=Flare, rendererClasses []}
  {emitterName=ShatterSprites, rendererClasses []}
  {emitterName=ShatterMeshes, rendererClasses []}
  {emitterName=Light, rendererClasses []}
  {emitterName=SquaresLOD0, rendererClasses []}
 ]
 userVariables []
}
