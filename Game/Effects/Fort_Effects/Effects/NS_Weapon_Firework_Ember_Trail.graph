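# Niagara system for a firework weapon's ember trail, pairing a Glow sprite emitter with a Smoke_Trail ribbon emitter.
# Glow burst-spawns short-lived sprites sized by User FuelAmount and DamageAmount, tinted bright from a user color curve with hue shift.
# Smoke_Trail spawns ribbon points by distance traveled, which drift upward, widen and fade in BaseColor, with dynamic material parameters animated over life.
# OpacityScalar feeds only a disabled module.
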NiagaraSystem /Game/Effects/Fort_Effects/Effects/NS_Weapon_Firework_Ember_Trail
dependencies {
 usedDataInterfaces []
 usedDynamicInputs [&SetBoolByFloatComparison, &VectorLength]
 usedModules [
  &SystemState
  &FN_Timer
  &self:SystemSpawnScript.SystemScriptSource.SystemScriptGraph.NiagaraNodeAssignment_0.SetVariables_#1
 ]
}
emitters {
 Glow {
  emitterUpdateScript [
   &EmitterState(Life Cycle Mode=::NewEnumerator1, Loop Duration=0.25, UseLoopDelay=false)
   &SpawnBurst_Instantaneous(Use Spawn Probability=false, UseLoopCountLimit=false)
  ]
  particleSpawnScript [&InitializeParticle(
    Lifetime=0.5
    Sprite Size Mode=::NewEnumerator3
    Uniform Sprite Size=&Multiply_Float(
     A=500
     B=&Lerp_Float(
      A=0.5
      B=1
      Alpha=&FloatFromCurve(
       FloatCurve=DataInterface() {propertyValues {Curve {keys [
           {interpMode=RCIM_Cubic, tangentMode=RCTM_Break, time=0, value=0}
           {arriveTangent=1.4927, interpMode=RCIM_Cubic, tangentMode=RCTM_Break, time=1, value=1}
          ]}}}
       CurveIndex=$User.FuelAmount
       Scale Curve=$User.DamageAmount
      )
     )
    )
    UsePositionOffset=true
    Write Material Random=true
   )]
  particleUpdateScript [
   &ParticleState()
   &DynamicMaterialParameters(
    Index 0 Param 1=1
    Index 1 Param 1=250
    Index 1 Param 4=1
    Module.Param0WriteEnabled=true
    Module.Param0WriteEnabled1=true
    Module.Param1WriteEnabled=true
    Module.Param1WriteEnabled1=true
    Module.Param2WriteEnabled=false
    Module.Param2WriteEnabled1=true
    Module.Param3WriteEnabled=false
    Module.Param3WriteEnabled1=true
    Write Parameter Index 1=true
   )
   &ScaleColor(
    Scale RGB=&VectorFromFloat(Value=&RandomRangeFloat(Minimum=3, Maximum=10))
    ScaleA=false
    ScaleRGB=true
   ) {enabled=false}
   &ScaleColor(
    Scale Alpha=&FloatFromCurve(FloatCurve=DataInterface() {propertyValues {Curve {keys [
         {interpMode=RCIM_Cubic, tangentMode=RCTM_User, time=0, value=0}
         {interpMode=RCIM_Cubic, tangentMode=RCTM_User, time=0.5, value=1}
         {tangentMode=RCTM_User, time=1, value=0}
        ]}}})
    Scale RGB=&Multiply_Vector(
     A=&MakeVectorFromLinearColorRGB(Color=&ColorFromCurve(
       ColorCurve=$User.BaseCurveColor
       CurveIndex=&Add_Float(A=&FloatFromCurve(FloatCurve=$User.BaseCurveColorHueShift), B=$User.BaseCurveColorPick)
      ))
     B=&VectorFromFloat(Value=10)
    )
    ScaleA=true
    ScaleRGB=true
   ) {module=ScaleColor001}
   &CameraOffset(Camera Offset Amount=15)
  ]
  rendererClasses [&NiagaraSpriteRendererProperties]
  renderers [{rendererClass=&NiagaraSpriteRendererProperties}]
 }
 Smoke_Trail {
  emitterSpawnScript [/SetVariables(
    Emitter.RandIntensity:NiagaraFloat=&RandomRangeFloat(Minimum=0.6, Maximum=0.75)
    Emitter.RandLifetime:NiagaraFloat=&RandomRangeFloat(Minimum=2, Maximum=2)
    Emitter.RandOpacity:NiagaraFloat=&RandomRangeFloat(Minimum=0.7, Maximum=0.9)
    Emitter.RandPanSpeed:NiagaraFloat=&RandomRangeFloat(Minimum=-2, Maximum=-2)
    Emitter.RandRibbonWidth:NiagaraFloat=&RandomRangeFloat(Minimum=80, Maximum=150)
   ) {isSetParameters=true}]
  emitterUpdateScript [
   &EmitterState(Life Cycle Mode=::NewEnumerator1, Loop Duration=0.25, UseLoopDelay=false)
   &SpawnPerUnit(
    Max Movement Threshold=100000
    Module.Use Max Movement Threshold=true
    Module.Use Movement Tolerance=true
    Movement Tolerance=0.1
    Spawn Spacing=100
    Use Spawn Probability=false
   )
   &SpawnPerUnit(
    Max Movement Threshold=100000
    Module.Use Max Movement Threshold=true
    Module.Use Movement Tolerance=true
    Movement Tolerance=0.1
    Spawn Spacing=50
    Use Spawn Probability=false
   ) {enabled=false, module=SpawnPerUnit001}
   &SpawnRate(SpawnRate=200, Use Spawn Probability=false) {enabled=false}
   /SetVariables(Emitter.NearFadeDistance=1200) {isSetParameters=true}
  ]
  particleSpawnScript [
   &InitializeRibbon(
    Color=$User.BaseColor
    Lifetime=&Multiply_Float(A=$Emitter.RandLifetime, B=$User.LifetimeScalar)
    Ribbon Width=$Emitter.RandRibbonWidth
    UsePositionOffset=false
    Write Material Random=true
   )
   &SystemLocation(MaskSpawnGroup=false)
   &AddVelocity()
  ]
  particleUpdateScript [
   &AccelerationForce(Acceleration=<0 0 30>)
   &SolveForcesAndVelocity(Clamp Velocity=false, Limit Acceleration=false)
   &ParticleState()
   &ScaleRibbonWidth(Ribbon Width Scale=&FloatFromCurve(
     FloatCurve=DataInterface() {propertyValues {Curve {keys [
         {interpMode=RCIM_Cubic, tangentMode=RCTM_Break, time=0, value=1}
         {arriveTangent=1.511, interpMode=RCIM_Cubic, leaveTangent=1.511, tangentMode=RCTM_User, time=0.7, value=1.4}
         {arriveTangent=1.6891, interpMode=RCIM_Cubic, tangentMode=RCTM_Break, time=1, value=2}
        ]}}}
     Scale Curve=&FloatFromCurve(FloatCurve=DataInterface() {propertyValues {Curve {keys [<at 0 0>,<at 0.25 1>]}}}, CurveIndex=$Emitter.Age)
    ))
   &ScaleColor(
    Scale Alpha=&FloatFromCurve(
     FloatCurve=DataInterface() {propertyValues {Curve {keys [
         {interpMode=RCIM_Cubic, tangentMode=RCTM_Break, time=0, value=1}
         {arriveTangent=-1.6091, interpMode=RCIM_Cubic, leaveTangent=-1.6091, tangentMode=RCTM_User, time=0.7, value=0.5}
         {arriveTangent=-1.8871, interpMode=RCIM_Cubic, tangentMode=RCTM_Break, time=1, value=0}
        ]}}}
     Scale Curve=$Emitter.RandOpacity
    )
    Scale RGB=&MakeVectorFromLinearColorRGB(Color=&ColorFromCurve(ColorCurve=DataInterface() {propertyValues {
        AlphaCurve {keys []}
        BlueCurve {keys [{arriveTangent=1.3532, interpMode=RCIM_Cubic, time=0, value=1}, {interpMode=RCIM_Cubic, leaveTangent=1.3532, time=0.3742, value=0.1}]}
        GreenCurve {keys [{arriveTangent=0.3631, interpMode=RCIM_Cubic, time=0, value=0}, {interpMode=RCIM_Cubic, leaveTangent=0.3631, time=0.3742, value=0.1}]}
        RedCurve {keys [{arriveTangent=-3.205, interpMode=RCIM_Cubic, time=0, value=0.9987}, {interpMode=RCIM_Cubic, leaveTangent=-3.205, time=0.3742, value=0.1}]}
       }}))
    ScaleA=true
    ScaleRGB=true
   ) {enabled=false}
   &ScaleColor(
    Scale Alpha=$User.OpacityScalar
    Scale RGB=&VectorFromFloat(Value=$Emitter.RandIntensity)
    ScaleA=true
    ScaleRGB=true
   ) {enabled=false, module=ScaleColor001}
   &ScaleColor(
    Scale Alpha=&FloatFromCurve(
     FloatCurve=DataInterface() {propertyValues {Curve {keys [
         {arriveTangent=9.0725, interpMode=RCIM_Cubic, leaveTangent=9.0725, tangentMode=RCTM_User, time=0, value=0}
         {arriveTangent=-2.5249, interpMode=RCIM_Cubic, leaveTangent=-2.5249, tangentMode=RCTM_User, time=0.1408, value=0.9847}
         {arriveTangent=-0.0405, interpMode=RCIM_Cubic, leaveTangent=-0.0405, tangentMode=RCTM_User, time=1, value=0}
        ]}}}
     Scale Curve=0.25
    )
    Scale RGB=&Multiply_VectorByFloat(
     Vector=&MakeVectorFromLinearColorRGB(Color=$User.BaseColor)
     Float=&FloatFromCurve(FloatCurve=DataInterface() {propertyValues {Curve {keys [
          {interpMode=RCIM_Cubic, tangentMode=RCTM_User, time=0, value=5}
          {arriveTangent=-0.0405, interpMode=RCIM_Cubic, leaveTangent=-0.0405, tangentMode=RCTM_User, time=0.7, value=3.5}
          {arriveTangent=-8.0089, interpMode=RCIM_Cubic, leaveTangent=-8.0089, tangentMode=RCTM_User, time=1, value=0}
         ]}}})
    )
    ScaleA=true
    ScaleRGB=true
   ) {module=ScaleColor002}
   &DynamicMaterialParameters(
    Index 0 Param 1=&FloatFromCurve(FloatCurve=DataInterface() {propertyValues {Curve {keys [
         {interpMode=RCIM_Cubic, leaveTangent=-1.462, tangentMode=RCTM_Break, time=0, value=1}
         {arriveTangent=-0.7544, interpMode=RCIM_Cubic, tangentMode=RCTM_Break, time=1, value=-1}
        ]}}})
    Index 0 Param 2=&FloatFromCurve(FloatCurve=DataInterface() {propertyValues {Curve {keys [<at 0 2>,<at 0.15 0>]}}})
    Index 0 Param 3=$Emitter.RandPanSpeed
    Index 0 Param 4=1
    Module.Param0WriteEnabled=true
    Module.Param1WriteEnabled=true
    Module.Param2WriteEnabled=true
    Module.Param3WriteEnabled=true
   )
  ]
  rendererClasses [&NiagaraRibbonRendererProperties]
  renderers [{rendererClass=&NiagaraRibbonRendererProperties}]
 }
}
summary {
 systemName=NS_Weapon_Firework_Ember_Trail
 emitters [
  {emitterName=Glow, rendererClasses [&NiagaraSpriteRendererProperties]}
  {emitterName=Smoke_Trail, rendererClasses [&NiagaraRibbonRendererProperties]}
 ]
 userVariables [
  User.BaseColor {type=LinearColor, description="", defaultValue {struct=LinearColor, value=<rgba 1 0.1 0.02 1>}}
  User.BaseCurveColor {
   type=NiagaraDataInterfaceColorCurve
   description=""
   defaultValue {
    struct=NiagaraExt_VariableValue_DataInterface
    value {
     dataInterface=&self:NiagaraDataInterfaceColorCurve_0
     dataInterfaceClass=&NiagaraDataInterfaceColorCurve
    }
   }
  }
  User.BaseCurveColorHueShift {
   type=NiagaraDataInterfaceCurve
   description=""
   defaultValue {
    struct=NiagaraExt_VariableValue_DataInterface
    value {
     dataInterface=&self:NiagaraDataInterfaceCurve_0
     dataInterfaceClass=&NiagaraDataInterfaceCurve
    }
   }
  }
  User.BaseCurveColorPick {type=NiagaraFloat, defaultValue=0.87, description=""}
  User.DamageAmount {type=NiagaraFloat, defaultValue=1, description=""}
  User.FuelAmount {type=NiagaraFloat, defaultValue=1, description=""}
  User.LifetimeScalar {type=NiagaraFloat, defaultValue=1, description=""}
  User.OpacityScalar {type=NiagaraFloat, defaultValue=1, description=""}
  User.RibbonIDOffset {type=NiagaraInt32}
  User.ThrusterOffset {type=NiagaraFloat, defaultValue=0, description=""}
 ]
}
