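# Translucent unlit effect material, usable on Niagara mesh particles, that draws a tinted box with brightened face edges.
# Sphere masks on U and V gradients form the edge mask, which joins a base color and Fresnel to build the glow.
# Switches optionally apply particle color and time-of-day tint before emissive adjustments.
# Further switches choose particle alpha or clamped edge mask for opacity, and toggle a camera-vector-based world position offset.
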Material /Game/Effects/Fort_Effects/Materials/BoxMaterial
graph {
 properties {
  bUsedWithNanite=true
  bUsedWithNiagaraMeshParticles=true
  blendMode=BLEND_Translucent
  pixelDepthOffsetMode=PDOM_Legacy
  shadingModel=MSM_Unlit
  usedShadingModels=Unlit
 }
 nodes [
  n1 = MaterialFunctionCall(
   Color_Morning=VectorParameter() {parameterName=ColorMorning, defaultValue=<rgba 1 1 1 1>}.RGB
   Color_Day=VectorParameter() {parameterName=ColorDay, defaultValue=<rgba 1 1 1 1>}.RGB
   Color_Evening=VectorParameter() {parameterName=ColorEvening, defaultValue=<rgba 1 1 1 1>}.RGB
   Color_Night=VectorParameter() {parameterName=ColorNight, defaultValue=<rgba 1 1 1 1>}.RGB
  ) {materialFunction=&Color_Time_of_Day}
  n2 = ParticleColor()
  n3 = MaterialFunctionCall() {materialFunction=&LinearGradient}
  n4 = Constant() {r=0.5}
  n5 = ScalarParameter() {defaultValue=0.495, parameterName=Edge_Radius}
  n6 = Constant() {r=0.99}
  n7 = SphereMask(A=ComponentMask(None=@n3.UGradient) {r=true}, B=@n4, Radius=@n5, Hardness=@n6) {hardnessPercent=99}
  n8 = SphereMask(A=ComponentMask(None=@n3.VGradient) {r=true}, B=@n4, Radius=@n5, Hardness=@n6) {hardnessPercent=90}
  n9 = Clamp(None=Add(A=OneMinus(None=@n7), B=OneMinus(None=@n8)))
  n10 = ScalarParameter() {defaultValue=1, parameterName=Edge_intensity}
  n11 = Add(
   A=Fresnel(
    ExponentIn=ScalarParameter() {defaultValue=5, parameterName=Fresnel_Exponentin}
    BaseReflectFractionIn=ScalarParameter() {defaultValue=0.04, parameterName=Fresnel_BaseReflectFraction}
   )
   B=Add(A=VectorParameter() {parameterName=Color, defaultValue=<rgba 0.0881 0.1669 0.8319 1>}.RGB, B=Multiply(A=@n9, B=@n10))
  )
  n12 = StaticSwitchParameter(True=Multiply(A=@n2.RGB, B=@n11), False=@n11) {parameterName=Use_particle_color}
  MaterialFunctionCall(Emissive=StaticSwitchParameter(True=Multiply(A=@n1.Result, B=@n12), False=@n12) {parameterName=Use Color_Time_Of_Day}) {materialFunction=&MF_FX_EmissiveAdjustments}
  StaticSwitchParameter(
   True=Constant()
   False=MaterialFunctionCall!(Use World Position Offsets) {materialFunction=/Engine/Functions/Engine_MaterialFunctions02/Utility/CameraVectorWithWPOOptions}.Result
  ) {parameterName=Do_not_use_WPO}
  StaticSwitchParameter(
   True=@n2.A
   False=StaticSwitchParameter(True=@n2.A, False=Clamp(None=@n9) {maxDefault=0.1, minDefault=0.04}) {parameterName=Use_particle_color}
  ) {parameterName=Use_simple_alpha}
 ]
}
