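# BoxMaterial2Sided is an unlit, translucent, two-sided effects material with an orange tint, named for use on boxes.
# A Fresnel term is added to a constant orange color, then to a CameraDepthFade result clamped to a low maximum.
# The fade uses a negative length to invert its distance falloff, and the same clamped value also feeds a separate Add node.
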Material /Game/Effects/Fort_Effects/Materials/BoxMaterial2Sided
graph {
 properties {blendMode=BLEND_Translucent, pixelDepthOffsetMode=PDOM_Legacy, shadingModel=MSM_Unlit, twoSided=true, usedShadingModels=Unlit}
 nodes [
  n1 = Clamp(None=MaterialFunctionCall(Fade Length=Constant() {r=-1000}, Fade Offset=Constant() {r=5000}) {materialFunction=&CameraDepthFade}.Result) {maxDefault=0.05}
  Add(A=Add(A=Fresnel(), B=Constant3Vector() {constant=<rgba 1 0.1231 0 1>}), B=@n1)
  Add(B=@n1)
 ]
}
