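# BoxMaterial_2sided is an unlit, translucent, two-sided effects material that draws a faint blue box with highlighted edges.
# LinearGradient U and V each feed a centered SphereMask, and the inverted masks are summed into a clamped UV border mask.
# One branch adds that mask to a constant blue and a Fresnel term, apparently for emissive color.
# A second branch clamps it to a low, nonzero range, apparently for opacity.
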Material /Game/Effects/Fort_Effects/Materials/BoxMaterial_2sided
graph {
 properties {blendMode=BLEND_Translucent, pixelDepthOffsetMode=PDOM_Legacy, shadingModel=MSM_Unlit, twoSided=true, usedShadingModels=Unlit}
 nodes [
  n1 = MaterialFunctionCall() {materialFunction=&LinearGradient}
  n2 = Constant() {r=0.5}
  n3 = Constant() {r=0.495}
  n4 = Constant() {r=0.99}
  n5 = SphereMask(A=ComponentMask(None=@n1.UGradient) {r=true}, B=@n2, Radius=@n3, Hardness=@n4) {hardnessPercent=99}
  n6 = SphereMask(A=ComponentMask(None=@n1.VGradient) {r=true}, B=@n2, Radius=@n3, Hardness=@n4) {hardnessPercent=90}
  n7 = Clamp(None=Add(A=OneMinus(None=@n5), B=OneMinus(None=@n6)))
  Add(A=Fresnel(), B=Add(A=@n7, B=Constant3Vector() {constant=<rgba 0.0881 0.1669 0.8319 0>}))
  Clamp(None=@n7) {maxDefault=0.1, minDefault=0.04}
 ]
}
