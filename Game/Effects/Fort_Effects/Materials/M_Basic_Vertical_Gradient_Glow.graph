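# Unlit additive effects material that renders a vertical gradient glow between parameterized dark and bright colors.
# The V texture coordinate, raised to a power, drives the color lerp and serves as the base fade mask.
# Depth fade and view-angle fade multiply that mask, but a shading-path switch replaces both with 1 on mobile.
# A static switch, on by default, also applies near-camera fade.
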Material /Game/Effects/Fort_Effects/Materials/M_Basic_Vertical_Gradient_Glow
graph {
 properties {blendMode=BLEND_Additive, pixelDepthOffsetMode=PDOM_Legacy, shadingModel=MSM_Unlit, usedShadingModels=Unlit}
 nodes [
  n1 = Power(Base=ComponentMask(None=TextureCoordinate()) {g=true}, Exp=ScalarParameter() {defaultValue=1.5, parameterName=GradientPower})
  LinearInterpolate(
   A=VectorParameter() {parameterName=DarkColor, defaultValue=<rgba 0.1396 0.0965 0.5 0>}.RGB
   B=VectorParameter() {parameterName=BrightColor, defaultValue=<rgba 0.1494 0.7738 1.25 0>}.RGB
   Alpha=@n1
  )
  n2 = DepthFade(FadeDistance=ScalarParameter() {defaultValue=25, parameterName=DepthFadeDist})
  n3 = MaterialFunctionCall("FadeSensitivity (optional)"=ScalarParameter() {defaultValue=1.5, parameterName=AngleFade}) {materialFunction=&FadeBasedOnViewAngle}
  n4 = Multiply(A=@n1, B=ShadingPathSwitch(Default=Multiply(A=@n2, B=@n3.Result), Mobile=Constant() {r=1}) {desc=Migrated from FeatureLevelSwitch})
  n5 = ScalarParameter() {defaultValue=1400, parameterName=NearCamFadeDist}
  StaticSwitchParameter(
   True=Multiply(A=@n4, B=MaterialFunctionCall(Distance=@n5) {materialFunction=&Near_Camera_Fade}.Result)
   False=@n4
  ) {defaultValue=true, parameterName=UseNearCamFade}
 ]
}
