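# Translucent, unlit, two-sided trail material for the Boardwalk pickaxe swing, usable on Niagara ribbons, mesh particles and beam trails.
# A sine-based UV edge fade minus tiled, offset noise, times an inverted sphere mask, forms the streak shape.
# That mask blends two emissive colors and shapes opacity, both scaled by particle alpha.
# SwizeUVforNiagara blends to inverted UVs; CameraOffset supplies an adjustable position offset.
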Material /Game/Effects/Fort_Effects/Materials/M_Boardwalk_Pickaxe_AnimTrail_B
graph {
 properties {
  bUsedWithBeamTrails=true
  bUsedWithNanite=true
  bUsedWithNiagaraMeshParticles=true
  bUsedWithNiagaraRibbons=true
  blendMode=BLEND_Translucent
  pixelDepthOffsetMode=PDOM_Legacy
  shadingModel=MSM_Unlit
  twoSided=true
  usedShadingModels=Unlit
 }
 nodes [
  MaterialFunctionCall(Offset Amount=ScalarParameter() {defaultValue=5, parameterName=Offset}) {materialFunction=/Engine/Functions/Engine_MaterialFunctions02/WorldPositionOffset/CameraOffset}
  n1 = TextureCoordinate()
  n2 = LinearInterpolate(A=@n1, B=OneMinus(None=@n1), Alpha=ScalarParameter() {parameterName=SwizeUVforNiagara})
  n3 = Sine(None=Divide(A=@n2))
  n4 = Multiply(A=ComponentMask(None=@n3) {r=true}, B=ComponentMask(None=@n3) {g=true})
  n5 = Divide(
   A=Multiply(A=@n2, B=Constant2Vector() {g=4, r=0.04})
   B=ComponentMask(None=VectorParameter() {parameterName=Trail_UV_tile, defaultValue=<rgba 12 15 0 1>}.RGB) {g=true, r=true}
  ) {constB=12}
  n6 = ComponentMask(None=VectorParameter() {parameterName=Offset_UV, defaultValue=<rgba 0.4 0 0 1>}.RGB) {g=true, r=true}
  n7 = LinearInterpolate(Alpha=TextureSample(UVs=Add(A=@n5, B=@n6)) {samplerType=SAMPLERTYPE_Normal, texture=/Game/Global/Textures/Default/Noise/64x64Noise}.G) {constA=0.2, constB=0.5}
  n8 = Constant2Vector() {g=1, r=0.5}
  n9 = OneMinus(None=SphereMask(A=Multiply(A=@n2, B=@n8), B=Constant2Vector() {g=0.82, r=0.85}, Radius=Constant() {r=0.8}, Hardness=Constant() {r=0.3}))
  n10 = Saturate(None=Multiply(A=Subtract(A=@n4, B=@n7), B=Power(Base=@n9) {constExponent=5}))
  n11 = LinearInterpolate(
   A=VectorParameter() {parameterName=EmissiveB, defaultValue=<rgba 0 0.4485 1 0>}.RGB
   B=VectorParameter() {parameterName=EmissiveA, defaultValue=<rgba 0.2067 0.8192 1 0>}.RGB
   Alpha=@n10
  )
  n12 = ParticleColor()
  MaterialFunctionCall(Emissive=Multiply(A=Multiply(A=@n11, B=@n12.A), B=LinearInterpolate(Alpha=@n10) {constA=1, constB=2})) {materialFunction=&MF_FX_EmissiveAdjustments}
  Multiply(A=@n12.A, B=Multiply(A=Power(Base=@n4), B=@n10))
 ]
}
