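# Translucent unlit trail material for electric contrails on ribbons, beams and Niagara particles, built from a tiling lightning texture.
# Scaled UVs pan continuously or optionally step, jittered by noise that advances in time steps for crackle.
# Texture green blends glow toward particle color and drives opacity, blue masks HDR emissive, red feeds optional erosion.
# Opacity adds optional panning noise, length and edge fades, depth fade and non-mobile near-camera fade.
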Material /Game/Effects/Fort_Effects/Materials/M_Contrail_Electric_Ribbon
graph {
 properties {
  bAllowNegativeEmissiveColor=true
  bUsedWithBeamTrails=true
  bUsedWithNiagaraMeshParticles=true
  bUsedWithNiagaraRibbons=true
  bUsedWithNiagaraSprites=true
  blendMode=BLEND_Translucent
  pixelDepthOffsetMode=PDOM_Legacy
  shadingModel=MSM_Unlit
  usedShadingModels=Unlit
 }
 nodes [
  n1 = ParticleColor()
  n2 = VectorParameter() {parameterName=SecondaryGlowColor, defaultValue=<rgba 60 55.9389 21.5625 1>}
  n3 = ScalarParameter() {defaultValue=1, parameterName=SecondaryGlowColor_Influence}
  n4 = Multiply(A=Time(), B=ScalarParameter() {defaultValue=16, parameterName=UVDistortSpeed}) {constB=16}
  n5 = Panner(
   Coordinate=LinearInterpolate(
    A=TextureCoordinate() {uTiling=0.46, vTiling=0.2476}
    B=TextureCoordinate() {coordinateIndex=1, uTiling=0.46, vTiling=0.2476}
    Alpha=ScalarParameter() {defaultValue=1, parameterName=UseSecondUV}
   )
   Time=Divide(A=Ceil(None=@n4)) {constB=4}
  ) {speedX=0.4, speedY=0.4}
  n6 = Subtract(A=ComponentMask(None=TextureSample(UVs=@n5) {samplerType=SAMPLERTYPE_LinearColor, texture=/Game/Global/Textures/Default/Noise/TilingNoise06_N_uncompressed}.RGB) {g=true, r=true}) {constB=0.5}
  n7 = ScalarParameter() {defaultValue=1, parameterName=U_Scale}
  n8 = DynamicParameter() {paramNames [Dynamic V Scale, Erode Offset, Dynamic U Scale, RandomUV_Noise]}
  n9 = ScalarParameter() {parameterName=Use DynamicParameterforUScale}
  n10 = ScalarParameter() {defaultValue=1, parameterName=V Scale}
  n11 = ScalarParameter() {parameterName=Use DynamicParameterforVScale}
  n12 = Multiply(
   A=LinearInterpolate(A=TextureCoordinate(), B=TextureCoordinate() {coordinateIndex=1}, Alpha=ScalarParameter() {defaultValue=1, parameterName=UseSecondUV})
   B=AppendVector(A=LinearInterpolate(A=@n7, B=@n8.Dynamic U Scale, Alpha=@n9), B=LinearInterpolate(A=@n10, B=@n8.Dynamic V Scale, Alpha=@n11))
  )
  n13 = MaterialFunctionCall(Float2=@n12) {materialFunction=&BreakOutFloat2Components}
  n14 = MaterialFunctionCall(X=@n13.R, Y=Add(A=@n13.G, B=ScalarParameter() {defaultValue=0.2, parameterName=BumpUp}) {constB=0.4}) {materialFunction=&MakeFloat2}
  n15 = AppendVector(A=ScalarParameter() {defaultValue=1, parameterName=X Speed}, B=ScalarParameter() {parameterName=Y Speed})
  n16 = MaterialFunctionCall!(
   TimeSpeed=ScalarParameter() {defaultValue=1, group=Step Panner, parameterName=StepPan_Timespeed}
   Jump Amount Per Step=AppendVector(
    A=ScalarParameter() {defaultValue=5, group=Step Panner, parameterName=StepPan_JumpAmount_X}
    B=ScalarParameter() {group=Step Panner, parameterName=StepPan_JumpAmount_Y}
   )
   Number of Steps Per Unit of Time=AppendVector(
    A=ScalarParameter() {defaultValue=1.5, group=Step Panner, parameterName=XStepPan_NumberofStepsPerSecond}
    B=ScalarParameter() {defaultValue=1.5, group=Step Panner, parameterName=YStepPan_NumberofStepsPerSecond}
   )
   Time Offset
  ) {materialFunction=/Engine/Functions/Engine_MaterialFunctions02/SteppingPannerTime}
  n17 = ScalarParameter() {group=Step Panner, parameterName=Use_StepPanner}
  n18 = Add(
   A=LinearInterpolate(B=Multiply(A=@n6) {constB=0.1}, Alpha=ScalarParameter() {defaultValue=1, parameterName=Use Distortion})
   B=Panner(
    Coordinate=StaticSwitchParameter(True=@n14.Result, False=@n12) {parameterName=BumpUpOnY}
    Speed=LinearInterpolate(A=@n15, B=@n16.Result, Alpha=@n17)
   ) {bFractionalPart=true}
  ) {constB=0}
  n19 = MaterialFunctionCall(Float2=StaticSwitchParameter(True=MaterialFunctionCall(XY=@n18) {materialFunction=&Swizzle}.YX, False=@n18) {defaultValue=true, parameterName=UseSwizzle}) {materialFunction=&BreakOutFloat2Components}
  n20 = OneMinus(None=@n19.G)
  n21 = ParticleRandom()
  n22 = AppendVector(
   A=LinearInterpolate(A=@n19.R, B=Clamp(None=@n19.R), Alpha=ScalarParameter() {parameterName=Clamp Texture V})
   B=LinearInterpolate(A=@n20, B=Add(A=@n20, B=@n21), Alpha=ScalarParameter() {parameterName=UseParticleRandomOffset})
  )
  n23 = AppendVector(A=DynamicParameter() {parameterIndex=1, paramNames [U Offset, Param02, Param03, Param04]}.U Offset, B=Constant())
  n24 = TextureSampleParameter2D(UVs=LinearInterpolate(A=@n22, B=Add(A=@n22, B=@n23), Alpha=ScalarParameter() {parameterName=Use U Dyn Offset})) {parameterName=TilingTexture, texture=/Game/Environments/Hexmap/HexClouds/Lightning/T_Hexmap_Lightning_02_E}
  n25 = Clamp(None=LinearInterpolate(A=LinearInterpolate(A=@n1.RGB, B=@n2.RGB, Alpha=@n3) {constB=0}, B=@n1.RGB, Alpha=@n24.G))
  n26 = ScalarParameter() {defaultValue=5, parameterName=HDR_Power}
  MaterialExpressionReroute_6 = Reroute(None=Multiply(
    A=MaterialFunctionCall(LDR Input=@n25, HDR Tint=@n1.RGB, Derive HDR Power=@n26) {materialFunction=&DeriveHDRfromLDR}.RemappedRange
    B=@n24.B
   ))
  MaterialExpressionScalarParameter_24 = ScalarParameter() {group=Options, parameterName=Use Eye Adaption Exposure Control for Emissives}
  MaterialExpressionEyeAdaptation_1 = EyeAdaptation()
  n27 = Custom!(
   LightValue=@MaterialExpressionReroute_6
   Alpha=@MaterialExpressionScalarParameter_24
   Adaptation=@MaterialExpressionEyeAdaptation_1.EyeAdaptation
  ) {
   code="float LerpLogScale = -Alpha * log(Adaptation);\r\nfloat Scale = exp(LerpLogScale);\r\nreturn LightValue * Scale;\r\n\r\n\r\n"
   inputs [
    {inputName=LightValue, input {expression=&self:MaterialExpressionReroute_6}}
    {inputName=Alpha, input {expression=&self:MaterialExpressionScalarParameter_24}}
    {inputName=Adaptation, input {expression=&self:MaterialExpressionEyeAdaptation_1}}
   ]
  }
  MaterialFunctionCall(Emissive=@n27) {materialFunction=&MF_FX_EmissiveAdjustments}
  n28 = Multiply(
   A=TextureCoordinate()
   B=AppendVector(
    A=ScalarParameter() {defaultValue=1, group=AlphaNoise, parameterName=AlphaNoise-Tiling-X}
    B=ScalarParameter() {defaultValue=1, group=AlphaNoise, parameterName=AlphaNoise-Tiling-Y}
   )
  )
  n29 = DynamicParameter() {paramNames [Dynamic V Scale, Erode Offset, Dynamic U Scale, RandomUV_Noise]}
  n30 = AppendVector(A=@n29.RandomUV_Noise, B=@n29.RandomUV_Noise)
  n31 = ScalarParameter() {group=AlphaNoise, parameterName=AlphaNoise-Panning-X}
  n32 = ScalarParameter() {defaultValue=0.25, group=AlphaNoise, parameterName=AlphaNoise-Panning-Y}
  n33 = TextureSampleParameter2D(UVs=Panner(Coordinate=Add(A=@n28, B=@n30), Speed=AppendVector(A=@n31, B=@n32))) {
   group=AlphaNoise
   parameterName=AlphaNoise-Texture
   samplerType=SAMPLERTYPE_Grayscale
   texture=&LowResBlurredNoise
  }
  n34 = ScalarParameter() {defaultValue=1, group=AlphaNoise, parameterName=AlphaNoise-Power}
  n35 = Clamp(None=Multiply(A=Power(Base=@n33.RGB, Exp=@n34), B=ScalarParameter() {defaultValue=1, group=AlphaNoise, parameterName=AlphaNoise-Hardness}))
  n36 = Multiply(A=@n1.A, B=@n24.G)
  n37 = LinearInterpolate(A=@n36, B=Multiply(A=@n36, B=Add(A=@n24.R, B=@n29.Erode Offset)), Alpha=ScalarParameter() {parameterName=USe Erode})
  n38 = TextureCoordinate()
  n39 = Power(Base=OneMinus(None=ComponentMask(None=@n38) {r=true}), Exp=ScalarParameter() {defaultValue=3.1188, parameterName=Beginning Fade Exp})
  n40 = Multiply(A=@n39, B=ScalarParameter() {defaultValue=1, parameterName=Power_mask_intersity})
  n41 = Multiply(
   A=LinearInterpolate(A=@n37, B=Multiply(A=@n37, B=@n35), Alpha=ScalarParameter() {group=AlphaNoise, parameterName=UseAlphaNoise, sortPriority=1})
   B=LinearInterpolate(A=@n39, B=Saturate(None=@n40), Alpha=ScalarParameter() {parameterName=Use_Clearer_power_mask})
  )
  n42 = DepthFade(Opacity=@n41, FadeDistance=ScalarParameter() {defaultValue=5, group=Depth Fade, parameterName=NearFadeDistance})
  n43 = ScalarParameter() {group=Depth Fade, parameterName=Use Depth Fade}
  n44 = MaterialFunctionCall(Distance=ScalarParameter() {defaultValue=128, parameterName=CameraFadeDistance}) {materialFunction=&Near_Camera_Fade}
  n45 = Constant() {r=1}
  n46 = TextureCoordinate()
  n47 = Multiply(A=Power(Base=ComponentMask(None=@n46) {r=true}), B=ScalarParameter() {defaultValue=5, parameterName=left edge fade intensity}) {constB=5}
  Multiply(
   A=Multiply(
    A=LinearInterpolate(A=@n41, B=@n42, Alpha=@n43)
    B=ShadingPathSwitch(Default=@n44.Result, Mobile=@n45) {desc=Migrated from FeatureLevelSwitch}
   )
   B=LinearInterpolate(A=Constant() {r=1}, B=Saturate(None=@n47), Alpha=ScalarParameter() {parameterName=Use left edge fade})
  )
  TextureCoordinate() {coordinateIndex=1}
 ]
}
