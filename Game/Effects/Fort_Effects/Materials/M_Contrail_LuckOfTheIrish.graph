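# Unlit, two-sided additive material for the Luck of the Irish contrail on Niagara ribbons and mesh particles.
# A panning smoke texture, warped by a scrolling normal map and time-offset by a dynamic parameter, drives brightness and opacity.
# "Use Rainbow color" selects a three-color blend when true, otherwise a hue-shifted main color, optionally rainbow-driven.
# Opacity combines depth fades, sphere or trail masks, particle alpha and optional time-of-day and dynamic alpha.
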Material /Game/Effects/Fort_Effects/Materials/M_Contrail_LuckOfTheIrish
graph {
 properties {
  bUsedWithNiagaraMeshParticles=true
  bUsedWithNiagaraRibbons=true
  blendMode=BLEND_Additive
  pixelDepthOffsetMode=PDOM_Legacy
  shadingModel=MSM_Unlit
  twoSided=true
  usedShadingModels=Unlit
 }
 nodes [
  n1 = VectorParameter() {parameterName=Base Color A, defaultValue=<rgba 0 0 0 1>}
  n2 = VectorParameter() {parameterName=Base Color B, defaultValue=<rgba 1 1 1 1>}
  n3 = VectorParameter() {parameterName=UV, defaultValue=<rgba 1 1 1 1>}
  n4 = Multiply(A=TextureCoordinate() {uTiling=0.55}, B=AppendVector(A=@n3.R, B=@n3.R))
  n5 = Panner(Coordinate=TextureCoordinate() {uTiling=0.35, vTiling=0.5}) {speedX=-0.8}
  n6 = Multiply(
   A=ComponentMask(None=TextureSample(UVs=@n5) {samplerType=SAMPLERTYPE_Normal, texture=&T_Weap_Distort_N_02}.RGB) {g=true, r=true}
   B=ScalarParameter() {defaultValue=0.28, parameterName=Noise_Size}
  ) {constB=0.28}
  n7 = Time()
  n8 = DynamicParameter() {paramNames [PanTexture-Offset, Alpha, Param3, Param4]}
  n9 = ScalarParameter() {defaultValue=-0.9, parameterName=Speed_UV}
  n10 = Constant()
  n11 = TextureSampleParameter2D(UVs=Panner(Coordinate=Add(A=@n4, B=@n6), Time=Add(A=@n7, B=@n8.PanTexture-Offset), Speed=AppendVector(A=@n9, B=@n10)) {speedX=-0.9}) {parameterName=Base Texture, samplerType=SAMPLERTYPE_Grayscale, texture=&T_Smoke_Trail_Soft}
  n12 = MaterialFunctionCall(Float2=TextureCoordinate()) {materialFunction=&BreakOutFloat2Components}
  n13 = ScalarParameter() {defaultValue=2, parameterName=Color Power}
  n14 = Constant()
  n15 = TextureCoordinate() {uTiling=0.005}
  n16 = TextureSample(UVs=Panner(Coordinate=ComponentMask(None=@n15) {r=true}) {speedX=-0.07}) {texture=/Game/Global/Textures/Default/Masks_and_Gradients/T_Grad_Rainbow}
  n17 = Multiply(
   A=MaterialFunctionCall(
    Hue Shift Percentage=StaticSwitchParameter(True=@n14, False=@n16.RGB) {parameterName=Use a fixed color}
    Texture=VectorParameter() {parameterName=Main color, defaultValue=<rgba 0.205 1 0.108 1>}.RGB
   ) {materialFunction=&HueShift}.Result
   B=StaticSwitchParameter(True=Constant() {r=1}, False=ParticleColor().RGB) {parameterName=Use particle color in the last part}
  )
  n18 = StaticSwitchParameter(
   True=LinearInterpolate(
    A=LinearInterpolate(A=@n1.RGB, B=@n2.RGB, Alpha=@n11.RGB)
    B=VectorParameter() {parameterName=Base Color C, defaultValue=<rgba 1 1 1 1>}.RGB
    Alpha=Power(Base=@n12.R, Exp=@n13)
   )
   False=MaterialFunctionCall(LDR Input=Clamp(None=@n17), HDR Tint=@n17) {materialFunction=&DeriveHDRfromLDR}.RemappedRange
  ) {parameterName=Use Rainbow color}
  n19 = ScalarParameter() {defaultValue=1.2, parameterName=RGB}
  n20 = Power(Base=@n11.RGB, Exp=ScalarParameter() {defaultValue=1, parameterName=Texture power})
  n21 = Desaturation(None=Multiply(A=Multiply(A=@n18, B=@n19), B=@n20), Fraction=ScalarParameter() {defaultValue=-2, parameterName=Desaturation Color})
  n22 = ParticleColor()
  MaterialFunctionCall(Emissive=StaticSwitchParameter(True=Multiply(A=@n21, B=@n22.RGB), False=@n21) {parameterName=Use particle color in the last part}) {materialFunction=&MF_FX_EmissiveAdjustments}
  n23 = Multiply(A=@n20, B=@n20)
  n24 = DepthFade(FadeDistance=ScalarParameter() {defaultValue=200, parameterName=DepthFade-Inverted_Distance})
  n25 = DepthFade(Opacity=Add(A=@n23, B=OneMinus(None=@n24)), FadeDistance=ScalarParameter() {defaultValue=100, parameterName=DepthFade-Standard_Distance}) {fadeDistanceDefault=20}
  n26 = ComponentMask(None=TextureCoordinate()) {r=true}
  n27 = ComponentMask(None=TextureCoordinate()) {r=true}
  n28 = Power(
   Base=Multiply(A=OneMinus(None=@n27), B=ScalarParameter() {defaultValue=0.5, parameterName=Holiday Cookies Mask Mult}) {constB=5}
   Exp=ScalarParameter() {defaultValue=0.5, parameterName=Holiday Cookies Mask Power}
  ) {constExponent=0.5}
  n29 = Constant() {r=1}
  n30 = SphereMask(A=TextureCoordinate(), B=Constant() {r=0.5}) {attenuationRadius=0.5, hardnessPercent=0}
  n31 = Multiply(A=@n30, B=@n30)
  n32 = ScalarParameter() {defaultValue=1, parameterName=Use Sphere Mask}
  n33 = Multiply(
   A=ShadingPathSwitch(Default=@n25, Deferred=@n25, Mobile=@n23) {desc=Migrated from FeatureLevelSwitch}
   B=StaticSwitchParameter(True=Multiply(A=@n26, B=@n28), False=LinearInterpolate(A=@n29, B=@n31, Alpha=@n32)) {parameterName=Use HolidayCookies_Trail_Mask}
  )
  n34 = ScalarParameter() {defaultValue=1, parameterName=Alpha_Day}
  n35 = ScalarParameter() {defaultValue=1, parameterName=Alpha_Night}
  n36 = Multiply(
   A=@n33
   B=MaterialFunctionCall(Color_Morning=@n34, Color_Day=@n34, Color_Evening=@n35, Color_Night=@n35) {materialFunction=&Color_Time_of_Day}.Result
  )
  Multiply(
   A=Multiply(A=@n22.A, B=StaticSwitchParameter(True=@n36, False=@n33) {parameterName=Alpha of day})
   B=StaticSwitchParameter(True=DynamicParameter() {paramNames [PanTexture-Offset, Alpha, Param3, Param4]}.Alpha, False=Constant() {r=1}) {parameterName=Use Dynamic to Control Alpha}
  )
 ]
}
