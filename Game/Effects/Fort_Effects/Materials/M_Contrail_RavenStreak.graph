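# Unlit translucent ribbon and beam-trail material for the Raven contrail, drawing a dark, inky smoke streak.
# Panned smoke textures build color and opacity, with U blended between UV1 and a sphere-mask gradient around an offset object position, plus dynamic-parameter offsets.
# Emissive follows time of day and particle color.
# Opacity combines UV edge masks, a sine pulse, depth fade and camera-proximity fade, with optional leading-edge mask and distance fade.
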Material /Game/Effects/Fort_Effects/Materials/M_Contrail_RavenStreak
graph {
 properties {
  bUsedWithBeamTrails=true
  bUsedWithNiagaraRibbons=true
  blendMode=BLEND_Translucent
  pixelDepthOffsetMode=PDOM_Legacy
  shadingModel=MSM_Unlit
  usedShadingModels=Unlit
 }
 nodes [
  n1 = VectorParameter() {parameterName=BaseColor_A, defaultValue=<rgba 0.0083 0 0.0156 1>}
  n2 = VectorParameter() {parameterName=BaseColor_B, defaultValue=<rgba 0 0.0009 0.0208 1>}
  n3 = DynamicParameter() {defaultValue=<rgba 1 0 0 0>, paramNames [AlphaMask_U,UVOffset,Param3,Param4]}
  n4 = TextureCoordinate() {coordinateIndex=1}
  n5 = DynamicParameter() {defaultValue=<rgba 1 0 0 0>, paramNames [AlphaMask_U,UVOffset,Param3,Param4]}
  n6 = SphereMask(
   A=WorldPosition().XYZ
   B=Add(A=ObjectPositionWS(), B=VectorParameter() {parameterName=InitialOffset, defaultValue=<rgba 0 0 -1000 1>}.RGB)
   Radius=ScalarParameter() {defaultValue=2550, parameterName=Radius}
   Hardness=Constant()
  )
  n7 = Multiply(A=@n6, B=@n6)
  n8 = Constant() {r=-200}
  n9 = LinearInterpolate(
   A=Add(A=ComponentMask(None=@n4) {r=true}, B=@n5.UVOffset)
   B=Add(A=Add(A=@n7, B=@n8) {constB=-20}, B=@n5.UVOffset)
   Alpha=ScalarParameter() {defaultValue=1, parameterName=UseWorldSpace_UV, sliderMax=1}
  )
  n10 = Multiply(A=Add(A=@n9, B=@n3.AlphaMask_U)) {constB=7}
  n11 = ComponentMask(None=TextureCoordinate()) {g=true}
  n12 = AppendVector(A=Constant() {r=2.2}, B=Constant())
  n13 = Multiply(A=Time(), B=ScalarParameter() {defaultValue=1, parameterName=TimeMutiply})
  n14 = TextureSample(UVs=Panner(Coordinate=AppendVector(A=@n10, B=@n11), Time=@n13, Speed=@n12) {speedX=-2}) {samplerType=SAMPLERTYPE_Grayscale, texture=&T_Smoke_Trail_Soft}
  n15 = ComponentMask(None=TextureCoordinate()) {r=true}
  n16 = AppendVector(
   A=Multiply(A=@n9, B=ScalarParameter() {defaultValue=22, parameterName=XTiling}) {constB=22}
   B=ComponentMask(None=TextureCoordinate()) {g=true}
  )
  n17 = TextureSampleParameter2D(UVs=Panner(Coordinate=@n16, Time=@n13, Speed=@n12)) {parameterName=TEX_Trail, texture=/Game/Effects/Fort_Effects/Textures/T_InkySmokeTrail}
  n18 = Multiply(A=Add(A=@n17.R, B=@n17.R)) {constB=3}
  n19 = Multiply(A=@n14.R, B=@n14.R)
  n20 = Multiply(
   A=LinearInterpolate(A=LinearInterpolate(A=@n1.RGB, B=@n2.RGB, Alpha=@n14.R), B=Constant(), Alpha=Multiply(A=@n15, B=@n15))
   B=Multiply(A=Add(A=@n18, B=@n19)) {constB=4}
  ) {constA=1}
  n21 = Constant() {r=1.2}
  n22 = Constant() {r=0.5}
  n23 = MaterialFunctionCall(Color_Morning=@n21, Color_Day=@n21, Color_Evening=@n22, Color_Night=@n22) {materialFunction=&Color_Time_of_Day}
  n24 = ParticleColor()
  MaterialFunctionCall(Emissive=Multiply(A=Multiply(A=@n20, B=@n23.Result), B=@n24.RGB)) {materialFunction=&MF_FX_EmissiveAdjustments}
  n25 = Sine(None=Multiply(A=Constant() {r=0.22}, B=Time()))
  n26 = TextureSampleParameter2D(UVs=Panner(Coordinate=@n16, Time=@n13, Speed=@n12) {speedX=-5}) {parameterName=TEX_Trail, texture=&T_InkySmokeTrail}
  n27 = Multiply(A=Add(A=Add(A=@n3.AlphaMask_U, B=@n9)) {constB=0.6}) {constB=9}
  n28 = TextureSample(UVs=Panner(Coordinate=AppendVector(A=@n27, B=@n11), Time=@n13, Speed=@n12) {speedX=-2}) {samplerType=SAMPLERTYPE_Grayscale, texture=&T_Smoke_Trail_Soft}
  n29 = MaterialFunctionCall!(
   Blend=Multiply(A=Add(A=@n26.G, B=@n26.G)) {constB=1.1}
   Base=Multiply(A=MaterialFunctionCall(Blend=@n14.G, Base=@n28.G) {materialFunction=/Engine/Functions/Engine_MaterialFunctions03/Blends/Blend_Lighten}.Result) {constB=1.25}
  ) {materialFunction=&Blend_Lighten}
  n30 = TextureCoordinate()
  n31 = MaterialFunctionCall(Float2=@n30) {materialFunction=&BreakOutFloat2Components}
  n32 = Constant() {r=0.5}
  n33 = SphereMask(
   A=Power(Base=MaterialFunctionCall(Float2=@n30) {materialFunction=&BreakOutFloat2Components}.R) {constExponent=0.15}
   B=Constant() {r=0.5}
  ) {attenuationRadius=0.5, hardnessPercent=0}
  n34 = Multiply(
   A=Multiply(A=SphereMask(A=@n31.G, B=@n32) {attenuationRadius=0.5, hardnessPercent=0}, B=Multiply(A=@n33, B=@n33))
   B=SphereMask(A=TextureCoordinate(), B=Constant2Vector() {g=0.5, r=1.025}) {attenuationRadius=1.05, hardnessPercent=90}
  )
  n35 = SphereMask(A=WorldPosition().XYZ, B=CameraPositionWS(), Radius=Constant() {r=100}, Hardness=Constant() {r=0.2})
  n36 = Multiply(A=Clamp(None=Abs(None=@n25)) {minDefault=0.6}, B=Multiply(A=Multiply(A=@n29.Result, B=@n34), B=OneMinus(None=@n35)))
  n37 = Multiply(A=@n24.A, B=ShadingPathSwitch(Default=DepthFade(Opacity=@n36), Mobile=@n36) {desc=Migrated from FeatureLevelSwitch})
  n38 = MaterialFunctionCall(Float2=LinearInterpolate(
    A=TextureCoordinate()
    B=TextureCoordinate() {coordinateIndex=1}
    Alpha=ScalarParameter() {group=LeadingEdgeMask, parameterName=Use_UV_1, sortPriority=-1}
   )) {materialFunction=&BreakOutFloat2Components}
  n39 = ScalarParameter() {defaultValue=1, group=LeadingEdgeMask, parameterName=LeadingEdgeMask_Tile_U, sortPriority=2}
  n40 = ScalarParameter() {defaultValue=1, group=LeadingEdgeMask, parameterName=LeadingEdgeMask_Tile_V, sortPriority=3}
  n41 = TextureSampleParameter2D(UVs=AppendVector(A=Multiply(A=@n38.R, B=@n39), B=Multiply(A=@n38.G, B=@n40))) {
   group=LeadingEdgeMask
   parameterName=MultipliedTexture_B
   samplerSource=SSM_Wrap_WorldGroupSettings
   sortPriority=1
   texture=/Game/Effects/Fort_Effects/Textures/Hightower_Ribbon_LeadingEdge_Mask
  }
  n42 = LinearInterpolate(A=@n37, B=Multiply(A=@n37, B=@n41.RGB), Alpha=ScalarParameter() {group=LeadingEdgeMask, parameterName=UseLeadingEdgeMask})
  n43 = SphereMask(
   A=PixelDepth().R
   B=Constant()
   Radius=ScalarParameter() {defaultValue=4000, parameterName=NearCameraFadeDistance}
   Hardness=ScalarParameter() {parameterName=NearCameraFadeHardness}
  )
  StaticSwitchParameter(True=Multiply(A=@n42, B=OneMinus(None=@n43)), False=@n42) {parameterName=Use_CameraFade}
 ]
}
