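# Unlit translucent smoke material for contrail particle sprites, drawing on SubUV smoke flipbook masks.
# It fakes lighting with a normal-mapped Fresnel, optionally aimed along a sun/moon and view vector blend, tinted by desaturated sunlight or particle color.
# Opacity combines the smoke mask with Fresnel erosion, particle alpha, depth fade and near-camera fade.
# Dynamic parameters drive erosion and shading; the quality switch feeds identical opacity to Default and Low.
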Material /Game/Effects/Fort_Effects/Materials/M_Contrail_Smoke_Unlit_Simple
graph {
 properties {
  bAutomaticallySetUsageInEditor=false
  bEnableResponsiveAA=true
  bUsedWithNiagaraSprites=true
  bUsedWithParticleSprites=true
  bUsedWithStaticMesh=false
  blendMode=BLEND_Translucent
  opacityMaskClipValue=0.5
  pixelDepthOffsetMode=PDOM_Legacy
  shadingModel=MSM_Unlit
  translucencyLightingMode=TLM_VolumetricPerVertexDirectional
  translucentSelfShadowDensityScale=50
  usedShadingModels=Unlit
  translucentMultipleScatteringExtinction=<rgba 1 1 1 1>
 }
 nodes [
  n1 = DynamicParameter() {defaultValue=<rgba 1 1 0 1>, paramNames [Erode, ErodeExponent, unused, Shading]}
  n2 = ScalarParameter() {defaultValue=0.2, group=Diffuse, parameterName=Diffuse}
  n3 = TextureSampleParameterSubUV!(UVs, Apply View MipBias) {
   group=Textures
   parameterName=Smoke Map SubUV
   samplerType=SAMPLERTYPE_LinearColor
   texture=/Game/Effects/Fort_Effects/Gameplay/SmokeDustExplosions/Textures/T_SmokeDeath_8x8_01
  }
  n4 = Power(
   Base=DotProduct(A=@n3.RGB, B=VectorParameter() {group=Textures, parameterName=DiffuseMaskPicker, defaultValue=<rgba 0 1 0 0>}.RGB)
   Exp=ScalarParameter() {defaultValue=2, group=Diffuse, parameterName=Diffuse Exponent}
  )
  n5 = CollectionParameter() {
   collection=&FortniteMaterialParameters
   parameterName=SunAndMoonModelDirectionalVector
  }
  n6 = ComponentMask(None=@n5) {b=true, g=true, r=true}
  n7 = StaticSwitchParameter(True=Transform(None=Constant3Vector() {constant=<rgba 0 0 1 0>}), False=CameraVectorWS()) {parameterName=Velocity Align Fresnel}
  n8 = ScalarParameter() {defaultValue=0.55, group=Cell Shading Parameters, parameterName=Sun/Moon Directional Shading Blend Smoke}
  n9 = TextureSampleParameterSubUV() {
   group=Normal Map
   parameterName=Normal Map SubUV
   samplerType=SAMPLERTYPE_Normal
   texture=/Game/Effects/Fort_Effects/Gameplay/SmokeDustExplosions/Textures/T_SmokeDeath_8x8_01_Normals
  }
  n10 = ScalarParameter() {defaultValue=1, group=Normal Map, parameterName=Normal U Repeat}
  n11 = ScalarParameter() {defaultValue=1, group=Normal Map, parameterName=Normal V Repeat}
  n12 = TextureSampleParameter2D(UVs=Multiply(A=TextureCoordinate(), B=AppendVector(A=@n10, B=@n11))) {
   group=Normal Map
   parameterName=Normal Map
   samplerType=SAMPLERTYPE_Normal
   texture=/Game/Effects/Fort_Effects/Gameplay/SmokeDustExplosions/Textures/T_SmokePoof_8x8_Normal_01
  }
  n13 = StaticSwitchParameter(True=@n9.RGB, False=@n12.RGB) {defaultValue=true, group=Normal Map, parameterName=Normal Map Particle SubUV}
  n14 = Multiply(A=@n13, B=VectorParameter() {group=Normal Map, parameterName=Normal Map Intensity, defaultValue=<rgba 1 1 1 0>}.RGB)
  n15 = MaterialFunctionCall(Float2=ComponentMask(None=@n13) {g=true, r=true}) {materialFunction=&BreakOutFloat2Components}
  n16 = MaterialFunctionCall(X=@n15.R, Y=OneMinus(None=@n15.G)) {materialFunction=&MakeFloat2}
  n17 = ScalarParameter() {defaultValue=0.1, group=Normal Map, parameterName=Detail Normal Distort}
  n18 = ScalarParameter() {defaultValue=8, group=Normal Map, parameterName=Detail Normal SubUVs U}
  n19 = ScalarParameter() {defaultValue=8, group=Normal Map, parameterName=Detail Normal SubUVs V}
  n20 = Frac(None=Multiply(A=TextureCoordinate(), B=AppendVector(A=@n18, B=@n19)))
  n21 = AppendVector(
   A=ScalarParameter() {defaultValue=0.672, group=Normal Map, parameterName=Detail Normal U Repeat}
   B=ScalarParameter() {defaultValue=0.672, group=Normal Map, parameterName=Detail Normal V Repeat}
  )
  n22 = TextureSampleParameter2D(UVs=Add(A=Multiply(A=@n16.Result, B=@n17), B=Multiply(A=@n20, B=@n21))) {
   group=Normal Map
   parameterName=Detail Normal Map
   samplerType=SAMPLERTYPE_Normal
   texture=/Game/Global/Textures/Default/Noise/T_SwirlNoise_N
  }
  n23 = VectorParameter() {group=Normal Map, parameterName=Detail Normal Map Intensity, defaultValue=<rgba 1.4 1.4 1 0>}
  n24 = StaticSwitchParameter(True=Add(A=@n14, B=Multiply(A=@n22.RGB, B=@n23.RGB)), False=@n14) {group=Normal Map, parameterName=Use Detail Normal Map}
  n25 = StaticSwitchParameter(True=Transform(None=Normalize(VectorInput=@n24)), False=VertexNormalWS()) {defaultValue=true, group=Normal Map, parameterName=Use Normal Map}
  n26 = MaterialFunctionCall(
   Brightness=ScalarParameter() {defaultValue=1, group=Smoke Colors, parameterName=Main Shading Fresnel Brightness}
   Exponent=ScalarParameter() {defaultValue=2, group=Smoke Colors, parameterName=Main Shading Fresnel Exponent}
   Camera Vector=StaticSwitchParameter(True=LinearInterpolate(A=@n6, B=@n7, Alpha=@n8), False=@n7) {defaultValue=true, group=Smoke Colors, parameterName=ToD Shading}
   Vertex / Pixel Normal=@n25
  ) {materialFunction=&MF_Fresnel_FX}
  n27 = TextureSampleParameterSubUV() {
   group=Textures
   parameterName=Smoke Map SubUV
   samplerType=SAMPLERTYPE_LinearColor
   texture=&T_SmokeDeath_8x8_01
  }
  n28 = Power(
   Base=DotProduct(A=@n27.RGB, B=VectorParameter() {group=Textures, parameterName=OpacityMaskPicker, defaultValue=<rgba 1 0 0 0>}.RGB)
   Exp=DynamicParameter() {defaultValue=<rgba 1 1 0 1>, paramNames [Erode, ErodeExponent, unused, Shading]}.ErodeExponent
  )
  n29 = LinearInterpolate(B=LinearInterpolate(A=LinearInterpolate(A=@n2, Alpha=@n4), Alpha=Multiply(A=@n26.Fresnel, B=@n28)) {constA=1}, Alpha=@n1.Shading) {constA=1}
  n30 = Multiply(A=@n1.Shading, B=@n29) {constB=0.5}
  n31 = Desaturation(
   None=CollectionParameter() {collection=&FortniteMaterialParameters, parameterName=SunLightColor}
   Fraction=Constant() {r=0.5}
  )
  n32 = ParticleColor()
  MaterialFunctionCall(Emissive=Multiply(
    A=StaticSwitchParameter(True=@n30, False=@n29) {defaultValue=true, group=Smoke Colors, parameterName=Tri-Color Shading}
    B=StaticSwitchParameter(True=@n31, False=@n32.RGB) {defaultValue=true, group=Smoke Colors, parameterName=ToD Shading}
   )) {materialFunction=&MF_FX_EmissiveAdjustments}
  MaterialFunctionCall() {materialFunction=/Game/Effects/Fort_Effects/Materials/_Functions/CurveColor_Time_of_Day}
  n33 = ScalarParameter() {
   defaultValue=1
   desc="Determines the \"roll\" to erode the mesh with. Keep this in a reasonable range (1 to 8)"
   group=Opacity
   parameterName=Erosion Fresnel Brightness
  }
  n34 = ScalarParameter() {
   defaultValue=1
   desc="Determines the \"roll\" to erode the mesh with. Keep this in a reasonable range (1 to 8)"
   group=Opacity
   parameterName=Erosion Fresnel Exponent
  }
  n35 = MaterialFunctionCall(Brightness=@n33, Exponent=@n34, Vertex / Pixel Normal=@n25) {materialFunction=&MF_Fresnel_FX}
  n36 = DynamicParameter() {defaultValue=<rgba 1 1 0 1>, paramNames [Erode, ErodeExponent, unused, Shading]}
  n37 = DepthFade(Opacity=@n28, FadeDistance=ScalarParameter() {defaultValue=50, group=Opacity, parameterName=Depth Fade Distance})
  n38 = Multiply(
   A=Multiply(
    A=Add(A=@n35.Fresnel, B=@n36.Erode)
    B=StaticSwitchParameter(True=@n37, False=@n28) {defaultValue=true, group=Opacity, parameterName=Use Depth Fade?}
   )
   B=@n28
  )
  Multiply(
   A=MaterialFunctionCall(Distance=ScalarParameter() {defaultValue=512, parameterName="Distance (S)"}) {materialFunction=&Near_Camera_Fade}.Result
   B=Multiply(A=ParticleColor().A, B=QualitySwitch(Default=@n38, Low=@n38) {desc=Cheap opacity on low detail})
  )
  ScalarParameter() {defaultValue=1, group=Opacity, parameterName=Opacity Exponent}
 ]
}
