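# Additive, unlit, two-sided particle material that draws a Viking pattern for contrail effects, usable on mesh particles.
# The pattern samples either rotating UV1 (default) or normal-map-distorted UV0, and its blue channel multiplies particle color, with optional HDR remapping.
# Opacity combines pattern, particle alpha and inverted half-sphere mask, with depth fade off mobile and optional inverse Fresnel.
# Vertex red/green and dynamic radius parameters scale a normal-direction offset.
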Material /Game/Effects/Fort_Effects/Materials/M_Contrail_Viking_Pattern
graph {
 properties {
  bUsedWithMeshParticles=true
  bUsedWithNiagaraMeshParticles=true
  blendMode=BLEND_Additive
  pixelDepthOffsetMode=PDOM_Legacy
  shadingModel=MSM_Unlit
  twoSided=true
  usedShadingModels=Unlit
 }
 nodes [
  n1 = DynamicParameter() {paramNames [UpRadius,DownRadius,Param3,Param4]}
  n2 = VertexColor()
  n3 = VertexNormalWS()
  Add(A=Multiply(A=Multiply(A=@n1.UpRadius, B=@n2.R), B=@n3), B=Multiply(A=@n3, B=Multiply(A=@n1.DownRadius, B=@n2.G)))
  n4 = ParticleColor()
  n5 = TextureCoordinate() {coordinateIndex=1}
  n6 = TextureCoordinate()
  n7 = MaterialFunctionCall(Float2=TextureCoordinate() {uTiling=3, vTiling=0.6}) {materialFunction=&BreakOutFloat2Components}
  n8 = AppendVector(A=@n7.R, B=Power(Base=@n7.G))
  n9 = Multiply(
   A=ComponentMask(None=TextureSample(UVs=@n8) {samplerType=SAMPLERTYPE_Normal, texture=&T_Weap_Distort_N_02}.RGB) {g=true, r=true}
   B=Multiply(
    A=ScalarParameter() {defaultValue=0.12, parameterName=Distortion}
    B=DynamicParameter() {paramNames [UpRadius,DownRadius,Param3,Param4]}.Param4
   )
  ) {constB=0.12}
  n10 = TextureSampleParameter2D(UVs=StaticSwitchParameter(True=Rotator(Coordinate=@n5) {speed=1}, False=Add(A=@n6, B=@n9)) {defaultValue=true, parameterName="UseUVchannel(1)"}) {parameterName=BaseTexture, texture=/Game/Effects/Fort_Effects/Textures/Misc/T_Viking_Pattern02}
  n11 = Multiply(A=@n4.RGB, B=@n10.B)
  n12 = Clamp(None=@n11)
  MaterialFunctionCall(Emissive=StaticSwitchParameter(
    True=MaterialFunctionCall(LDR Input=@n12, HDR Tint=@n11) {materialFunction=&DeriveHDRfromLDR}.RemappedRange
    False=@n4.RGB
   ) {defaultValue=true, parameterName=UseHDR}) {materialFunction=&MF_FX_EmissiveAdjustments}
  Panner(Coordinate=@n8) {speedY=-0.5}
  n13 = TextureSample(UVs=@n5) {texture=/Game/Effects/Fort_Effects/Textures/Mask/T_HalfSphere_D}
  n14 = Multiply(A=@n4.A, B=Multiply(A=@n10.B, B=OneMinus(None=@n13.R)))
  n15 = ShadingPathSwitch(Default=DepthFade(Opacity=@n14, FadeDistance=ScalarParameter() {defaultValue=100, parameterName=FadeDistance}), Mobile=@n14) {desc=Migrated from FeatureLevelSwitch}
  n16 = Fresnel(ExponentIn=ScalarParameter() {defaultValue=1, parameterName=FresnelExponentIn})
  StaticSwitchParameter(True=Multiply(A=@n15, B=OneMinus(None=@n16)), False=@n15) {parameterName=Fresnel}
 ]
}
