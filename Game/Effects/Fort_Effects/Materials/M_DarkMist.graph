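# Translucent, two-sided lit material for a dark, scrolling mist effect, flagged for beam trails and Niagara ribbons.
# Two panning bat-glider noise samples, one on square-root-warped UVs, multiply with particle and dynamic alpha, then a ramp texture erodes the edges.
# The warped noise, masked along U, blends two dark colors for an adjustable glow.
# Depth and near-camera fades soften the alpha, except on mobile, which skips both.
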Material /Game/Effects/Fort_Effects/Materials/M_DarkMist
graph {
 properties {
  bUsedWithBeamTrails=true
  bUsedWithNiagaraRibbons=true
  blendMode=BLEND_Translucent
  pixelDepthOffsetMode=PDOM_Legacy
  twoSided=true
  usedShadingModels=Default Lit
 }
 nodes [
  n1 = TextureCoordinate()
  ComponentMask(None=@n1) {r=true}
  MaterialFunctionCall(Offset Amount=Constant() {r=-50}) {materialFunction=/Engine/Functions/Engine_MaterialFunctions02/WorldPositionOffset/CameraOffset}
  n2 = VectorParameter() {parameterName=Color1, defaultValue=<rgba 0.02 0.0161 0.016 1>}
  n3 = VectorParameter() {parameterName=Color2, defaultValue=<rgba 0.15 0.0428 0.0398 1>}
  n4 = SquareRoot(None=@n1)
  n5 = TextureSample(UVs=Panner(Coordinate=SquareRoot(None=@n4)) {speedX=-1, speedY=-0.15}) {samplerType=SAMPLERTYPE_Masks, texture=/Game/Gadgets/Assets/VinderTech_GliderChute/Glider_Vampire/Textures/T_Glider_Bat_Noises}
  n6 = Multiply(A=@n5.B, B=ComponentMask(None=TextureCoordinate()) {r=true})
  MaterialFunctionCall(Emissive=Multiply(A=LinearInterpolate(A=@n2.RGB, B=@n3.RGB, Alpha=@n6), B=ScalarParameter() {defaultValue=0.25, parameterName=Emissive}) {constB=0.25}) {materialFunction=&MF_FX_EmissiveAdjustments}
  n7 = DepthFade(FadeDistance=ScalarParameter() {defaultValue=200, parameterName=DF Distance}) {fadeDistanceDefault=50}
  n8 = MaterialFunctionCall(Distance=ScalarParameter() {defaultValue=200, parameterName=CF Distance}) {materialFunction=&Near_Camera_Fade}
  n9 = TextureSample(UVs=Panner(Coordinate=TextureCoordinate() {uTiling=0.5, vTiling=0.5}) {speedX=-0.5, speedY=-0.1}) {samplerType=SAMPLERTYPE_Masks, texture=&T_Glider_Bat_Noises}
  n10 = ParticleColor()
  n11 = DynamicParameter() {paramNames [Alpha,Param2,Param3,Param4]}
  n12 = Subtract(
   A=Multiply(A=Multiply(A=@n9.B, B=@n5.B), B=Multiply(A=@n10.A, B=@n11.Alpha))
   B=OneMinus(None=TextureSample() {texture=/Game/Global/Textures/Default/Masks_and_Gradients/T_Water_Top_Ramp}.R)
  )
  ShadingPathSwitch(Default=Multiply(A=Multiply(A=@n7, B=@n8.Result), B=@n12), Mobile=@n12) {desc=Migrated from FeatureLevelSwitch}
  TextureCoordinate() {coordinateIndex=1, uTiling=0.5, vTiling=0.5}
  TextureSample() {texture=/Game/Effects/Fort_Effects/Textures/Ramps/T_Husk_Tornado_Ramp}
 ]
}
