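# M_Electric_Bolt is an unlit, additive effects material for mesh particles, usable by both Cascade and Niagara emitters.
# It passes the particle color RGB into the Emissive input of the shared MF_FX_EmissiveAdjustments material function.
# It samples no textures of its own, so the emitter's particle color drives the look.
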Material /Game/Effects/Fort_Effects/Materials/M_Electric_Bolt
graph {
 properties {
  bUsedWithMeshParticles=true
  bUsedWithNiagaraMeshParticles=true
  blendMode=BLEND_Additive
  pixelDepthOffsetMode=PDOM_Legacy
  shadingModel=MSM_Unlit
  usedShadingModels=Unlit
 }
 nodes [MaterialFunctionCall(Emissive=ParticleColor().RGB) {materialFunction=&MF_FX_EmissiveAdjustments}]
}
