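# Translucent unlit sprite material for flipbook fire particles, named for the Evil Cowboy effect and usable in Cascade and Niagara.
# Squared green channel of the SubUV texture, tinted by particle color, is scaled per time of day, then passed through MF_FX_EmissiveAdjustments for glow.
# Opacity combines particle and texture alpha with a powered near-camera fade.
# A shading path switch skips depth fade on mobile.
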Material /Game/Effects/Fort_Effects/Materials/M_EvilCowboy_FireSubUV
graph {
 properties {
  bUsedWithNiagaraSprites=true
  bUsedWithParticleSprites=true
  blendMode=BLEND_Translucent
  pixelDepthOffsetMode=PDOM_Legacy
  shadingModel=MSM_Unlit
  usedShadingModels=Unlit
 }
 nodes [
  n1 = ScalarParameter() {defaultValue=1, parameterName=Color_morning}
  n2 = ParticleColor()
  n3 = TextureSampleParameterSubUV!(UVs=TextureCoordinate(), Apply View MipBias) {parameterName=FireSubUV, texture=/Game/Effects/Fort_Effects/Textures/Fire/T_FireFluid_1_8X8}
  n4 = Multiply(A=@n2.RGB, B=Multiply(A=@n3.G, B=@n3.G))
  n5 = ScalarParameter() {defaultValue=1, parameterName=Color_Day}
  n6 = ScalarParameter() {defaultValue=1, parameterName=Color_Evening}
  n7 = ScalarParameter() {defaultValue=1, parameterName=Color_Night}
  MaterialFunctionCall(Emissive=MaterialFunctionCall(
    Color_Morning=Multiply(A=@n1, B=@n4)
    Color_Day=Multiply(A=@n5, B=@n4)
    Color_Evening=Multiply(A=@n6, B=@n4)
    Color_Night=Multiply(A=@n4, B=@n7)
   ) {materialFunction=&Color_Time_of_Day}.Result) {materialFunction=&MF_FX_EmissiveAdjustments}
  n8 = Multiply(A=@n2.A, B=@n3.A)
  n9 = MaterialFunctionCall(Distance=ScalarParameter() {defaultValue=128, parameterName=NearCamFade_Distance}) {materialFunction=&Near_Camera_Fade}
  n10 = ScalarParameter() {defaultValue=15, parameterName=NearCamFade_Power}
  Multiply(
   A=ShadingPathSwitch(Default=DepthFade(Opacity=@n8), Mobile=@n8) {desc="Added pref for mobile,Yang."}
   B=Clamp(None=Power(Base=@n9.Result, Exp=@n10)) {minDefault=0.02}
  )
 ]
}
