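# Unlit translucent particle material for falling vapor, built from a vertically panning tileable smoke texture tinted by particle color.
# A first smoke sample drives both bump-offset UV distortion of the second and a noisy circular mask radius, with dynamic-parameter UV offsets.
# Opacity combines that mask with depth fade, near-camera fade and a switchable world-Z gradient.
# The same switch negates world position for particles below minimum Z, collapsing them.
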Material /Game/Effects/Fort_Effects/Materials/M_FallingVapor
graph {
 properties {
  bUsedWithNanite=true
  bUsedWithNiagaraSprites=true
  bUsedWithParticleSprites=true
  blendMode=BLEND_Translucent
  pixelDepthOffsetMode=PDOM_Legacy
  shadingModel=MSM_Unlit
  usedShadingModels=Unlit
 }
 nodes [
  Constant() {r=0.45}
  n1 = TextureCoordinate()
  n2 = DynamicParameter()
  n3 = TextureSample(UVs=Add(A=Panner(Coordinate=@n1) {speedY=0.05}, B=AppendVector(A=@n2.Param1, B=@n2.Param1))) {texture=/Game/Global/Textures/Default/Tileable_Patterns/T_Smoke_Puff_Tile_Up}
  n4 = ScalarParameter() {defaultValue=0.1, parameterName=Height}
  n5 = DynamicParameter()
  n6 = Add(A=Panner(Coordinate=BumpOffset(Height=@n3.R, HeightRatioInput=@n4)) {speedY=0.05}, B=AppendVector(A=@n5.Param1, B=@n5.Param1))
  n7 = ParticleColor()
  MaterialFunctionCall(Emissive=Multiply(A=TextureSample(UVs=@n6) {texture=&T_Smoke_Puff_Tile_Up}.R, B=@n7.RGB)) {materialFunction=&MF_FX_EmissiveAdjustments}
  n8 = SphereMask(
   A=TextureCoordinate()
   B=Constant() {r=0.5}
   Radius=LinearInterpolate(A=ScalarParameter() {defaultValue=0.4, parameterName=Min}, B=ScalarParameter() {defaultValue=0.6, parameterName=Max}, Alpha=@n3.R)
  ) {hardnessPercent=0}
  n9 = Multiply(A=@n7.A, B=@n8)
  n10 = LinearInterpolate(Alpha=@n8) {constA=150, constB=5}
  n11 = ComponentMask(None=WorldPosition().XYZ) {b=true}
  n12 = ScalarParameter() {defaultValue=12892, parameterName=World Z Min}
  n13 = Saturate(None=Divide(A=Subtract(A=@n11, B=@n12), B=ScalarParameter() {defaultValue=5000, parameterName=World Z Length}))
  n14 = Constant() {r=1}
  Multiply(
   A=Multiply(
    A=DepthFade(Opacity=@n9, FadeDistance=@n10)
    B=StaticSwitchParameter(True=@n13, False=@n14) {defaultValue=true, parameterName=UseWPOHeightMask}
   )
   B=MaterialFunctionCall(Distance=ScalarParameter() {defaultValue=512, parameterName=CameraFade_Distance}) {materialFunction=&Near_Camera_Fade}.Result
  )
  ScalarParameter() {defaultValue=1, parameterName=Opacity}
  n15 = ParticlePositionWS()
  n16 = WorldPosition()
  StaticSwitchParameter(
   True=If(
    A=ComponentMask(None=@n15) {b=true}
    B=ScalarParameter() {defaultValue=12892, parameterName=World Z Min}
    "A > B"=Constant()
    "A < B"=Multiply(A=@n16.XYZ) {constB=-1}
   )
   False=Constant()
  ) {defaultValue=true, parameterName=UseWPOHeightMask}
  TextureCoordinate()
 ]
}
