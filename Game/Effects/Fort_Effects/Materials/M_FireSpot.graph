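# Unlit translucent particle sprite material that draws a small distorted fire spot from a pilot-light texture.
# A panning noise normal map, shifted per particle by offsetX, distorts the flame UVs.
# The flame's green channel, sharpened by the dynamic Harden value, blends two colors tinted by particle color and also shapes opacity.
# Opacity gets a per-particle depth fade except at low quality, and a boosted color feeds the shared FX emissive adjustments.
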Material /Game/Effects/Fort_Effects/Materials/M_FireSpot
graph {
 properties {
  bUsedWithNanite=true
  bUsedWithParticleSprites=true
  blendMode=BLEND_Translucent
  pixelDepthOffsetMode=PDOM_Legacy
  shadingModel=MSM_Unlit
  usedShadingModels=Unlit
 }
 nodes [
  n1 = DynamicParameter() {defaultValue=<rgba 3 4 -0.25 1>, paramNames [offsetX,offsetY,Harden,FadeDistance]}
  n2 = Panner(Coordinate=TextureCoordinate() {uTiling=0.2, vTiling=0.2}) {speedX=0.1, speedY=0.6}
  n3 = DynamicParameter() {defaultValue=<rgba 3 4 -0.25 1>, paramNames [offsetX,offsetY,Harden,FadeDistance]}
  n4 = AppendVector(A=@n3.offsetX, B=@n3.offsetX)
  n5 = ComponentMask(None=TextureSample(UVs=Add(A=@n2, B=@n4)) {samplerType=SAMPLERTYPE_Normal, texture=/Game/Global/Textures/Default/Noise/T_Noise_Normal_01}.RGB) {g=true, r=true}
  n6 = ScalarParameter() {defaultValue=0.18, parameterName=UVdistort}
  n7 = TextureSample(UVs=Add(A=TextureCoordinate(), B=Multiply(A=@n5, B=@n6))) {samplerSource=SSM_Clamp_WorldGroupSettings, texture=/Game/Effects/Fort_Effects/Textures/Fire/T_Fire_PilotLight}
  n8 = Clamp(None=LinearInterpolate(A=@n1.Harden, B=Add(A=@n1.Harden) {constB=20}, Alpha=@n7.G) {constA=-0.5, constB=8, desc="Lerp -1 to 0 on 0, 31 to 1 on 1"})
  n9 = Multiply(
   A=ParticleColor().RGB
   B=LinearInterpolate(
    A=VectorParameter() {parameterName=Color0, defaultValue=<rgba 0.1458 0.0678 0.0093 0>}.RGB
    B=VectorParameter() {parameterName=Color1, defaultValue=<rgba 9 1.2342 0.4271 0>}.RGB
    Alpha=@n8
   )
  )
  n10 = MaterialFunctionCall(LDR Input=Clamp(None=@n9), HDR Tint=@n9) {materialFunction=&DeriveHDRfromLDR}
  n11 = Multiply(A=@n9, B=ScalarParameter() {defaultValue=1.5, parameterName=LOW_ColorBoost})
  FeatureLevelSwitch(Default=@n10.RemappedRange, ES2=@n11, ES3_1=@n11, SM4_REMOVED=@n10.RemappedRange)
  MaterialFunctionCall(Emissive=@n11) {materialFunction=&MF_FX_EmissiveAdjustments}
  n12 = Multiply(A=@n8, B=Multiply(A=@n7.G, B=ParticleColor().A))
  QualitySwitch(
   Default=DepthFade(
    Opacity=@n12
    FadeDistance=DynamicParameter() {defaultValue=<rgba 3 4 -0.25 1>, paramNames [offsetX,offsetY,Harden,FadeDistance]}.FadeDistance
   ) {fadeDistanceDefault=8}
   Low=@n12
  )
 ]
}
