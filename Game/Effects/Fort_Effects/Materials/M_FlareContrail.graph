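# Translucent unlit beam-trail material for a flare contrail, with a glowing head and noisy tail.
# Particle time slides a warped starburst texture and gradient along the trail, setting red-to-hot-orange emissive and, with two panning noise layers, alpha.
# Dynamic parameters or particle color channels drive timing, offset and randomness, and InvertU flips direction.
# Optional panning-noise world position offset wobbles the trail, partly scaled by a remapped particle-to-world vector length.
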Material /Game/Effects/Fort_Effects/Materials/M_FlareContrail
graph {
 properties {
  bUsedWithBeamTrails=true
  blendMode=BLEND_Translucent
  pixelDepthOffsetMode=PDOM_Legacy
  shadingModel=MSM_Unlit
  twoSided=true
  usedShadingModels=Unlit
 }
 nodes [
  Frac(None=Time())
  n1 = TextureCoordinate()
  n2 = DynamicParameter() {defaultValue=<rgba 0.5 1 1 1>, paramNames [ParticleTime,CurveU,RandomV,Param4]}
  n3 = ParticleColor()
  n4 = ScalarParameter() {parameterName=UseParticleColorForDynamics}
  n5 = LinearInterpolate(A=@n2.RandomV, B=@n3.B, Alpha=@n4)
  n6 = AppendVector(A=@n5, B=@n5)
  n7 = TextureSample(UVs=Panner(Coordinate=Add(A=@n1, B=@n6)) {speedX=-0.5}) {samplerType=SAMPLERTYPE_LinearColor, texture=&HighContrastNoise}
  n8 = ComponentMask(None=Add(A=Multiply(A=@n7.RGB) {constB=4}) {constB=-2}) {g=true, r=true}
  n9 = MaterialFunctionCall(Float2=@n8) {materialFunction=&BreakOutFloat2Components}
  n10 = Add(A=Multiply(A=Add(A=@n9.R, B=@n9.R)) {constB=0.4})
  n11 = Constant3Vector() {constant=<rgba 1 0 0 0>}
  n12 = MaterialFunctionCall(
   Input=MaterialFunctionCall(Vector 3=TransformPosition(None=@n11) {transformSourceType=TRANSFORMPOSSOURCE_Particle, transformType=TRANSFORMPOSSOURCE_World}) {materialFunction=/Engine/Functions/Engine_MaterialFunctions02/Utility/VectorLength}.V3 Length
   Input Low=Constant() {r=40}
   Input High=Constant() {r=50}
   Target Low=ScalarParameter() {parameterName=TargetLow}
   Target High=ScalarParameter() {defaultValue=1, parameterName=TargetHigh}
  ) {materialFunction=&RemapValueRange}
  n13 = Multiply(A=Add(A=Clamp(None=@n12.Result)) {constB=0.3}) {constB=30}
  LinearInterpolate(B=AppendVector(A=@n8, B=Multiply(A=@n10, B=@n13)), Alpha=ScalarParameter() {defaultValue=1, parameterName=UseWPO})
  n14 = TextureCoordinate()
  n15 = OneMinus(None=@n14)
  LinearInterpolate(
   A=@n14
   B=AppendVector(A=ComponentMask(None=@n14) {r=true}, B=ComponentMask(None=@n15) {g=true})
   Alpha=ScalarParameter() {parameterName=InvertU}
  )
  n16 = LinearInterpolate(
   A=DynamicParameter() {defaultValue=<rgba 0.5 1 1 1>, paramNames [ParticleTime,CurveU,RandomV,Param4]}.ParticleTime
   B=ParticleColor().R
   Alpha=ScalarParameter() {parameterName=UseParticleColorForDynamics}
  )
  n17 = OneMinus(None=@n16)
  n18 = ScalarParameter() {parameterName=InvertU}
  n19 = Add(A=Multiply(A=LinearInterpolate(A=@n16, B=@n17, Alpha=@n18)) {constB=2}) {constB=-2}
  n20 = Constant2Vector() {r=0.75}
  n21 = Constant2Vector() {g=0.5, r=1}
  n22 = Multiply(A=TextureCoordinate(), B=@n21)
  n23 = Multiply(A=@n21, B=LinearInterpolate(B=Constant2Vector() {g=0.5, r=1}, Alpha=ScalarParameter() {parameterName=InvertU}) {constA=0.5, constB=-0.5}) {constB=0.5}
  n24 = Clamp(None=Add(A=Multiply(A=@n19, B=@n20), B=Add(A=@n22, B=@n23) {constB=0.25}))
  n25 = OneMinus(None=@n24)
  n26 = LinearInterpolate(
   A=@n25
   B=AppendVector(A=ComponentMask(None=@n24) {r=true}, B=ComponentMask(None=@n25) {g=true})
   Alpha=ScalarParameter() {parameterName=InvertU}
  )
  n27 = TextureSample(UVs=Power(Base=@n26, Exp=Constant2Vector() {g=1, r=8}) {constExponent=4}) {texture=/Game/Effects/Fort_Effects/Textures/Flares/T_Starburst}
  n28 = Constant2Vector() {r=0.075}
  n29 = Frac(None=Clamp(None=Add(A=@n26, B=@n28) {constA=2, constB=-0.5}))
  n30 = Power(Base=ComponentMask(None=@n29) {r=true})
  n31 = Add(A=@n27.G, B=Multiply(A=@n30) {constB=0.25})
  n32 = Multiply(
   A=LinearInterpolate(A=Constant3Vector() {constant=<rgba 1 0.0909 0.026 0>}, B=Constant3Vector() {constant=<rgba 10 4.534 0.89 0>}, Alpha=@n31)
   B=@n31
  )
  n33 = ParticleColor()
  n34 = Multiply(A=@n32, B=@n33.RGB)
  n35 = ScalarParameter() {defaultValue=1, parameterName=UseParticleColorForColor}
  MaterialFunctionCall(Emissive=Multiply(A=LinearInterpolate(A=@n32, B=@n34, Alpha=@n35), B=ScalarParameter() {defaultValue=1, parameterName=Brightness})) {materialFunction=&MF_FX_EmissiveAdjustments}
  n36 = Multiply(A=@n30, B=OneMinus(None=@n30))
  n37 = ComponentMask(None=@n29) {g=true}
  n38 = OneMinus(None=@n37)
  n39 = LinearInterpolate(A=@n2.CurveU, B=@n3.G, Alpha=@n4)
  n40 = AppendVector(A=LinearInterpolate(A=@n39, B=OneMinus(None=@n39), Alpha=ScalarParameter() {parameterName=InvertU}), B=@n5)
  n41 = Constant2Vector() {g=0.03, r=0.2}
  n42 = ScalarParameter() {defaultValue=1, parameterName=Speed}
  n43 = LinearInterpolate(A=@n42, B=Multiply(A=@n42) {constB=-1}, Alpha=ScalarParameter() {parameterName=InvertU}) {constA=1, constB=-1}
  n44 = Multiply(A=Constant2Vector() {g=1, r=5}, B=Panner(Coordinate=Add(A=@n40, B=@n1), Speed=Multiply(A=@n41, B=@n43) {constB=0.2}) {speedX=0.2})
  n45 = TextureObject() {samplerType=SAMPLERTYPE_LinearColor, texture=/Game/Global/Textures/Default/Noise/T_Noise01_Inverse}
  n46 = Constant2Vector() {g=-0.01, r=0.4}
  n47 = Multiply(A=Panner(Speed=Multiply(A=@n43, B=@n46) {constB=0.4}) {speedX=0.75}, B=Constant2Vector() {g=1, r=3})
  n48 = Multiply(
   A=Multiply(A=Add(A=@n36) {constB=0}, B=Multiply(A=@n37, B=@n38))
   B=Add(
    A=TextureSample(UVs=@n44, Tex=@n45) {samplerType=SAMPLERTYPE_Grayscale, texture=/Game/Global/Textures/Default/Noise/TilingNoise06}.G
    B=TextureSample(UVs=@n47, Tex=@n45) {samplerType=SAMPLERTYPE_Grayscale, texture=&TilingNoise06}.G
   )
  )
  Multiply(A=@n33.A, B=Add(A=@n27.G, B=Multiply(A=@n48) {constB=10}))
 ]
}
