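# Unlit translucent particle material for a flash spike effect, usable on Cascade and Niagara sprites.
# Particle color starts boosted and fades to black over particle life, tinted by the hue-shifted, optionally desaturated spike texture.
# A static switch swaps that tint for a time-panned rainbow gradient with zero-tiled UVs.
# Mobile skips the fade, color passes through shared emissive adjustments, and texture alpha multiplies particle alpha.
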Material /Game/Effects/Fort_Effects/Materials/M_Flash_Spike_01
graph {
 properties {
  bUsedWithNanite=true
  bUsedWithNiagaraSprites=true
  bUsedWithParticleSprites=true
  blendMode=BLEND_Translucent
  pixelDepthOffsetMode=PDOM_Legacy
  shadingModel=MSM_Unlit
  usedShadingModels=Unlit
 }
 nodes [
  n1 = ParticleColor()
  n2 = LinearInterpolate(A=Multiply(B=@n1.RGB) {constA=8}, B=Multiply(A=@n1.RGB) {constB=0}, Alpha=ParticleRelativeTime()) {constA=6}
  n3 = ComponentMask(None=TextureCoordinate() {uTiling=0, vTiling=0}) {g=true}
  n4 = TextureSample() {texture=/Game/Effects/Fort_Effects/Textures/Explo/T_Flash_Spike_01}
  n5 = MaterialFunctionCall(Hue Shift Percentage=ScalarParameter() {parameterName=HueShiftPercentage}, Texture=@n4.RGB) {materialFunction=&HueShift}
  n6 = StaticSwitchParameter(
   True=TextureSample(UVs=Panner(Coordinate=@n3) {speedX=1.6667}) {texture=/Game/Global/Textures/Default/Masks_and_Gradients/T_Grad_Rainbow}.RGB
   False=StaticSwitchParameter(True=Desaturation(None=@n5.Result), False=@n5.Result) {parameterName=Desaturate Texture}
  ) {parameterName=Use Rainbow}
  n7 = Multiply(A=@n2, B=@n6)
  n8 = ScalarParameter() {defaultValue=1, parameterName=MultiplyTexture}
  MaterialFunctionCall(Emissive=ShadingPathSwitch(Default=LinearInterpolate(A=@n2, B=@n7, Alpha=@n8), Mobile=Multiply(A=@n1.RGB, B=@n6)) {desc=Migrated from FeatureLevelSwitch}) {materialFunction=&MF_FX_EmissiveAdjustments}
  Multiply(A=@n4.A, B=ParticleColor().A)
 ]
}
