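# Additive, unlit, two-sided effect material, named for a generator sphere, with a smoky look and a warm orange color fed through MF_FX_EmissiveAdjustments.
# Two slowly panning samples of T_Night_Smoke at different tilings add to a faint base and a pixel-depth-driven term.
# That mask is multiplied by a screen-position fade and a camera-vector-versus-world-up term, so intensity varies with viewing elevation.
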Material /Game/Effects/Fort_Effects/Materials/M_Generator_Sphere
graph {
 properties {blendMode=BLEND_Additive, pixelDepthOffsetMode=PDOM_Legacy, shadingModel=MSM_Unlit, twoSided=true, usedShadingModels=Unlit}
 nodes [
  Constant() {r=0.5}
  MaterialFunctionCall(Emissive=Constant3Vector() {constant=<rgba 1 0.7 0.3 0>}) {materialFunction=&MF_FX_EmissiveAdjustments}
  n1 = TextureCoordinate() {uTiling=5, vTiling=5}
  n2 = AppendVector(A=Constant() {r=0.4}, B=Constant() {r=0.4})
  n3 = TextureSample(UVs=Panner(Coordinate=Add(A=@n1, B=@n2)) {speedX=0.01}) {texture=/Game/Effects/Fort_Effects/Textures/T_Night_Smoke}
  n4 = TextureSample(UVs=Panner() {speedX=0.01}) {texture=&T_Night_Smoke}
  n5 = Divide(A=Subtract(B=PixelDepth().R), B=Constant() {r=300})
  n6 = Constant() {r=0.5}
  n7 = OneMinus(None=Clamp(None=Power(Base=@n5, Exp=@n6)))
  n8 = Constant() {r=0.2}
  n9 = Add(A=Multiply(A=Add(A=@n3.RGB, B=@n4.RGB), B=Constant() {r=0.02}), B=Add(A=Constant() {r=0.018}, B=Multiply(A=@n7, B=@n8)))
  n10 = ScreenPosition()
  n11 = Power(Base=Multiply(A=ComponentMask(None=@n10.ViewportUV) {b=true}, B=Constant() {r=0.0032}), Exp=Constant() {r=1.5})
  n12 = Constant()
  n13 = Constant() {r=1}
  n14 = DotProduct(A=CameraVectorWS(), B=Constant3Vector() {constant=<rgba 0 0 1 0>})
  n15 = Constant() {r=0.8}
  Multiply(
   A=Multiply(A=Clamp(None=@n9), B=Clamp(None=@n11, Min=@n12, Max=@n13))
   B=Clamp(None=Power(Base=@n14, Exp=@n15), Min=Constant(), Max=Constant() {r=1})
  )
 ]
}
