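# Additive unlit particle material for glider contrail engine streaks, built for sprites, Niagara ribbons and mesh particles.
# Four panning caustics samples at different tilings blend into a flame pattern, sharpened by FlameAmount.
# A spark mask and particle color tint and shape it, with a brighter variant on low quality.
# Optional refraction follows the flame, and depth fade is computed, both falling back to constants on mobile and low quality.
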Material /Game/Effects/Fort_Effects/Materials/M_GliderContrail_Engine_Mid_Streaks
graph {
 properties {
  bEnableResponsiveAA=true
  bUsedWithNiagaraMeshParticles=true
  bUsedWithNiagaraRibbons=true
  bUsedWithNiagaraSprites=true
  bUsedWithParticleSprites=true
  blendMode=BLEND_Additive
  pixelDepthOffsetMode=PDOM_Legacy
  refractionMethod=RM_IndexOfRefraction
  shadingModel=MSM_Unlit
  usedShadingModels=Unlit
 }
 nodes [
  n1 = TextureSampleParameter2D(UVs=Panner(Coordinate=TextureCoordinate() {vTiling=0.25}) {speedX=-0.5, speedY=1}) {parameterName=TexA, texture=&T_ShieldCaustics}
  n2 = TextureSampleParameter2D(UVs=Panner(Coordinate=TextureCoordinate() {uTiling=2, vTiling=0.5}) {speedX=1, speedY=2}) {parameterName=TexB, texture=&T_ShieldCaustics}
  n3 = TextureSampleParameter2D(UVs=Panner(Coordinate=TextureCoordinate() {uTiling=0.5, vTiling=0.25}) {speedX=-0.5, speedY=1}) {parameterName=TexC, texture=&T_ShieldCaustics}
  n4 = TextureSampleParameter2D(UVs=Panner(Coordinate=TextureCoordinate() {vTiling=0.5}) {speedX=1, speedY=2}) {parameterName=TexD, texture=&T_ShieldCaustics}
  n5 = Power(
   Base=LinearInterpolate(A=LinearInterpolate(A=@n1.G, B=@n2.G), B=LinearInterpolate(A=@n3.G, B=@n4.G))
   Exp=ScalarParameter() {defaultValue=1, parameterName=FlameAmount}
  )
  n6 = TextureSampleParameter2D() {parameterName=Mask, texture=/Game/Weapons/FORT_Rifles/Effects/Textures/spark}
  n7 = ParticleColor()
  n8 = Multiply(A=@n6.R, B=@n7.RGB)
  n9 = Multiply(A=Multiply(A=Constant() {r=1.6}, B=@n6.R), B=@n7.RGB)
  Multiply(A=@n5, B=Multiply(A=QualitySwitch(Default=@n8, Low=@n9), B=@n7.A))
  n10 = LinearInterpolate(Alpha=Multiply(A=Multiply(A=@n5, B=@n7.A), B=ScalarParameter() {defaultValue=1, parameterName=RefractionAmount})) {constA=1, constB=1.2}
  n11 = Constant() {r=1}
  QualitySwitch(
   Default=ShadingPathSwitch(Default=StaticSwitchParameter(True=@n10, False=@n11) {parameterName=UseRefraction}, Mobile=@n11) {desc=Migrated from FeatureLevelSwitch}
   Low=@n11
   Medium=@n11
  )
  n12 = ScalarParameter() {parameterName=FadeDistance}
  n13 = Constant() {r=1}
  QualitySwitch(Default=ShadingPathSwitch(Default=DepthFade(FadeDistance=@n12), Mobile=@n13) {desc=Migrated from FeatureLevelSwitch}, Low=@n13)
 ]
}
