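# Unlit, two-sided translucent particle material for glider contrail wheel airflow, built on a panning distortion texture.
# The texture blends dark and light colors, tinted by particle color, and drives opacity with a second differently panned sample.
# Opacity combines per-face axis-aligned Fresnel using dynamic-parameter exponents, UV gradient masks, optional edge and Fresnel fades, particle alpha and time-of-day dimming.
# Shading-path switches omit depth fade and simplify Fresnel on mobile.
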Material /Game/Effects/Fort_Effects/Materials/M_GliderContrail_WheelAirflow
graph {
 properties {
  bUsedWithMeshParticles=true
  bUsedWithNiagaraMeshParticles=true
  bUsedWithNiagaraSprites=true
  blendMode=BLEND_Translucent
  pixelDepthOffsetMode=PDOM_Legacy
  shadingModel=MSM_Unlit
  twoSided=true
  usedShadingModels=Unlit
 }
 nodes [
  n1 = Constant() {r=1}
  n2 = Multiply(
   A=TextureCoordinate()
   B=AppendVector(
    A=ScalarParameter() {defaultValue=1, group=Tiling, parameterName=X Tiling, sliderMax=1, sortPriority=11}
    B=ScalarParameter() {defaultValue=0.5, group=Tiling, parameterName=Y Tiling, sliderMax=1, sortPriority=12}
   )
  )
  n3 = DynamicParameter() {parameterIndex=1, defaultValue=<rgba 1 1 20 0>, paramNames [XOffset, YOffset, "", ""]}
  n4 = Add(A=@n2, B=AppendVector(A=@n3.XOffset, B=@n3.XOffset))
  n5 = ScalarParameter() {desc=Uses Paramter Index 1 R/G for X/Y, group=Tiling, parameterName=Use Per Particle UV Offset, sortPriority=15}
  n6 = ScalarParameter() {group=Speed, parameterName=X Speed, sortPriority=3}
  n7 = ScalarParameter() {defaultValue=-5, group=Speed, parameterName=Y Speed, sortPriority=4}
  n8 = TextureSampleParameter2D(UVs=Panner(Coordinate=LinearInterpolate(A=@n2, B=@n4, Alpha=@n5), Speed=AppendVector(A=@n6, B=@n7)) {speedY=-5}) {
   group=Color and Texture
   parameterName=OpacityTexture
   sortPriority=0
   texture=/Game/Global/Textures/Default/Masks_and_Gradients/T_BlasterBeamDistort
  }
  n9 = Multiply(B=Power(Base=@n8.R, Exp=ScalarParameter() {defaultValue=1, parameterName=OpacityPower})) {constA=1}
  n10 = DynamicParameter() {defaultValue=<rgba 1 1 20 0>, paramNames [InnerFresnelExp, Ref, WPO, OuterFresnelExp]}
  n11 = VertexNormalWS()
  n12 = Multiply(A=@n11) {constB=-1}
  n13 = TwoSidedSign()
  n14 = LinearInterpolate(
   A=Multiply(A=MaterialFunctionCall!(Axis, Exponent=@n10.InnerFresnelExp, Normal=@n12) {materialFunction=/Game/Effects/Fort_Effects/Materials/_Functions/MF_AxisAlignedFrensnel}.Fresnel) {constB=1.5}
   B=Multiply(A=MaterialFunctionCall(Exponent=@n10.OuterFresnelExp, Normal=@n11) {materialFunction=&MF_AxisAlignedFrensnel}.Fresnel) {constB=0.7}
   Alpha=@n13
  )
  n15 = Power(Base=MaterialFunctionCall(Float2=TextureCoordinate()) {materialFunction=&BreakOutFloat2Components}.G) {constExponent=0.8}
  n16 = Constant() {r=0.5}
  n17 = Multiply(
   A=LinearInterpolate(
    B=ShadingPathSwitch(Default=@n14, Mobile=@n13) {desc=Migrated from FeatureLevelSwitch}
    Alpha=ScalarParameter() {defaultValue=1, group=Fade Controls, parameterName=UseFresnel02, sortPriority=18}
   ) {constA=1}
   B=LinearInterpolate(Alpha=SphereMask(A=@n15, B=@n16) {attenuationRadius=0.5, hardnessPercent=1}) {constB=0.8}
  )
  n18 = ScalarParameter() {defaultValue=5, group=Fade Controls, parameterName=Depth Fade Distance, sortPriority=10}
  n19 = Multiply(
   A=TextureCoordinate()
   B=AppendVector(
    A=ScalarParameter() {defaultValue=1, group=Tiling, parameterName=Fade X Tiling, sliderMax=1, sortPriority=13}
    B=ScalarParameter() {defaultValue=1, group=Tiling, parameterName=Fade Y Tiling, sliderMax=1, sortPriority=14}
   )
  )
  n20 = DynamicParameter() {parameterIndex=1, defaultValue=<rgba 1 1 20 0>, paramNames [XOffset, YOffset, "", ""]}
  n21 = Add(A=@n19, B=AppendVector(A=@n20.XOffset, B=@n20.XOffset))
  n22 = ScalarParameter() {desc=Uses Paramter Index 1 R/G for X/Y, group=Tiling, parameterName=Use Per Particle UV Offset, sortPriority=15}
  n23 = TextureSampleParameter2D(UVs=Panner(Coordinate=LinearInterpolate(A=@n19, B=@n21, Alpha=@n22)) {speedX=0.3, speedY=-0.8}) {
   group=Color and Texture
   parameterName=OpacityTexture
   sortPriority=0
   texture=&T_BlasterBeamDistort
  }
  n24 = TextureCoordinate()
  n25 = Multiply(
   A=ShadingPathSwitch(Default=DepthFade(Opacity=@n17, FadeDistance=@n18), Mobile=@n17) {desc=Migrated from FeatureLevelSwitch}
   B=LinearInterpolate(
    B=@n23.R
    Alpha=MaterialFunctionCall(Float2=@n24) {materialFunction=&BreakOutFloat2Components}.G
   ) {constA=1}
  )
  n26 = ScalarParameter() {
   defaultValue=1
   desc=Material uses the opacity texture as a multiplier against itself for fading the bottom edge
   group=Fade Controls
   parameterName=Texture Fade Strength
   sliderMax=1
   sortPriority=9
  }
  n27 = Multiply(A=LinearInterpolate(Alpha=@n9) {constB=3}, B=Saturate(None=Multiply(A=@n25, B=@n26)))
  n28 = TextureCoordinate()
  n29 = ScalarParameter() {defaultValue=3, desc=Bottom Edge Fade exponent, group=Fade Controls, parameterName=Bottom Fade Smoothness, sortPriority=6}
  n30 = Power(
   Base=SphereMask(
    A=MaterialFunctionCall(Float2=@n28) {materialFunction=&BreakOutFloat2Components}.G
    B=Constant()
   ) {attenuationRadius=1, hardnessPercent=1}
   Exp=@n29
  ) {constExponent=3}
  n31 = LinearInterpolate(
   A=@n27
   B=Multiply(A=@n27, B=LinearInterpolate(Alpha=@n30))
   Alpha=ScalarParameter() {group=Fade Controls, parameterName=Use Bottom Fade, sortPriority=5}
  )
  n32 = TextureCoordinate()
  n33 = Sine(None=Multiply(A=ComponentMask(None=@n32) {r=true}) {constB=0.5})
  n34 = ScalarParameter() {defaultValue=3, desc=Side Edge Fade exponent, group=Fade Controls, parameterName=Side Fade Smoothness, sortPriority=8}
  n35 = Multiply(A=@n31, B=LinearInterpolate(Alpha=Power(Base=@n33, Exp=@n34) {constExponent=3}))
  n36 = ScalarParameter() {desc=Enables fade on both of the sides in UV space, group=Fade Controls, parameterName=Use Side Fade, sortPriority=7}
  n37 = ParticleColor()
  n38 = Multiply(
   A=MaterialFunctionCall(Color_Morning=@n1, Color_Day=@n1, Color_Evening=Constant() {r=0.7}, Color_Night=Constant() {r=0.3}) {materialFunction=&Color_Time_of_Day}.Result
   B=Multiply(A=LinearInterpolate(A=@n31, B=@n35, Alpha=@n36), B=@n37.A) {constB=0.2}
  )
  n39 = ScalarParameter() {defaultValue=1, group=Fade Controls, parameterName=FresnelBrightness, sortPriority=18}
  n40 = ScalarParameter() {defaultValue=1, group=Fade Controls, parameterName=FresnelExponent, sortPriority=19}
  n41 = VertexNormalWS()
  LinearInterpolate(
   A=@n38
   B=Multiply(
    A=MaterialFunctionCall(Brightness=@n39, Exponent=@n40, Vertex / Pixel Normal=@n41) {materialFunction=&MF_Fresnel_FX}.Fresnel Double Sided
    B=@n38
   )
   Alpha=ScalarParameter() {desc=Enable Fresnel, group=Fade Controls, parameterName=UseFresnel, sortPriority=17}
  )
  n42 = VectorParameter() {group=Color and Texture, parameterName=DarkColor, sortPriority=1, defaultValue=<rgba 0 0 0 1>}
  n43 = VectorParameter() {group=Color and Texture, parameterName=LightColor, sortPriority=2, defaultValue=<rgba 1 0.8928 0.7996 1>}
  MaterialFunctionCall(Emissive=Multiply(A=@n37.RGB, B=LinearInterpolate(A=@n42.RGB, B=@n43.RGB, Alpha=@n9))) {materialFunction=&MF_FX_EmissiveAdjustments}
 ]
}
