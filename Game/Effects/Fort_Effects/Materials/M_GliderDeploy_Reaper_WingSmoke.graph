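# Translucent unlit material for wing smoke on the Reaper glider deploy effect.
# Three panning noise textures, masked by UV edge falloff and Fresnel, drive a dark base color with hue-shifted rainbow thin-film tint.
# Opacity adds erosion and depth fades, simplified on mobile.
# Curl-like vector noise displaces vertices outside mobile, and UseStaticMesh swaps particle color and dynamic parameters for scalar parameters and vertex color.
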Material /Game/Effects/Fort_Effects/Materials/M_GliderDeploy_Reaper_WingSmoke
graph {
 properties {
  bUsedWithInstancedStaticMeshes=true
  bUsedWithMeshDeformer=true
  bUsedWithMeshParticles=true
  bUsedWithNiagaraMeshParticles=true
  bUsedWithNiagaraSprites=true
  bUsedWithSkeletalMesh=true
  blendMode=BLEND_Translucent
  pixelDepthOffsetMode=PDOM_Legacy
  shadingModel=MSM_Unlit
  twoSided=true
  usedShadingModels=Unlit
 }
 nodes [
  n1 = ScalarParameter() {defaultValue=1, parameterName=Emissive_Intensity}
  n2 = ParticleColor()
  n3 = VectorParameter() {parameterName=BaseColor, defaultValue=<rgba 0.0054 0.0019 0.0185 1>}
  n4 = ScalarParameter() {parameterName="Hue Shift Percentage (S)"}
  n5 = ScalarParameter() {defaultValue=0.7, group=ThinFilm, parameterName=ThinFilm_Scale}
  n6 = VertexNormalWS()
  n7 = Fresnel(ExponentIn=ScalarParameter() {defaultValue=1, group=ThinFilm, parameterName=ThinFilm_EX}, Normal=@n6) {baseReflectFraction=0}
  n8 = Multiply(
   A=TextureSample(UVs=Multiply(A=@n5, B=@n7)) {texture=/Game/Global/Textures/Default/Masks_and_Gradients/FortniteRainbow}.RGB
   B=ScalarParameter() {defaultValue=1, group=ThinFilm, parameterName=ThinFilm_Intensity}
  )
  n9 = Multiply(
   A=OneMinus(None=ScalarParameter() {defaultValue=0.7, parameterName=Roughness})
   B=ScalarParameter() {defaultValue=1, group=ThinFilm, parameterName=ThinFilm_RoughnessScale}
  )
  n10 = ScalarParameter() {defaultValue=1.2164, group=ThinFilm, parameterName=ThinFilm_Visibility}
  n11 = LinearInterpolate(
   A=@n3.RGB
   B=Add(
    A=@n3.RGB
    B=MaterialFunctionCall(Hue Shift Percentage=@n4, Texture=@n8) {materialFunction=&HueShift}.Result
   )
   Alpha=Multiply(A=Clamp(None=@n9), B=Fresnel(ExponentIn=@n10, Normal=@n6) {baseReflectFraction=0})
  )
  n12 = Power(Base=Fresnel(ExponentIn=ScalarParameter() {defaultValue=1, parameterName=SmokeFrensnel})) {constExponent=4}
  n13 = TextureCoordinate() {uTiling=3}
  n14 = StaticSwitchParameter(True=VertexColor().R, False=ParticleColor().R) {parameterName=UseStaticMesh}
  n15 = Add(A=@n14, B=Multiply(A=ScalarParameter() {defaultValue=1, parameterName=XTileMult}, B=ComponentMask(None=@n13) {r=true}))
  n16 = ComponentMask(None=@n13) {g=true}
  n17 = ScalarParameter() {defaultValue=1, parameterName=PanSpeedB}
  n18 = Constant()
  n19 = TextureSampleParameter2D(UVs=Panner(Coordinate=AppendVector(A=@n15, B=@n16), Speed=AppendVector(A=@n17, B=@n18)) {speedX=1}) {parameterName=TileTextureC, samplerType=SAMPLERTYPE_Grayscale, texture=&T_Smoke_Trail_Soft}
  n20 = TextureCoordinate() {uTiling=2, vTiling=0.65}
  n21 = Multiply(A=ComponentMask(None=@n20) {r=true}, B=ScalarParameter() {defaultValue=1, parameterName=XTileMult})
  n22 = ScalarParameter() {defaultValue=0.5, parameterName=PanSpeed}
  n23 = Panner(Coordinate=AppendVector(A=Add(A=@n14, B=@n21), B=ComponentMask(None=@n20) {g=true}), Speed=AppendVector(A=Constant(), B=@n22)) {speedY=0.5}
  n24 = ScalarParameter() {defaultValue=1, parameterName=XTileMult}
  n25 = TextureCoordinate() {uTiling=2, vTiling=0.65}
  n26 = ComponentMask(None=@n25) {r=true}
  n27 = AppendVector(A=Add(A=@n14, B=Multiply(A=@n24, B=@n26)), B=ComponentMask(None=@n25) {g=true})
  n28 = AppendVector(A=Constant(), B=@n22)
  n29 = Add(
   A=Add(
    A=@n19.G
    B=TextureSampleParameter2D(UVs=@n23) {parameterName=TileTextureA, texture=/Game/Weapons/FORT_Melee/Effects/LaserSwordFounders/T_LaserSwordTrailNoise}.G
   )
   B=TextureSampleParameter2D(UVs=Panner(Coordinate=@n27, Speed=@n28) {speedY=0.5}) {parameterName=TileTextureB, texture=&T_Inky_Smoke_Tile}.G
  )
  n30 = ScalarParameter() {defaultValue=1.25, parameterName=TextureContrast}
  n31 = TextureCoordinate()
  n32 = ComponentMask(None=@n31) {r=true}
  n33 = Multiply(A=@n32, B=OneMinus(None=@n32))
  n34 = ComponentMask(None=@n31) {g=true}
  n35 = Multiply(A=@n34, B=OneMinus(None=@n34))
  n36 = Constant() {r=1}
  n37 = Power(Base=VertexColor().G, Exp=ScalarParameter() {defaultValue=2, parameterName=VertexColorContrast})
  n38 = Multiply(
   A=Multiply(A=OneMinus(None=@n12), B=Power(Base=@n29, Exp=@n30) {constExponent=1.25})
   B=Multiply(A=Multiply(A=@n33, B=@n35), B=StaticSwitchParameter(True=@n36, False=@n37) {parameterName=UseStaticMesh})
  )
  n39 = LinearInterpolate(B=@n38, Alpha=ScalarParameter() {defaultValue=1, parameterName=PreMultAlpha}) {constA=1}
  MaterialFunctionCall(Emissive=Multiply(A=StaticSwitchParameter(True=@n1, False=@n2.G) {parameterName=UseStaticMesh}, B=Multiply(A=@n11, B=@n39))) {materialFunction=&MF_FX_EmissiveAdjustments}
  n40 = Fresnel(ExponentIn=ScalarParameter() {defaultValue=1, parameterName=Fresnel})
  n41 = ScalarParameter() {defaultValue=1, parameterName=Alpha_Intensity}
  n42 = ParticleColor()
  n43 = MaterialFunctionCall!(
   Dynamic Param=Multiply(A=@n38, B=StaticSwitchParameter(True=@n41, False=@n42.A) {parameterName=UseStaticMesh})
   Alpha=StaticSwitchParameter(
    True=ScalarParameter() {defaultValue=1, parameterName=Erode_Intensity}
    False=DynamicParameter() {defaultValue=<rgba 1 50 15 1>, paramNames [Curl_Speed, Curl_Scale, Curl_Intensity, Erode]}.Erode
   ) {parameterName=UseStaticMesh}
  ) {materialFunction=/Game/Effects/Fort_Effects/Materials/_Functions/Erode_Alpha}
  n44 = ScalarParameter() {parameterName=CameraFadeDistance}
  n45 = DepthFade(
   Opacity=Multiply(
    A=@n43.Result
    B=MaterialFunctionCall(Fade Length=@n44) {materialFunction=&CameraDepthFade}.Result
   )
   FadeDistance=ScalarParameter() {defaultValue=10, parameterName=FadeDistance}
  )
  n46 = Multiply(A=Power(Base=OneMinus(None=@n40), Exp=ScalarParameter() {defaultValue=2, parameterName=Fresnel_Power}), B=@n45)
  n47 = ScalarParameter() {defaultValue=1, parameterName=Opacity_Low}
  Multiply(
   A=ShadingPathSwitch(Default=StaticSwitchParameter(True=@n46, False=@n45) {parameterName=UseFresnel}, Mobile=Multiply(A=@n43.Result, B=@n47)) {desc=Migrated from FeatureLevelSwitch}
   B=ScalarParameter() {defaultValue=0.25, parameterName=MultOpacity}
  )
  ParticleColor()
  ParticleColor()
  ParticleColor()
  n48 = ActorPositionWS()
  n49 = ObjectPositionWS()
  n50 = ScalarParameter() {defaultValue=1, parameterName=Distortion_speed}
  n51 = Multiply(A=VertexColor().R, B=Time())
  n52 = Subtract(
   A=Subtract(A=WorldPosition().XYZ, B=StaticSwitchParameter(True=@n48, False=@n49) {defaultValue=true, parameterName=ActorPositionWPO})
   B=StaticSwitchParameter(
    True=Multiply(A=@n50, B=@n51)
    False=DynamicParameter() {defaultValue=<rgba 1 50 15 1>, paramNames [Curl_Speed, Curl_Scale, Curl_Intensity, Erode]}.Curl_Speed
   ) {parameterName=UseStaticMesh}
  )
  n53 = StaticSwitchParameter(
   True=ScalarParameter() {defaultValue=1, parameterName=Curl_Scale}
   False=DynamicParameter() {defaultValue=<rgba 1 50 15 1>, paramNames [Curl_Speed, Curl_Scale, Curl_Intensity, Erode]}.Curl_Scale
  ) {parameterName=UseStaticMesh}
  n54 = Multiply(
   A=StaticSwitchParameter(
    True=ScalarParameter() {defaultValue=1, parameterName=Curl_Intensity}
    False=DynamicParameter() {defaultValue=<rgba 1 50 15 1>, paramNames [Curl_Speed, Curl_Scale, Curl_Intensity, Erode]}.Curl_Intensity
   ) {parameterName=UseStaticMesh}
   B=VectorNoise(World Position=Divide(A=@n52, B=@n53)) {noiseFunction=VNF_VectorALU}
  )
  n55 = TextureCoordinate()
  n56 = OneMinus(None=Power(Base=ComponentMask(None=@n55) {g=true}, Exp=ScalarParameter() {defaultValue=1, parameterName=Curl_Stable_Power}) {constExponent=1})
  ShadingPathSwitch(Default=LinearInterpolate(A=@n54, B=Multiply(A=@n54, B=@n56), Alpha=ScalarParameter() {parameterName=Curl_Stability}), Mobile=Constant()) {desc=Migrated from FeatureLevelSwitch}
 ]
}
