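# M_Glider_IceCream_VertexAnim is a lit mesh and particle material for a melting ice cream glider effect.
# MF_MeltingVertexAnimation drives vertex offset, masked by inverted UV V, and supplies a normal.
# That normal blends with world-aligned noise and panning swirl layers, flattened by height relative to object or particle position and UVs.
# View angle and swirl blend two colors into emissive, with cheaper mobile and low-quality paths.
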Material /Game/Effects/Fort_Effects/Materials/M_Glider_IceCream_VertexAnim
graph {
 properties {
  bUsedWithMeshDeformer=true
  bUsedWithMeshParticles=true
  bUsedWithNiagaraMeshParticles=true
  bUsedWithSkeletalMesh=true
  pixelDepthOffsetMode=PDOM_Legacy
  usedShadingModels=Default Lit
 }
 nodes [
  n1 = WorldPosition()
  n2 = TransformPosition(None=WorldPosition().XYZ) {transformSourceType=TRANSFORMPOSSOURCE_World}
  n3 = StaticBoolParameter() {parameterName=UseParticlePosition}
  n4 = ParticlePositionWS()
  n5 = TransformPosition(None=ObjectPositionWS()) {transformSourceType=TRANSFORMPOSSOURCE_World}
  n6 = ComponentMask(None=Subtract(A=StaticSwitch(True=@n1.XYZ, False=@n2, Value=@n3), B=StaticSwitch(True=@n4, False=@n5, Value=@n3))) {b=true}
  n7 = Divide(A=@n6) {constB=300}
  LinearInterpolate(Alpha=Saturate(None=Multiply(A=@n7) {constB=2})) {constB=0.3}
  n8 = VectorParameter() {group=Color, parameterName=BaseColorHigh, defaultValue=<rgba 1 0.7977 0.5891 1>}
  n9 = DotProduct(A=CameraVectorWS(), B=VertexNormalWS())
  n10 = MaterialFunctionCall(Float2=TextureCoordinate() {vTiling=1.2}) {materialFunction=&BreakOutFloat2Components}
  n11 = Multiply(A=@n10.G, B=@n10.G)
  n12 = TextureSampleParameter2D(UVs=Panner(Coordinate=AppendVector(A=@n10.R, B=@n11)) {speedY=0.2}) {parameterName=NormalMapMain, samplerType=SAMPLERTYPE_Normal, texture=/Game/Global/Textures/Default/Flowmaps/T_SwirlFlow_02_N}
  n13 = TextureCoordinate()
  n14 = LinearInterpolate(
   A=@n12.B
   Alpha=OneMinus(None=MaterialFunctionCall(Float2=@n13) {materialFunction=&BreakOutFloat2Components}.G)
  )
  n15 = ShadingPathSwitch(
   Default=LinearInterpolate(
    A=VectorParameter() {group=Color, parameterName=BaseColorLow, defaultValue=<rgba 1 0.4907 0.1739 1>}.RGB
    B=@n8.RGB
    Alpha=Multiply(A=@n9, B=@n14)
   )
   Mobile=@n8.RGB
  ) {desc=Migrated from FeatureLevelSwitch}
  MaterialFunctionCall(Emissive=Divide(
    A=QualitySwitch(Default=@n15, Low=@n8.RGB, High=@n15, Medium=@n15)
    B=ScalarParameter() {defaultValue=5, group=Color, parameterName=EmisIntensity}
   ) {constB=5}) {materialFunction=&MF_FX_EmissiveAdjustments}
  n16 = TextureCoordinate()
  n17 = TextureObjectParameter() {
   group=VertexAnim
   parameterName=VertexAnimNormalMain
   samplerType=SAMPLERTYPE_Normal
   texture=/Game/Effects/Fort_Effects/Textures/T_DustDistortion
  }
  n18 = MaterialFunctionCall!(
   VertexAnimNormalMain=@n17
   VertexAnimTexSize=ScalarParameter() {defaultValue=75, group=VertexAnim, parameterName=VertexAnimTexSize}
   VertexAnimNormalSpeed=ScalarParameter() {defaultValue=30, group=VertexAnim, parameterName=VertexAnimNormalSpeed}
   WPO_PanSpeedWhileMoving=ScalarParameter() {defaultValue=1, group=VertexAnim, parameterName=WPO_PanSpeedWhileMoving}
   UseParticlePositionOrObjectPosition=StaticBoolParameter() {parameterName=UseParticlePosition}
   VertexAnimIntesityXY=ScalarParameter() {defaultValue=5, group=VertexAnim, parameterName=VertexAnimIntesityXY}
   WPO_Intensity=ScalarParameter() {defaultValue=0.3, group=VertexAnim, parameterName=WPO_Intensity}
   VertexAnimVerticalMaskBias=ScalarParameter() {defaultValue=0.3, group=VertexAnim, parameterName=VertexAnimVerticalMaskBias}
  ) {materialFunction=/Game/Effects/Fort_Effects/Materials/_Functions/MF_MeltingVertexAnimation}
  Multiply(
   A=OneMinus(None=MaterialFunctionCall(Float2=@n16) {materialFunction=&BreakOutFloat2Components}.G)
   B=@n18."VertexAnimation(WPO)"
  )
  ScalarParameter() {defaultValue=0.5, group=Color, parameterName=AO}
  n19 = Add(A=Divide(A=@n6) {constB=150}) {constB=0.3}
  n20 = Saturate(None=OneMinus(None=Multiply(A=@n19) {constB=2}))
  n21 = Constant3Vector() {constant=<rgba 0 0 1 0>}
  n22 = ShadingPathSwitch(
   Default=MaterialFunctionCall(Normal=@n18."XYZ(Normal)", Flatness=LinearInterpolate(Alpha=@n20) {constA=1, constB=-1}) {materialFunction=&FlattenNormal}.Result
   Mobile=@n21
  ) {desc=Migrated from FeatureLevelSwitch}
  n23 = QualitySwitch(Default=@n22, Low=@n21, High=@n22, Medium=@n22)
  n24 = TextureObjectParameter() {parameterName=NormalMapSecondary, samplerType=SAMPLERTYPE_Normal, texture=/Game/Effects/Fort_Effects/Textures/Misc/T_NoiseNormal}
  n25 = WorldPosition()
  n26 = StaticSwitch(True=ParticlePositionWS(), False=ObjectPositionWS(), Value=StaticBoolParameter() {parameterName=UseParticlePosition})
  n27 = Constant()
  n28 = Multiply(A=Time(), B=ScalarParameter() {defaultValue=30, group=VertexAnim, parameterName=VertexAnimNormalSpeed}) {constB=30}
  n29 = MaterialFunctionCall(
   TextureObject=@n24
   TextureSize=Constant() {r=20}
   WorldPosition=Add(
    A=Subtract(A=@n25.XYZ, B=@n26)
    B=MaterialFunctionCall(X=@n27, Y=@n27, Z=@n28) {materialFunction=&MakeFloat3}.Result
   )
  ) {materialFunction=&WorldAlignedTexture}
  n30 = Constant3Vector() {constant=<rgba 0 0 1 0>}
  n31 = ShadingPathSwitch(Default=@n29.XYZ Texture, Deferred=@n29.XYZ Texture, Mobile=@n30) {desc=Migrated from FeatureLevelSwitch}
  n32 = MaterialFunctionCall(Float2=TextureCoordinate() {uTiling=4.5, vTiling=5}) {materialFunction=&BreakOutFloat2Components}
  n33 = Power(Base=@n32.G) {constExponent=1.5}
  n34 = TextureSample(UVs=Panner(Coordinate=AppendVector(A=@n32.R, B=@n33)) {speedY=0.5}, Tex=@n24) {samplerType=SAMPLERTYPE_Normal, texture=&T_NoiseNormal}
  n35 = MaterialFunctionCall(
   Normal=StaticSwitchParameter(True=QualitySwitch(Default=@n31, Low=@n30, High=@n31, Medium=@n34.RGB), False=@n30) {defaultValue=true, parameterName=UseSecondaryNormalMap}
   Flatness=LinearInterpolate(Alpha=@n20) {constA=-0.6, constB=2}
  ) {materialFunction=&FlattenNormal}
  n36 = Constant3Vector() {constant=<rgba 0 0 1 0>}
  n37 = ShadingPathSwitch(Default=@n35.Result, Deferred=@n35.Result, Mobile=@n36) {desc=Migrated from FeatureLevelSwitch}
  n38 = QualitySwitch(Default=@n37, Low=@n36, High=@n37, Medium=@n37)
  n39 = Add(A=Divide(A=@n6) {constB=150}) {constB=0.5}
  n40 = Saturate(None=OneMinus(None=Multiply(A=@n39) {constB=2}))
  n41 = OneMinus(None=MaterialFunctionCall(Float2=TextureCoordinate()) {materialFunction=&BreakOutFloat2Components}.G)
  n42 = LinearInterpolate(Alpha=Saturate(None=Add(A=@n40, B=@n41))) {constA=-2}
  StaticSwitchParameter(
   True=MaterialFunctionCall(
    BaseNormal=MaterialFunctionCall(BaseNormal=@n23, AdditionalNormal=@n38) {materialFunction=&BlendAngleCorrectedNormals}.Result
    AdditionalNormal=MaterialFunctionCall(Normal=@n12.RGB, Flatness=@n42) {materialFunction=&FlattenNormal}.Result
   ) {materialFunction=&BlendAngleCorrectedNormals}.Result
   False=Constant3Vector() {constant=<rgba 0 0 1 0>}
  ) {defaultValue=true, parameterName=UseNormalMap}
 ]
}
