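# Unlit translucent hit-impact effect material for mesh particles, with Nanite usage enabled.
# Dynamic parameters slide the tiled impact texture along V (FadeOut) and size a particle-centered sphere mask (LerpRadius).
# Texture, mask and fresnel drive a dark-to-light color blend or switchable panning rainbow gradient, optionally particle-tinted.
# The texture mask also feeds depth-faded, optionally particle-alpha opacity, which a switch replaces with squared near-camera fade; low quality skips fresnel and depth fade.
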Material /Game/Effects/Fort_Effects/Materials/M_HitImpactDefault
graph {
 properties {
  bUsedWithMeshParticles=true
  bUsedWithNanite=true
  bUsedWithNiagaraMeshParticles=true
  blendMode=BLEND_Translucent
  pixelDepthOffsetMode=PDOM_Legacy
  shadingModel=MSM_Unlit
  usedShadingModels=Unlit
 }
 nodes [
  ComponentMask(None=TextureCoordinate()) {g=true}
  n1 = Constant() {r=3}
  n2 = MaterialFunctionCall(Exponent=@n1) {materialFunction=&MF_Fresnel_FX}
  n3 = QualitySwitch(Default=@n2.Fresnel, Low=@n1, High=@n2.Fresnel, Medium=@n2.Fresnel)
  n4 = Power(Base=SphereMask(
    A=ParticlePositionWS()
    B=WorldPosition().XYZ
    Radius=DynamicParameter() {defaultValue=<rgba 0.25 1 1 1>, paramNames [FadeOut,LerpRadius,Param3,Param4]}.LerpRadius
    Hardness=Constant()
   )) {constExponent=4}
  n5 = Constant()
  n6 = LinearInterpolate(
   A=Constant() {r=-1}
   B=Constant() {r=1}
   Alpha=DynamicParameter() {defaultValue=<rgba 0.25 1 1 1>, paramNames [FadeOut,LerpRadius,Param3,Param4]}.FadeOut
  )
  n7 = TextureSample(UVs=Add(
    A=TextureCoordinate() {uTiling=4}
    B=MaterialFunctionCall(X=@n5, Y=@n6) {materialFunction=&MakeFloat2}.Result
   )) {texture=/Game/Effects/Fort_Effects/Textures/T_HitImpact}
  n8 = Clamp(None=Subtract(A=Multiply(A=@n7.RGB) {constB=4}) {constB=3})
  n9 = Multiply(A=Clamp(None=Multiply(A=@n3, B=@n4)), B=@n8)
  n10 = ComponentMask(None=@n9) {g=true}
  n11 = StaticSwitchParameter(
   True=TextureSample(UVs=Panner(Coordinate=@n10) {speedX=1.6667}) {texture=/Game/Global/Textures/Default/Masks_and_Gradients/T_Grad_Rainbow}.RGB
   False=LinearInterpolate(
    A=VectorParameter() {parameterName=DarkCol, defaultValue=<rgba 5 0.545 0.05 0>}.RGB
    B=VectorParameter() {parameterName=LightCol, defaultValue=<rgba 8 6.2152 0.4 0>}.RGB
    Alpha=@n9
   )
  ) {parameterName=Use Rainbow Hit Coloration}
  n12 = ParticleColor()
  MaterialFunctionCall(Emissive=LinearInterpolate(A=@n11, B=Multiply(A=@n11, B=@n12.RGB), Alpha=ScalarParameter() {parameterName=UseParticleColor})) {materialFunction=&MF_FX_EmissiveAdjustments}
  n13 = MaterialFunctionCall(Distance=ScalarParameter() {defaultValue=150, parameterName=NearFadeDist}) {materialFunction=&Near_Camera_Fade}
  n14 = DepthFade(FadeDistance=Constant() {r=8})
  n15 = Multiply(A=@n8, B=QualitySwitch(Default=@n14, Low=Constant() {r=1}, High=@n14, Medium=@n14))
  StaticSwitchParameter(
   True=Multiply(A=@n13.Result, B=@n13.Result)
   False=LinearInterpolate(A=@n15, B=Multiply(A=@n12.A, B=@n15), Alpha=ScalarParameter() {parameterName=UseParticleOpacity})
  ) {parameterName=UseNearFadeCamera}
 ]
}
