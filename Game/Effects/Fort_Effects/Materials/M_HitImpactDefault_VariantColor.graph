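# Translucent unlit mesh-particle material for a hit impact effect, tinted by a VariantColor parameter.
# The FadeOut dynamic parameter slides a tiled impact texture along V, and scaling and clamping turn it into a sharp mask.
# That mask, Fresnel and a particle-centred sphere mask sized by LerpRadius brighten the emissive color.
# The texture mask is also multiplied by depth fade; low quality replaces Fresnel and depth fade with constants.
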Material /Game/Effects/Fort_Effects/Materials/M_HitImpactDefault_VariantColor
graph {
 properties {
  bUsedWithMeshParticles=true
  bUsedWithNiagaraMeshParticles=true
  blendMode=BLEND_Translucent
  pixelDepthOffsetMode=PDOM_Legacy
  shadingModel=MSM_Unlit
  usedShadingModels=Unlit
 }
 nodes [
  Constant() {r=0.3}
  n1 = Multiply(A=VectorParameter() {parameterName=VariantColor, defaultValue=<rgba 1 0 0.25 0>}.RGB) {constB=5}
  n2 = Constant() {r=3}
  n3 = MaterialFunctionCall(Exponent=@n2) {materialFunction=&MF_Fresnel_FX}
  n4 = SphereMask(
   A=ParticlePositionWS()
   B=WorldPosition().XYZ
   Radius=DynamicParameter() {defaultValue=<rgba 45 1 1 1>, paramNames [FadeOut,LerpRadius,Param3,Param4]}.LerpRadius
   Hardness=Constant()
  )
  n5 = Clamp(None=Multiply(A=QualitySwitch(Default=@n3.Fresnel, Low=@n2, High=@n3.Fresnel, Medium=@n3.Fresnel), B=Power(Base=@n4) {constExponent=4}))
  n6 = Constant()
  n7 = LinearInterpolate(
   A=Constant() {r=-1}
   B=Constant() {r=1}
   Alpha=DynamicParameter() {defaultValue=<rgba 45 1 1 1>, paramNames [FadeOut,LerpRadius,Param3,Param4]}.FadeOut
  )
  n8 = TextureSample(UVs=Add(
    A=TextureCoordinate() {uTiling=4}
    B=MaterialFunctionCall(X=@n6, Y=@n7) {materialFunction=&MakeFloat2}.Result
   )) {texture=/Game/Effects/Fort_Effects/Textures/T_HitImpact}
  n9 = Clamp(None=Subtract(A=Multiply(A=@n8.RGB) {constB=4}) {constB=3})
  MaterialFunctionCall(Emissive=LinearInterpolate(A=@n1, B=Multiply(A=@n1) {constB=1.25}, Alpha=Multiply(A=@n5, B=@n9))) {materialFunction=&MF_FX_EmissiveAdjustments}
  n10 = DepthFade(FadeDistance=Constant() {r=8})
  Multiply(A=@n9, B=QualitySwitch(Default=@n10, Low=Constant() {r=1}, High=@n10, Medium=@n10))
 ]
}
