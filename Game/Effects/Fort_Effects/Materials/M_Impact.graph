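# Unlit translucent impact effect material, flagged for use on Cascade and Niagara mesh particles.
# The impact texture's color is multiplied by particle color and fed into the shared MF_FX_EmissiveAdjustments function.
# Texture alpha times particle alpha is clamped below full strength, which keeps the effect partly see-through if used as opacity.
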Material /Game/Effects/Fort_Effects/Materials/M_Impact
graph {
 properties {
  bUsedWithMeshParticles=true
  bUsedWithNiagaraMeshParticles=true
  blendMode=BLEND_Translucent
  pixelDepthOffsetMode=PDOM_Legacy
  shadingModel=MSM_Unlit
  usedShadingModels=Unlit
 }
 nodes [
  n1 = TextureSample() {texture=/Game/Effects/Fort_Effects/Textures/T_Impact}
  n2 = ParticleColor()
  Clamp(None=Multiply(A=@n1.A, B=@n2.A), Max=Constant() {r=0.7})
  MaterialFunctionCall(Emissive=Multiply(A=@n1.RGB, B=@n2.RGB)) {materialFunction=&MF_FX_EmissiveAdjustments}
 ]
}
