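# M_Metal_Screw is a Default Lit material for mesh particles, apparently for metal screw debris in effects.
# It samples T_Metal_Screw and feeds its RGB as diffuse color into ConvertFromDiffSpec.
# The same RGB, scaled down by a small Fresnel base reflect fraction, supplies that function's specular color, a diffuse/specular-to-PBR conversion pattern.
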Material /Game/Effects/Fort_Effects/Materials/M_Metal_Screw
graph {
 properties {bUsedWithMeshParticles=true, pixelDepthOffsetMode=PDOM_Legacy, usedShadingModels=Default Lit}
 nodes [
  Constant() {r=1}
  n1 = TextureSample() {texture=/Game/Effects/Fort_Effects/Textures/T_Metal_Screw}
  MaterialFunctionCall(DiffuseColor=@n1.RGB, SpecularColor=Multiply(A=@n1.RGB) {constA=1, constB=0.04, desc=FresnelBaseReflectFraction}) {materialFunction=&ConvertFromDiffSpec}
 ]
}
