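# Additive unlit, two-sided Niagara sprite and mesh material for electric arc effects on the AllKnowingl pickaxe.
# Panning noise distorts the UVs of the arc and dissolve textures.
# Multiplied sines create per-particle flicker that scales the arc mask, which Erode sharpens before particle-color tinting and HDR emissive remapping.
# Dynamic-parameter dissolve combines with depth fade, while an optional near-fade switch adds distance fading and an offset toward particle center.
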Material /Game/Effects/Fort_Effects/Materials/M_Pickaxe_AllKnowingl_ElectricArc
graph {
 properties {
  bUsedWithNiagaraMeshParticles=true
  bUsedWithNiagaraSprites=true
  blendMode=BLEND_Additive
  pixelDepthOffsetMode=PDOM_Legacy
  shadingModel=MSM_Unlit
  twoSided=true
  usedShadingModels=Unlit
 }
 nodes [
  n1 = ParticleColor()
  n2 = LinearInterpolate(A=ParticleRandom(), B=Constant(), Alpha=ScalarParameter() {parameterName=Use Particle Random Value})
  n3 = DynamicParameter() {defaultValue=<rgba 1 1 0 1>, paramNames [Random,FlickerSpeed,Erode,Dissolve]}
  n4 = Multiply(A=Add(A=Time(), B=Add(A=@n2, B=@n3.Random)), B=@n3.FlickerSpeed)
  n5 = Sine(None=@n4) {period=0.09}
  n6 = Sine(None=@n4) {period=0.075}
  n7 = ConstantBiasScale(None=Multiply(A=Sine(None=@n4) {period=0.11}, B=Multiply(A=@n5, B=@n6)))
  n8 = ScalarParameter() {defaultValue=3, parameterName=FlickerContrast}
  n9 = Multiply(
   A=Saturate(None=MaterialFunctionCall(In=@n7, Contrast=@n8) {materialFunction=&CheapContrast}.Result)
   B=Clamp(None=@n1.A) {clampMode=CMODE_ClampMax}
  )
  n10 = TextureSample(UVs=Panner(Coordinate=TextureCoordinate()) {speedX=1, speedY=1}) {samplerType=SAMPLERTYPE_Normal, texture=/Game/Characters/Player/Female/Medium/Bodies/F_MED_Bistro_Astronaut/FX/T_GaussianNoise_Normal}
  n11 = Add(
   A=TextureCoordinate()
   B=Multiply(A=ComponentMask(None=@n10.RGB) {g=true, r=true}, B=ScalarParameter() {defaultValue=0.08, parameterName=DIstor Intensity})
  )
  n12 = TextureSampleParameter2D(UVs=@n11) {parameterName=BaseTextire, texture=/Game/Effects/Fort_Effects/Textures/T_Pickaxe_AllKnowingl_Arc_01}
  n13 = Multiply(A=@n1.RGB, B=Power(Base=Multiply(A=@n9, B=@n12.G), Exp=LinearInterpolate(Alpha=@n3.Erode) {constA=1, constB=16}))
  n14 = Clamp(None=@n13)
  n15 = ScalarParameter() {defaultValue=2, parameterName=HDRMult}
  MaterialFunctionCall(Emissive=ShadingPathSwitch(
    Default=MaterialFunctionCall(LDR Input=@n14, HDR Tint=@n13, HDR Intensity Multiplier=@n15) {materialFunction=&DeriveHDRfromLDR}.RemappedRange
    Mobile=@n13
   )) {materialFunction=&MF_FX_EmissiveAdjustments}
  n16 = ScalarParameter() {defaultValue=1024, parameterName=FadeNearDistance}
  n17 = LinearInterpolate(Alpha=SphereMask(A=PixelDepth().R, B=Constant(), Radius=Subtract(A=@n16) {constB=350}) {hardnessPercent=10}) {constA=1, constB=0}
  n18 = DepthFade(FadeDistance=ScalarParameter() {defaultValue=40, parameterName=Depth Fade})
  n19 = TextureSampleParameter2D(UVs=@n11) {parameterName=Disolve Tex, texture=/Game/Effects/Fort_Effects/Textures/T_Pickaxe_AllKnowingl_Arc_02}
  n20 = ScalarParameter() {defaultValue=20, group=Dissolve, parameterName=Dissolve Edge Hardness, sliderMin=0.5}
  n21 = Multiply(A=Power(Base=@n19.R, Exp=ScalarParameter() {defaultValue=1, group=Dissolve, parameterName=Dissolve Power}), B=@n20)
  n22 = LinearInterpolate(A=@n20, Alpha=Multiply(A=ScalarParameter() {defaultValue=1, group=Dissolve, parameterName=Dissolve Range}, B=@n3.Dissolve)) {constB=-1}
  Multiply(A=StaticSwitchParameter(True=Multiply(A=@n17, B=@n18), False=@n18) {parameterName=UseNearFade}, B=Saturate(None=Subtract(A=@n21, B=@n22)))
  OneMinus(None=@n19.R)
  n23 = Subtract(A=WorldPosition().XYZ, B=ParticlePositionWS())
  n24 = SphereMask(A=PixelDepth().R, B=Constant(), Radius=@n16) {hardnessPercent=10}
  StaticSwitchParameter(True=LinearInterpolate(A=Multiply(A=@n23) {constB=-1}, Alpha=OneMinus(None=@n24)) {constB=0}, False=Constant()) {parameterName=UseNearFade}
 ]
}
