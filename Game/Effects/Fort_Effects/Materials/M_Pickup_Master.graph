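# Master pickup material: a lit, masked surface that bobs, spins and glows, with opacity mask fixed at 1.
# World position offset combines sine bounce and axis rotation, optionally seeded from custom primitive data or actor position, with normals corrected for rotation.
# Switchable emissive layers add sphere-mask inner glow, Fresnel rim and masked texture, optionally pulsed over time.
# Inner glow follows the bounce and drops to a constant on low quality.
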Material /Game/Effects/Fort_Effects/Materials/M_Pickup_Master
graph {
 properties {
  bUseMaterialAttributes=true
  bUsedWithMeshDeformer=true
  bUsedWithNiagaraMeshParticles=true
  bUsedWithSkeletalMesh=true
  blendMode=BLEND_Masked
  pixelDepthOffsetMode=PDOM_Legacy
  usedShadingModels=Default Lit
 }
 nodes [
  n1 = MaterialFunctionCall(Float3=Divide(A=ActorPositionWS(), B=ScalarParameter() {defaultValue=33, group=WPO, parameterName=SeedFromActorLocationScale})) {materialFunction=&BreakOutFloat3Components}
  n2 = Add(A=@n1.R, B=@n1.R)
  n3 = StaticSwitchParameter(
   True=StaticSwitchParameter(True=ScalarParameter() {bUseCustomPrimitiveData=true, defaultValue=1, parameterName=Seed}, False=Add(A=@n2, B=@n1.B)) {group=WPO, parameterName="Seed from CPD (T) or Actor Position (F)"}
   False=Constant()
  ) {defaultValue=true, group=WPO, parameterName=Random WPO from Seed}
  n4 = Add(A=Multiply(A=ScalarParameter() {defaultValue=1, group=WPO-Bounce, parameterName=BounceSpeed}, B=Time()), B=@n3)
  n5 = Transform(None=Normalize(VectorInput=VectorParameter() {group=WPO-Bounce, parameterName=BounceDirection_LocalSpace, defaultValue=<rgba 0 0 1 0>}.RGB)) {transformSourceType=TRANSFORMSOURCE_Local}
  n6 = ScalarParameter() {defaultValue=10, group=WPO-Bounce, parameterName=BounceDistance}
  n7 = Add(
   A=Multiply(A=Sine(None=@n4) {period=3.3}, B=Multiply(A=@n5, B=@n6))
   B=Multiply(A=@n5, B=ScalarParameter() {group=WPO-Bounce, parameterName=PositionOffset})
  )
  n8 = ScalarParameter() {defaultValue=0.1, group=WPO-Rotation, parameterName=RotateSpeed}
  n9 = Add(A=Multiply(A=Frac(None=@n3)) {constB=0.2}) {constB=0.9}
  n10 = Add(A=@n3, B=Multiply(A=Time(), B=Multiply(A=@n8, B=@n9)) {constB=0.2})
  n11 = ActorPositionWS()
  n12 = Transform(None=VectorParameter() {group=WPO-Rotation, parameterName=RotationPivotOffset, defaultValue=<rgba 0 0 44 0>}.RGB) {transformSourceType=TRANSFORMSOURCE_Local}
  n13 = Normalize(VectorInput=Transform(None=VectorParameter() {group=WPO-Rotation, parameterName=AxisofRotation_LocalSpace, defaultValue=<rgba 0 0 1 0>}.RGB) {transformSourceType=TRANSFORMSOURCE_Local})
  n14 = RotateAboutAxis(NormalizedRotationAxis=@n13, RotationAngle=@n10, PivotPoint=Add(A=Add(A=@n11, B=@n12), B=@n7), Position=WorldPosition().XYZ)
  n15 = TextureSampleParameter2D() {
   desc="R = SpecMask G = Metallic B = Roughness"
   group=Specular Metalic Roughness
   parameterName=SpecularMasks
   samplerSource=SSM_Wrap_WorldGroupSettings
   sortPriority=0
   texture=/Engine/Engine_MI_Shaders/T_Specular
  }
  n16 = TextureSampleParameter2D() {group=Normals, parameterName=Normals, samplerType=SAMPLERTYPE_Normal, texture=&FlatNormalBlue}
  n17 = StaticSwitchParameter(True=@n16.RGB, False=Constant3Vector() {constant=<rgba 0 0 1 0>}) {defaultValue=true, group=Normals, parameterName=Use Normals Texture, sortPriority=-1}
  n18 = MaterialFunctionCall(Rotation Axis=@n13, Rotation Angle=@n10, World Space Vertex Normals=Transform(None=@n17)) {materialFunction=&FixRotateAboutAxisNormals}
  n19 = Add(A=Multiply(A=Time(), B=ScalarParameter() {defaultValue=0.1, group=Emissive, parameterName=EmissiveModulateSpeed}), B=@n3)
  n20 = LinearInterpolate(
   A=ScalarParameter() {group=Emissive, parameterName=EmissiveModulateMin}
   B=ScalarParameter() {defaultValue=1, group=Emissive, parameterName=EmissiveModulateMax}
   Alpha=ConstantBiasScale(None=Sine(None=@n19) {period=1.14})
  )
  n21 = ScalarParameter() {defaultValue=4.5, group=Emissive-InnerGlow, parameterName=InnerGlowFresnelFadeExp}
  n22 = Constant()
  n23 = Subtract(
   A=WorldPosition().XYZ
   B=Add(A=ActorPositionWS(), B=StaticSwitchParameter(True=@n7, False=@n22) {defaultValue=true, group=WPO, parameterName=UseWPO, sortPriority=-1})
  )
  n24 = VectorParameter() {group=Emissive-InnerGlow, parameterName=InnerGlowSphereScale, defaultValue=<rgba 1 1 1 1>}
  n25 = SphereMask(
   A=Divide(A=@n23, B=@n24.RGB)
   B=Divide(A=VectorParameter() {group=Emissive-InnerGlow, parameterName=InnerGlowCenterOffset}.RGB, B=@n24.RGB)
   Radius=ScalarParameter() {defaultValue=11, group=Emissive-InnerGlow, parameterName=GlowWorldRadius}
   Hardness=Constant()
  ) {attenuationRadius=11, hardnessPercent=0.1}
  n26 = Multiply(
   A=OneMinus(None=Fresnel(ExponentIn=@n21, Normal=@n18.New World Space Vertex Normal) {baseReflectFraction=0})
   B=Multiply(A=Power(Base=@n25) {constExponent=6}) {constB=2}
  )
  n27 = Constant() {r=0.005}
  n28 = VectorParameter() {group=Emissive-InnerGlow, parameterName=InnerGlowEmissiveColor, sortPriority=11, defaultValue=<rgba 4.6976 0.1302 5 0>}
  n29 = Multiply(
   A=Multiply(A=QualitySwitch(Default=@n26, Low=@n27), B=@n28.RGB)
   B=ScalarParameter() {defaultValue=1, group=Emissive-InnerGlow, parameterName=InnerGlowMultiplier}
  )
  n30 = Constant()
  n31 = ScalarParameter() {defaultValue=4, group=Emissive-Fresnel, parameterName=FresnelEmissiveExp, sortPriority=2}
  n32 = VectorParameter() {group=Emissive-Fresnel, parameterName=Fresnel Emissive Color, sortPriority=1, defaultValue=<rgba 4.6976 0.1302 5 0>}
  n33 = Multiply(
   A=Multiply(A=Fresnel(ExponentIn=@n31, Normal=@n18.New World Space Vertex Normal) {baseReflectFraction=0}, B=@n32.RGB)
   B=ScalarParameter() {defaultValue=1, group=Emissive-Fresnel, parameterName=FresnelEmissiveStrength, sortPriority=3}
  )
  n34 = Constant()
  n35 = ScalarParameter() {defaultValue=4, group=Emissive, parameterName=Fresnel Maskout Exp, sortPriority=21}
  n36 = VectorParameter() {group=Emissive, parameterName=EmissiveColor, sortPriority=1, defaultValue=<rgba 4.6976 0.1302 5 0>}
  n37 = Multiply(
   A=Multiply(
    A=@n36.RGB
    B=TextureSampleParameter2D() {group=Emissive, parameterName=EmissiveMasks, sortPriority=2, texture=&T_BlacksRGB}.RGB
   )
   B=ScalarParameter() {defaultValue=1, group=Emissive, parameterName=Emissive Multiplier}
  )
  n38 = Multiply(A=OneMinus(None=Fresnel(ExponentIn=@n35, Normal=@n18.New World Space Vertex Normal) {baseReflectFraction=0}), B=@n37)
  n39 = Add(
   A=Add(
    A=StaticSwitchParameter(True=@n29, False=@n30) {group=Emissive-InnerGlow, parameterName=Add Inner Glow Emissive, sortPriority=10}
    B=StaticSwitchParameter(True=@n33, False=@n34) {group=Emissive-Fresnel, parameterName=Add Fresnel Emissive, sortPriority=0}
   )
   B=StaticSwitchParameter(True=StaticSwitchParameter(True=@n38, False=@n37) {group=Emissive, parameterName=Mask out Emissive on Fresnel, sortPriority=20}, False=@n34) {group=Emissive, parameterName=Emissive From Texture, sortPriority=-1}
  )
  MakeMaterialAttributes!(
   BaseColor=TextureSampleParameter2D() {group=Base Color, parameterName=Diffuse, sortPriority=0, texture=/Engine/EngineMaterials/DefaultDiffuse_Low}.RGB
   Metallic=@n15.G
   Specular=@n15.R
   Roughness=LinearInterpolate(
    A=ScalarParameter() {group=Specular Metalic Roughness, parameterName=SpecRoughnessMin, sortPriority=31}
    B=ScalarParameter() {defaultValue=1, group=Specular Metalic Roughness, parameterName=SpecRoughnessMax}
    Alpha=@n15.B
   )
   Anisotropy
   EmissiveColor=StaticSwitchParameter(True=Multiply(A=@n20, B=@n39), False=@n39) {group=Emissive, parameterName=Modulate Emissive, sortPriority=1}
   Opacity
   OpacityMask=Constant() {r=1}
   Normal=StaticSwitchParameter(True=@n18.New Tangent Space Vertex Normal, False=@n17) {defaultValue=true, group=WPO, parameterName=UseWPO, sortPriority=-1}
   Tangent
   WorldPositionOffset=StaticSwitchParameter(True=Add(A=@n7, B=@n14), False=Constant()) {defaultValue=true, group=WPO, parameterName=UseWPO, sortPriority=-1}
   SubsurfaceColor
   ClearCoat
   ClearCoatRoughness
   AmbientOcclusion
   Refraction
   CustomizedUVs_0
   CustomizedUVs_1
   CustomizedUVs_2
   CustomizedUVs_3
   CustomizedUVs_4
   CustomizedUVs_5
   CustomizedUVs_6
   CustomizedUVs_7
   PixelDepthOffset
   ShadingModel
   Displacement
  )
  Multiply(A=@n23, B=@n24.RGB)
 ]
}
