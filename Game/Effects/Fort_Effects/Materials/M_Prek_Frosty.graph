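# Additive, unlit, two-sided particle material for sprites and meshes that draws a distorted, scrolling trail mask with a soft glow.
# A static switch fills the mask with either two-color cartoon flames or tinted circuit-board lines.
# Dynamic parameters scale the glow and drive the mask's scroll, while a panning normal map warps its UVs.
# Particle color tints the result, and a vertical UV gradient sets alpha, with optional depth fade or decal-style intersection fading.
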Material /Game/Effects/Fort_Effects/Materials/M_Prek_Frosty
graph {
 properties {
  bUsedWithMeshParticles=true
  bUsedWithNiagaraMeshParticles=true
  bUsedWithNiagaraSprites=true
  bUsedWithParticleSprites=true
  blendMode=BLEND_Additive
  pixelDepthOffsetMode=PDOM_Legacy
  shadingModel=MSM_Unlit
  twoSided=true
  usedShadingModels=Unlit
 }
 nodes [
  n1 = VectorParameter() {parameterName=Glow Color, defaultValue=<rgba 0.067 0.492 0.969 1>}
  n2 = SphereMask(A=ComponentMask(None=TextureCoordinate()) {g=true}, B=Constant() {r=1}) {attenuationRadius=0.85, hardnessPercent=0}
  n3 = Multiply(A=Multiply(A=Multiply(A=@n1.RGB, B=@n2)) {constB=0.027}, B=DynamicParameter().Param1)
  n4 = VectorParameter() {parameterName=Color_Dark, defaultValue=<rgba 0 0 0 1>}
  n5 = VectorParameter() {parameterName=Color_Light, defaultValue=<rgba 1 1 1 1>}
  n6 = ComponentMask(None=VectorParameter() {parameterName=UV_1, defaultValue=<rgba 0.85 0.14 1 1>}.RGB) {g=true, r=true}
  n7 = TextureSampleParameter2D(UVs=Multiply(A=TextureCoordinate(), B=@n6)) {parameterName=Texture_1, texture=/Game/Effects/Fort_Effects/Textures/Tile/T_Tile_Texture_Cartoon_Flames}
  n8 = Multiply(A=TextureCoordinate(), B=@n6)
  n9 = Multiply(
   A=StaticComponentMaskParameter(None=TextureSample(UVs=@n8) {texture=/Game/Effects/Fort_Effects/Textures/T_FX_Circuitboard}.RGB) {defaultG=true, parameterName=TechLines_RGB}
   B=VectorParameter() {parameterName=TechLines_Color, defaultValue=<rgba 0.0238 0.8346 0.927 1>}.RGB
  )
  n10 = Multiply(A=TextureCoordinate(), B=ComponentMask(None=VectorParameter() {parameterName=UV, defaultValue=<rgba 1 0.65 1 1>}.RGB) {g=true, r=true})
  n11 = Panner(Coordinate=TextureCoordinate() {uTiling=0.5, vTiling=0.5}) {speedY=0.5}
  n12 = Multiply(A=ComponentMask(None=TextureSample(UVs=@n11) {samplerType=SAMPLERTYPE_Normal, texture=&T_Weap_Distort_N_02}.RGB) {g=true, r=true}) {constB=0.45}
  n13 = TextureSampleParameter2D(UVs=Panner(Coordinate=Add(A=@n10, B=@n12), Time=DynamicParameter().Param2) {speedY=1}) {parameterName=Texture, samplerSource=SSM_Clamp_WorldGroupSettings, texture=/Game/Effects/Fort_Effects/Textures/Smoke/T_Lobber_Trail_Mask}
  n14 = Multiply(
   A=StaticSwitchParameter(True=LinearInterpolate(A=@n4.RGB, B=@n5.RGB, Alpha=@n7.R), False=Multiply(A=@n9) {constB=19}) {parameterName=Wumba}
   B=Multiply(A=@n13.RGB, B=ScalarParameter() {defaultValue=1, parameterName=light})
  )
  n15 = ParticleColor()
  MaterialFunctionCall(Emissive=Multiply(A=Add(A=@n3, B=@n14), B=@n15.RGB)) {materialFunction=&MF_FX_EmissiveAdjustments}
  n16 = SphereMask(A=ComponentMask(None=TextureCoordinate()) {g=true}, B=Constant() {r=0.5}) {attenuationRadius=0.5, hardnessPercent=0}
  n17 = ScalarParameter() {defaultValue=100}
  n18 = DepthFade(FadeDistance=ScalarParameter() {defaultValue=100, parameterName=Param_1})
  n19 = Multiply(A=@n16, B=Multiply(A=DepthFade(Opacity=@n16, FadeDistance=@n17), B=OneMinus(None=@n18)))
  n20 = DepthFade(FadeDistance=ScalarParameter() {defaultValue=100, parameterName=Depth Fade})
  n21 = Constant() {r=1}
  n22 = Multiply(A=@n16, B=FeatureLevelSwitch(Default=@n20, ES2=@n21, ES3_1=@n21, SM4_REMOVED=@n20, SM5=@n20, SM6=@n20))
  Multiply(
   A=@n15.A
   B=StaticSwitchParameter(True=StaticSwitchParameter(True=@n19, False=@n22) {parameterName=Decal}, False=@n16) {parameterName=Depth Fade}
  )
 ]
}
