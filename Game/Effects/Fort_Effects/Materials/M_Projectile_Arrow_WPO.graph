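# Translucent unlit arrow projectile trail material that stretches red-vertex-masked geometry from the actor position toward BulletOrigin.
# An eased timer from StartTime over WPODuration drives opacity and a shrinking vertex-normal offset term.
# Panning smoke noise, tiled by trail length, combines with axis-aligned fresnel, UV gradient and particle alpha to shape opacity and blend dark to bright color.
# Depth fade and HDR color boost are skipped on mobile.
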Material /Game/Effects/Fort_Effects/Materials/M_Projectile_Arrow_WPO
graph {
 properties {blendMode=BLEND_Translucent, pixelDepthOffsetMode=PDOM_Legacy, shadingModel=MSM_Unlit, usedShadingModels=Unlit}
 nodes [
  n1 = VectorParameter() {parameterName=BulletOrigin}
  n2 = ActorPositionWS()
  Add(A=Multiply(A=VertexColor().R, B=Subtract(A=@n1.RGB, B=@n2))) {constB=0}
  n3 = Divide(
   A=Subtract(A=Time() {period=5}, B=ScalarParameter() {parameterName=StartTime})
   B=ScalarParameter() {defaultValue=0.2, group=Timestamp, parameterName=WPODuration}
  )
  n4 = Clamp(None=MaterialFunctionCall!(In=Clamp(None=@n3)) {materialFunction=/Game/Effects/Fort_Effects/Materials/_Functions/MF_Smoothstep_EaseOut}.Result)
  Multiply(A=VertexNormalWS(), B=LinearInterpolate(Alpha=@n4) {constA=15, constB=0})
  n5 = ParticleColor()
  n6 = Normalize(VectorInput=Transform(None=Constant3Vector() {constant=<rgba 1 0 0 0>}) {transformSourceType=TRANSFORMSOURCE_Local})
  n7 = ScalarParameter() {defaultValue=6, parameterName=FresnelExponent}
  n8 = ComponentMask(None=TextureCoordinate()) {g=true}
  n9 = LinearInterpolate(Alpha=@n4) {constA=1, constB=0.5}
  n10 = Multiply(
   A=@n5.A
   B=Multiply(
    A=MaterialFunctionCall(Axis=@n6, Exponent=@n7) {materialFunction=&AxisAlignedFresnel}.Fresnel
    B=Power(Base=@n8, Exp=@n9) {constExponent=0.5}
   )
  )
  n11 = DepthFade(Opacity=@n10, FadeDistance=ScalarParameter() {defaultValue=40, parameterName=DepthFadeDist})
  n12 = VectorParameter() {parameterName=UV_Tile_Pan, defaultValue=<rgba 1 1 0 0>}
  n13 = Multiply(A=TextureCoordinate() {vTiling=0.001}, B=ComponentMask(None=@n12.RGB) {g=true, r=true})
  n14 = MaterialFunctionCall(X=ComponentMask(None=@n13) {r=true}, Y=Multiply(A=ComponentMask(None=@n13) {g=true}, B=Distance(A=@n1.RGB, B=@n2))) {materialFunction=&MakeFloat2}
  n15 = MaterialFunctionCall(X=@n12.B, Y=@n12.B) {materialFunction=&MakeFloat2}
  n16 = Multiply(
   A=Multiply(A=ShadingPathSwitch(Default=@n11, Mobile=@n10) {desc=Migrated from FeatureLevelSwitch}, B=@n4)
   B=TextureSampleParameter2D(UVs=Panner(Coordinate=@n14.Result, Speed=@n15.Result) {speedY=-0.5}) {parameterName=Noise, texture=&T_Inky_Smoke_Tile}.G
  )
  Multiply(A=@n16, B=ScalarParameter() {defaultValue=1, parameterName=Opacity})
  n17 = Multiply(
   A=@n5.RGB
   B=LinearInterpolate(
    A=VectorParameter() {parameterName=DarkColor, defaultValue=<rgba 0.2 0.2 0.2 0>}.RGB
    B=VectorParameter() {parameterName=BrightColor, defaultValue=<rgba 1 1 1 0>}.RGB
    Alpha=@n16
   )
  )
  ShadingPathSwitch(
   Default=MaterialFunctionCall(LDR Input=Clamp(None=@n17), HDR Tint=@n17, HDR Intensity Multiplier=ScalarParameter() {defaultValue=30, parameterName=HDRMult}) {materialFunction=&DeriveHDRfromLDR}.RemappedRange
   Mobile=@n17
  ) {desc=Migrated from FeatureLevelSwitch}
 ]
}
