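# Translucent unlit bullet projectile material that stretches the mesh into a streak with world position offset.
# Red vertex color masks the stretch toward the bullet origin, animated by eased timestamp or Niagara parameters, or along local -X by particle-driven length.
# A UV gradient with animated falloff, times axis-aligned fresnel, blends dark to bright color for HDR emissive and drives opacity.
# Depth and near-camera fades soften it, with cheaper mobile paths.
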Material /Game/Effects/Fort_Effects/Materials/M_Projectile_Bullet_WPO
graph {
 properties {
  bUsedWithNiagaraMeshParticles=true
  blendMode=BLEND_Translucent
  pixelDepthOffsetMode=PDOM_Legacy
  shadingModel=MSM_Unlit
  usedShadingModels=Unlit
 }
 nodes [
  n1 = Time() {period=5}
  n2 = ScalarParameter() {parameterName=StartTime}
  n3 = Clamp(None=Divide(A=Subtract(A=@n1, B=@n2), B=ScalarParameter() {defaultValue=0.2, group=Timestamp, parameterName=WPODuration}))
  n4 = ScalarParameter() {parameterName=UsedAsParticle}
  n5 = DynamicParameter() {desc=Parameters set by Niagara, parameterIndex=1, paramNames [X,Y,Z,Animation]}
  n6 = ScalarParameter() {parameterName=UsedWithNiagaraParameter}
  n7 = LinearInterpolate(
   A=Clamp(
    None=MaterialFunctionCall!(In=@n3) {materialFunction=/Game/Effects/Fort_Effects/Materials/_Functions/MF_Smoothstep_EaseOut}.Result
    Min=Multiply(A=@n4) {constB=0.5}
   )
   B=@n5.Animation
   Alpha=@n6
  )
  n8 = Subtract(
   A=LinearInterpolate(
    A=VectorParameter() {parameterName=BulletOrigin}.RGB
    B=MaterialFunctionCall(X=@n5.X, Y=@n5.X, Z=@n5.X) {materialFunction=&MakeFloat3}.Result
    Alpha=@n6
   )
   B=LinearInterpolate(A=ActorPositionWS(), B=ParticlePositionWS(), Alpha=@n6)
  )
  n9 = VertexColor()
  n10 = Normalize(VectorInput=Transform(None=Constant3Vector() {constant=<rgba -1 0 0 0>}) {transformSourceType=TRANSFORMSOURCE_Local})
  n11 = DynamicParameter() {paramNames [Length,Param2,Param3,Param4]}
  LinearInterpolate(
   A=Multiply(A=@n7, B=Multiply(A=@n8, B=@n9.R))
   B=Multiply(A=@n9.R, B=Multiply(A=@n10, B=@n11.Length))
   Alpha=ScalarParameter() {parameterName=UsedAsParticle}
  )
  n12 = ParticleColor()
  n13 = TextureCoordinate()
  n14 = ScalarParameter() {defaultValue=5, parameterName=FadePowerEnd}
  n15 = ScalarParameter() {defaultValue=0.6, parameterName=FadePowerStart}
  n16 = Constant3Vector() {constant=<rgba 1 0 0 0>}
  n17 = Transform(None=@n16) {transformSourceType=TRANSFORMSOURCE_Local}
  n18 = Transform(None=@n16) {transformSourceType=TRANSFORMSOURCE_Local, transformType=TRANSFORM_ParticleWorld}
  n19 = ScalarParameter() {parameterName=UsedAsParticle}
  n20 = Normalize(VectorInput=LinearInterpolate(
    A=LinearInterpolate(A=@n17, B=@n18, Alpha=@n19)
    B=ParticleDirection()
    Alpha=ScalarParameter() {parameterName=ParticleDirectionForFresnelAxis}
   ))
  n21 = ScalarParameter() {defaultValue=6, parameterName=FresnelExponent}
  n22 = Constant() {r=1}
  n23 = Multiply(
   A=Power(Base=ComponentMask(None=@n13) {g=true}, Exp=LinearInterpolate(A=@n14, B=@n15, Alpha=@n7) {constA=5, constB=0.6})
   B=ShadingPathSwitch(
    Default=MaterialFunctionCall(Axis=@n20, Exponent=@n21) {materialFunction=&AxisAlignedFresnel}.Fresnel
    Forward=@n22
    Mobile=@n22
   )
  )
  n24 = Multiply(
   A=@n12.RGB
   B=LinearInterpolate(
    A=VectorParameter() {parameterName=DarkColor, defaultValue=<rgba 0.5 0.0442 0.002 0>}.RGB
    B=VectorParameter() {parameterName=BrightColor, defaultValue=<rgba 1.25 0.9375 0.025 0>}.RGB
    Alpha=@n23
   )
  )
  n25 = Clamp(None=@n24)
  n26 = ScalarParameter() {defaultValue=30, parameterName=HDRMult}
  MaterialFunctionCall(Emissive=ShadingPathSwitch(
    Default=MaterialFunctionCall(LDR Input=@n25, HDR Tint=@n24, HDR Intensity Multiplier=@n26) {materialFunction=&DeriveHDRfromLDR}.RemappedRange
    Mobile=@n24
   )) {materialFunction=&MF_FX_EmissiveAdjustments}
  n27 = Multiply(A=@n12.A, B=@n23)
  n28 = DepthFade(Opacity=@n27, FadeDistance=ScalarParameter() {defaultValue=40, parameterName=DepthFadeDist})
  n29 = Power(Base=@n7) {constExponent=0.3}
  n30 = ScalarParameter() {defaultValue=50, parameterName=CamFadeDistance}
  n31 = ScalarParameter() {defaultValue=250, parameterName=CamFadeDistance_Mobile}
  Multiply(
   A=Multiply(A=ShadingPathSwitch(Default=@n28, Mobile=@n27), B=Clamp(None=@n29))
   B=MaterialFunctionCall(Distance=ShadingPathSwitch(Default=@n30, Mobile=@n31)) {materialFunction=&Near_Camera_Fade}.Result
  )
 ]
}
