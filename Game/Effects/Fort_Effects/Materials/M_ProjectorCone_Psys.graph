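# Additive unlit mesh-particle material, named for a projector cone effect, with responsive AA enabled.
# A radial gradient texture pans over time at a speed set by a parameter and a per-particle dynamic parameter, then is tinted by particle color and passed through the shared emissive adjustment function.
# An alpha mask multiplies particle alpha, Fresnel, view-angle fade and a powered vertical UV gradient, shaping the falloff.
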Material /Game/Effects/Fort_Effects/Materials/M_ProjectorCone_Psys
graph {
 properties {
  bEnableResponsiveAA=true
  bUsedWithMeshParticles=true
  blendMode=BLEND_Additive
  pixelDepthOffsetMode=PDOM_Legacy
  shadingModel=MSM_Unlit
  usedShadingModels=Unlit
 }
 nodes [
  n1 = Time()
  n2 = ScalarParameter() {defaultValue=0.25, parameterName=PanSpeed}
  n3 = Panner(
   Coordinate=TextureCoordinate()
   Time=Multiply(A=Multiply(A=@n1, B=@n2), B=DynamicParameter() {paramNames [PanSpeed,Param2,Param3,Param4]}.PanSpeed)
  ) {speedX=1, speedY=1}
  n4 = ParticleColor()
  MaterialFunctionCall(Emissive=Multiply(A=TextureSample(UVs=@n3) {texture=/Game/Maps/Textures/RadialGradient2}.RGB, B=@n4.RGB)) {materialFunction=&MF_FX_EmissiveAdjustments}
  n5 = Fresnel(ExponentIn=ScalarParameter() {defaultValue=1.5, parameterName=FresnelExp})
  n6 = MaterialFunctionCall("FadeSensitivity (optional)"=ScalarParameter() {defaultValue=4, parameterName=FadeAngleSensitivity}) {materialFunction=&FadeBasedOnViewAngle}
  n7 = MaterialFunctionCall() {materialFunction=&LinearGradient}
  n8 = ScalarParameter() {defaultValue=3, parameterName=GradientFalloff}
  Multiply(A=@n4.A, B=Multiply(A=Multiply(A=@n5, B=@n6.Result), B=Power(Base=@n7.VGradient, Exp=@n8) {constExponent=3}))
 ]
}
