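# Translucent unlit Niagara ribbon material for an energy ribbon, colored by particle color rather than by texture.
# A procedural band, peaking mid-width on panned UVs, boosts that color and, multiplied by leading and trailing edge masks, forms opacity.
# Opacity then runs through shared FX opacity, depth fade and bevel functions, and fades out near the high end of UV1 U.
# The sampled smoke texture feeds only an unused vector-length node.
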Material /Game/Effects/Fort_Effects/Materials/M_RainbowEnergyRibbon
graph {
 properties {
  bUseMaterialAttributes=true
  bUsedWithNiagaraRibbons=true
  blendMode=BLEND_Translucent
  pixelDepthOffsetMode=PDOM_Legacy
  refractionMethod=RM_IndexOfRefraction
  shadingModel=MSM_Unlit
  usedShadingModels=Unlit
 }
 nodes [
  DynamicParameter()
  n1 = MaterialFunctionCall() {materialFunction=/Game/Effects/Niagara/MasterMaterials/Functions/MF_NFX_Bridge}
  n2 = MaterialFunctionCall!(
   UV Channel 2=Multiply(A=TextureCoordinate() {coordinateIndex=1}, B=AppendVector(A=@n1.UV1 U Freq, B=@n1.UV1 U Freq))
   UV Channel 1=Multiply(A=TextureCoordinate(), B=AppendVector(A=@n1.UV0 U Freq, B=@n1.UV0 U Freq))
   In
   UV Channel=Clamp(None=ScalarParameter() {group="2) BaseColor", parameterName="UV Channel (BaseColor)", sliderMax=1, sortPriority=1})
   U Freq Multiplier=ScalarParameter() {defaultValue=1, group="2) BaseColor", parameterName="U Freq (BaseColor)", sortPriority=2}
   V Freq Multiplier=ScalarParameter() {defaultValue=1, group="2) BaseColor", parameterName="V Freq (BaseColor)", sortPriority=3}
   Pan Speed U=ScalarParameter() {group="2) BaseColor", parameterName="Pan U Speed (BaseColor)", sortPriority=4}
   Pan Speed V=ScalarParameter() {group="2) BaseColor", parameterName="Pan V Speed (BaseColor)", sortPriority=5}
  ) {desc=The Niagara Bridge, materialFunction=/Game/Effects/Niagara/MasterMaterials/Functions/UVs/MF_NFX_TextureUVPanner}
  n3 = TextureSampleParameter2D(UVs=@n2.Result) {
   desc=This texture provides the foundation for the other channels like opacity and emissive color.
   group="2) BaseColor"
   parameterName=Texture
   texture=/Game/Effects/Fort_Effects/Textures/T_InkySmokeTrail
  }
  MaterialFunctionCall(Vector 3=@n3.RGB) {materialFunction=/Engine/Functions/Engine_MaterialFunctions02/Utility/VectorLength}
  n4 = ParticleColor()
  n5 = MaterialFunctionCall(Float2=@n2.Result) {materialFunction=&BreakOutFloat2Components}
  n6 = Multiply(A=@n5.G, B=OneMinus(None=@n5.G))
  n7 = Power(Base=@n6, Exp=ScalarParameter() {defaultValue=3, group="2) BaseColor", parameterName=MiddleBandEdges_PowerExp}) {constExponent=3}
  n8 = Add(
   A=@n4.RGB
   B=Multiply(A=Multiply(A=@n4.RGB, B=@n7), B=ScalarParameter() {defaultValue=100, group="2) BaseColor", parameterName=MiddleBandEdges_Hardness}) {constB=100}
  )
  n9 = MaterialFunctionCall(Float2=TextureCoordinate()) {materialFunction=&BreakOutFloat2Components}
  n10 = Power(Base=@n9.R, Exp=ScalarParameter() {defaultValue=1, group="2) BaseColor", parameterName=LeadingEdge_PowerExp})
  n11 = ScalarParameter() {defaultValue=1, group="2) BaseColor", parameterName=LeadingEdge_Hardness}
  n12 = Power(Base=OneMinus(None=@n9.R), Exp=ScalarParameter() {defaultValue=1, group="2) BaseColor", parameterName=TrailingEdge_PowerExp})
  n13 = ScalarParameter() {defaultValue=1, group="2) BaseColor", parameterName=TrailingEdge_Hardness}
  n14 = Multiply(A=Clamp(None=Multiply(A=@n10, B=@n11)), B=Clamp(None=Multiply(A=@n12, B=@n13)))
  n15 = LinearInterpolate(A=@n8, B=Multiply(A=@n8, B=@n14), Alpha=ScalarParameter() {group="2) BaseColor", parameterName=AlphaMaskInfluence})
  n16 = SetMaterialAttributes!(MaterialAttributes, Base Color=@n15, Emissive Color=@n15, Refraction, Opacity=Multiply(A=@n6, B=@n14)) {attributeSetTypes [
    #1
    #2
    #3
    #4
   ]}
  n17 = MaterialFunctionCall!(In=MaterialFunctionCall!(Material Attributes=@n16) {materialFunction=/Game/Effects/Niagara/MasterMaterials/Functions/V2/MF_Skydive_CameraFade_V2}."Opacitys and BaseColor (MA)") {
   desc=Controlled by the Material Component Mask Niagara Module.
   materialFunction=/Game/Effects/Niagara/MasterMaterials/Functions/MF_NFX_ComponentMask
  }
  n18 = MaterialFunctionCall!(In=@n17."BaseColor (MA) ") {
   desc=Controled throught the Material or Material ColorCorrect Niagara Module.
   materialFunction=/Game/Effects/Niagara/MasterMaterials/Functions/MF_Niagara_ColorCorrect
  }
  n19 = ScalarParameter() {
   desc=This parameter multiplies the texture by particle color and clamps it before going into the ldr range of the ldr to hdr function.
   group="5) Emissive"
   parameterName=Proper Particle HDR Coloring
   sliderMax=1
  }
  MaterialFunctionCall!(
   Override LDR imput with particle color=ScalarParameter() {group="2) BaseColor", parameterName=Override base color texture with particle color}
   In=@n18."BaseColor (MA)"
   Correct Coloring=@n19
  ) {materialFunction=/Game/Effects/Niagara/MasterMaterials/Functions/MF_FX_Colorize}
  n20 = MaterialFunctionCall!(Distortion, In=@n16) {materialFunction=/Game/Effects/Niagara/MasterMaterials/Functions/MF_FX_Opacity}
  n21 = GetMaterialAttributes(None=MaterialFunctionCall!(In=MaterialFunctionCall!(In=@n20.Opacity) {materialFunction=/Game/Effects/Niagara/MasterMaterials/Functions/Opacity/FX_DepthFade}."Opacity (MA)") {materialFunction=/Game/Effects/Niagara/MasterMaterials/Functions/Opacity/MF_Parametric_OpacityBevel}."Opacitys and BaseColor (MA)") {attributeGetTypes [#4, #5]}
  n22 = MaterialFunctionCall(
   Input=ComponentMask(None=TextureCoordinate() {coordinateIndex=1}) {r=true}
   Input Low=Constant() {r=0.8}
   Input High=Constant() {r=0.9}
   Target Low=Constant() {r=1}
   Target High=Constant()
  ) {materialFunction=&RemapValueRange}
  SetMaterialAttributes!(
   MaterialAttributes=@n21.MaterialAttributes
   Opacity=Multiply(A=@n21.Opacity, B=@n22.Result)
   Opacity Mask=Multiply(A=@n21.Opacity Mask, B=@n22.Result)
  ) {attributeSetTypes [#4, #5]}
  TextureCoordinate()
 ]
}
