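# Additive, unlit, two-sided rainbow effect material, named for spline use, that hides at night through time-of-day values.
# World position offset math builds camera-facing quads from UV0, sized by vertex red, collapsed when faded, and pushed toward camera by a height-blended distance.
# A tinted rainbow gradient drives emissive.
# Opacity combines soft UV1 masks, panning noise, depth, distance and optional view-from-above fades, plus optional particle alpha.
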Material /Game/Effects/Fort_Effects/Materials/M_Rainbow_Spline_Butterfly
graph {
 properties {blendMode=BLEND_Additive, pixelDepthOffsetMode=PDOM_Legacy, shadingModel=MSM_Unlit, twoSided=true, usedShadingModels=Unlit}
 nodes [
  n1 = MaterialFunctionCall(Float2=Subtract(A=TextureCoordinate()) {constB=0.5}) {materialFunction=&BreakOutFloat2Components}
  n2 = Subtract(A=ObjectPositionWS(), B=CameraPositionWS())
  n3 = Multiply(A=VectorParameter() {parameterName=CamDirectionMultiply, defaultValue=<rgba 1 1 1 0>}.RGB, B=@n2)
  n4 = CrossProduct(A=@n3, B=Constant3Vector() {constant=<rgba 0 0 1 0>})
  n5 = Add(
   A=ScalarParameter() {defaultValue=40567.5078, parameterName=Scale}
   B=Multiply(A=ScalarParameter() {defaultValue=-9849.6465, parameterName=Width}, B=VertexColor().R)
  )
  n6 = Constant() {r=1}
  n7 = Constant()
  n8 = Ceil(None=Multiply(
    A=MaterialFunctionCall(Value_Morning=@n6, Value_Day=@n6, Value_Evening=@n6, Value_Night=@n7) {materialFunction=&Scalar_Time_of_Day}.Result
    B=NamedRerouteUsage()
   ))
  n9 = MaterialFunctionCall(Float2=Multiply(
    A=Multiply(A=@n5, B=@n8)
    B=MaterialFunctionCall() {materialFunction=/Engine/Functions/Engine_MaterialFunctions02/WorldPositionOffset/ObjectScale}.Scale Z
   )) {materialFunction=&BreakOutFloat2Components}
  n10 = Multiply(A=Multiply(A=@n1.R, B=Normalize(VectorInput=@n4)), B=@n9.R)
  n11 = CrossProduct(A=@n3, B=@n4)
  n12 = Multiply(A=Multiply(A=@n1.G, B=Normalize(VectorInput=@n11)), B=@n9.G)
  n13 = Add(A=Subtract(A=Add(A=@n10, B=@n12), B=WorldPosition().XYZ), B=ObjectPositionWS())
  n14 = Subtract(A=CameraPositionWS(), B=Add(A=WorldPosition().XYZ, B=@n13))
  n15 = ScalarParameter() {parameterName=CameraOffsetGround}
  n16 = ScalarParameter() {defaultValue=7000, parameterName=CameraOffsetSky}
  n17 = CameraPositionWS()
  n18 = Clamp(None=Divide(A=ComponentMask(None=@n17) {b=true}, B=Constant() {r=20000}))
  Add(A=Multiply(A=Normalize(VectorInput=@n14), B=LinearInterpolate(A=@n15, B=@n16, Alpha=@n18)), B=@n13)
  CameraVectorWS()
  Clamp(None=MaterialFunctionCall(
    Fade Length=ScalarParameter() {defaultValue=8000, parameterName=FadeLength}
    Fade Offset=ScalarParameter() {defaultValue=8000, parameterName=FadeOffset}
   ) {materialFunction=&CameraDepthFade}.Result)
  Constant() {r=15000}
  Constant() {r=-10000}
  Constant() {r=1}
  Constant() {r=1}
  n19 = TextureCoordinate()
  n20 = MaterialFunctionCall(Float2=@n19) {materialFunction=&BreakOutFloat2Components}
  n21 = DotProduct(A=VertexNormalWS(), B=CameraVectorWS())
  LinearInterpolate(
   A=@n19
   B=AppendVector(A=OneMinus(None=@n20.R), B=@n20.G)
   Alpha=MaterialFunctionCall!(Mask Offset Value, Number Before White Result, Gradient=Clamp(None=@n21)) {materialFunction=/Engine/Functions/Engine_MaterialFunctions02/ValueStep}.Result
  )
  n22 = TextureSample(UVs=Panner(Coordinate=TextureCoordinate() {coordinateIndex=1, uTiling=0.05, vTiling=3}) {speedX=0.025}) {texture=&T_NoisePack_01}
  n23 = TextureSample!(
   UVs=Panner(Coordinate=TextureCoordinate() {coordinateIndex=1, uTiling=0.05, vTiling=4}) {speedX=-0.025}
   Tex
   Level=Constant() {r=4}
   Apply View MipBias
  ) {mipValueMode=TMVM_MipLevel, texture=&T_NoisePack_01}
  Multiply(A=@n22.G, B=@n23.R)
  n24 = Normalize(VectorInput=@n2)
  n25 = OneMinus(None=Abs(None=ComponentMask(None=@n24) {b=true}))
  n26 = Constant() {r=0.5}
  NamedRerouteDeclaration(None=StaticSwitchParameter(
    True=MaterialFunctionCall(In=@n25, Contrast=@n26) {materialFunction=&CheapContrast}.Result
    False=Constant() {r=1}
   ) {parameterName=FadeWhenAbove}) {name=FadeWhenAbove, nodeColor=<rgba 0.5 0 1 1>}
  n27 = Distance(A=CameraPositionWS(), B=ObjectPositionWS())
  n28 = ScalarParameter() {defaultValue=-10000, parameterName=StartFadeBias}
  n29 = Clamp(None=Divide(A=Add(A=@n27, B=@n28), B=ScalarParameter() {defaultValue=15000, parameterName=FadeDistance}))
  OneMinus(None=@n29)
  n30 = TextureCoordinate() {coordinateIndex=1, uTiling=-1.1}
  n31 = ScalarParameter() {defaultValue=1.3429, parameterName=ColorScroll}
  n32 = Add(
   A=VectorParameter() {parameterName=Color, defaultValue=<rgba 1 0.7381 0.0933 1>}.RGB
   B=TextureSample!(UVs=Add(A=@n30, B=@n31), Tex, Apply View MipBias) {texture=/Game/Global/Textures/Default/Masks_and_Gradients/T_Grad_Rainbow}.RGB
  )
  n33 = ScalarParameter() {defaultValue=0.142, parameterName=Brightness}
  Reroute(None=MaterialFunctionCall(Emissive=Multiply(A=@n32, B=@n33)) {materialFunction=&MF_FX_EmissiveAdjustments}.Result)
  ScalarParameter() {defaultValue=1, parameterName=OpacityMult}
  ScalarParameter() {defaultValue=2, parameterName=OpacityPower}
  ScalarParameter() {defaultValue=8000, parameterName=DepthFade}
  ScalarParameter() {defaultValue=0.66, parameterName=DayOpacity}
  n34 = Constant() {r=0.5}
  n35 = SphereMask(A=ComponentMask(None=TextureCoordinate() {coordinateIndex=1}) {r=true}, B=@n34, Radius=@n34) {hardnessPercent=0}
  n36 = Constant() {r=0.5}
  n37 = SphereMask(A=ComponentMask(None=TextureCoordinate() {coordinateIndex=1}) {g=true}, B=@n36, Radius=@n36) {hardnessPercent=0}
  n38 = Constant() {r=0.66}
  n39 = MaterialFunctionCall(Value_Morning=@n38, Value_Day=@n38, Value_Evening=@n38, Value_Night=Constant()) {materialFunction=&Scalar_Time_of_Day}
  n40 = Multiply(A=@n39.Result, B=Constant() {r=2.5})
  n41 = Multiply(
   A=Power(Base=Multiply(A=@n35, B=@n37), Exp=Constant() {r=2})
   B=Multiply(A=LinearInterpolate(A=@n40, B=@n39.Result, Alpha=@n18), B=NamedRerouteUsage())
  )
  n42 = Multiply(
   A=DepthFade(Opacity=Multiply(A=@n41, B=@n29), FadeDistance=Constant() {r=8000})
   B=LinearInterpolate(Alpha=Add(A=@n22.G, B=@n23.R)) {constA=0.75}
  )
  StaticSwitchParameter(True=Multiply(A=@n42, B=ParticleColor().A), False=@n42) {parameterName=UseOnParticle}
 ]
}
