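# M_Rectangular_Objective_Decal is a translucent deferred decal material, named for marking rectangular objectives, with a tintable Objective Color that defaults to blue.
# It raises the red channel of a square gradient texture to a power to shape the rectangular mask.
# A DebugTimeSine function, driven by Pulse Speed, blends that exponent between the Contrast parameter and a fixed value, so the gradient's falloff pulses over time.
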Material /Game/Effects/Fort_Effects/Materials/M_Rectangular_Objective_Decal
graph {
 properties {bAutomaticallySetUsageInEditor=false, blendMode=BLEND_Translucent, materialDomain=MD_DeferredDecal, usedShadingModels=Default Lit}
 nodes [
  n1 = ScalarParameter() {defaultValue=0.333, parameterName=Pulse Speed}
  Power(
   Base=TextureSample() {texture=/Game/Effects/Fort_Effects/Effects/Characters/Generic/ResIn/SquareGradient}.R
   Exp=LinearInterpolate(
    A=ScalarParameter() {defaultValue=1.283, parameterName=Contrast}
    Alpha=MaterialFunctionCall(Speed=@n1) {materialFunction=&DebugTimeSine}.Result
   ) {constB=2}
  )
  VectorParameter() {parameterName=Objective Color, defaultValue=<rgba 0 0.45 1 0>}
 ]
}
