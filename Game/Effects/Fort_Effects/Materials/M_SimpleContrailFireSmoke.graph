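# Unlit translucent contrail material for beam trails, showing fire that gives way to smoke along the trail's length.
# Panning smoke and fire textures use second-channel UVs shifted by a dynamic parameter; smoke takes particle color, fire an amplified flame tint.
# A powered UV gradient modulated by fire noise blends the two into emissive.
# Opacity combines smoke, side and tail falloff, a front-edge ramp, camera and depth fades, and particle alpha.
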Material /Game/Effects/Fort_Effects/Materials/M_SimpleContrailFireSmoke
graph {
 properties {
  bUsedWithBeamTrails=true
  bUsedWithNanite=true
  blendMode=BLEND_Translucent
  pixelDepthOffsetMode=PDOM_Legacy
  shadingModel=MSM_Unlit
  usedShadingModels=Unlit
 }
 nodes [
  n1 = Add(
   A=TextureCoordinate() {coordinateIndex=1}
   B=AppendVector(A=DynamicParameter() {paramNames [UV Offset, Param2, Param3, Param4]}.UV Offset, B=Constant())
  )
  n2 = AppendVector(
   A=ScalarParameter() {defaultValue=1, group=Smoke, parameterName=Smoke Tile U}
   B=ScalarParameter() {defaultValue=1, group=Smoke, parameterName=Smoke Tile V}
  )
  n3 = ScalarParameter() {group=Smoke, parameterName=Smoke Pan U}
  n4 = ScalarParameter() {group=Smoke, parameterName=Smoke Pan V}
  n5 = TextureSample(UVs=Panner(Coordinate=Multiply(A=@n1, B=@n2), Speed=AppendVector(A=@n3, B=@n4)) {speedX=-2}) {texture=/Game/Effects/Fort_Effects/Textures/Smoke/T_Lobber_Trail_Mask}
  n6 = Multiply(A=ParticleColor().RGB, B=@n5.G)
  n7 = Add(A=TextureCoordinate() {coordinateIndex=1}, B=DynamicParameter() {paramNames [UV Offset, Param2, Param3, Param4]}.UV Offset)
  n8 = AppendVector(
   A=ScalarParameter() {defaultValue=1, group=Fire, parameterName=Fire Tile U}
   B=ScalarParameter() {defaultValue=1, group=Fire, parameterName=Fire Tile V}
  )
  n9 = ScalarParameter() {group=Fire, parameterName=Fire Pan U}
  n10 = ScalarParameter() {group=Fire, parameterName=Fire Pan V}
  n11 = TextureSample(UVs=Panner(Coordinate=Multiply(A=@n7, B=@n8), Speed=AppendVector(A=@n9, B=@n10))) {texture=/Game/Global/Textures/Default/Tileable_Patterns/T_Smoke_Puff_Tile_Up}
  n12 = Multiply(
   A=@n11.RGB
   B=Multiply(
    A=VectorParameter() {parameterName=Flame, defaultValue=<rgba 0.985 0.4831 0.0764 1>}.RGB
    B=ScalarParameter() {defaultValue=2, parameterName=Amplify}
   )
  )
  n13 = TextureCoordinate() {vTiling=2}
  n14 = Add(A=@n13) {constB=-1}
  n15 = OneMinus(None=ComponentMask(None=Abs(None=@n14)) {g=true})
  n16 = Power(Base=OneMinus(None=ComponentMask(None=@n13) {r=true}), Exp=ScalarParameter() {defaultValue=2, parameterName=Fire Smoke Blend Power})
  n17 = Multiply(A=Power(Base=Multiply(A=@n15, B=@n16), Exp=ScalarParameter() {defaultValue=2, parameterName=Fire Smoke Blend Power}), B=@n11.R)
  n18 = Multiply(A=OneMinus(None=ComponentMask(None=@n13) {r=true}), B=@n15)
  n19 = Multiply(A=@n18, B=@n18)
  MaterialFunctionCall(Emissive=Multiply(A=LinearInterpolate(A=@n6, B=@n12, Alpha=@n17), B=@n19)) {materialFunction=&MF_FX_EmissiveAdjustments}
  n20 = Add(A=ComponentMask(None=TextureCoordinate()) {r=true}) {constB=-0.01}
  n21 = ScalarParameter() {defaultValue=5, parameterName=FrontEdgeMultStr}
  n22 = Multiply(A=@n19, B=Clamp(None=Multiply(A=@n20, B=@n21) {constB=2}))
  n23 = Clamp(None=Multiply(A=Multiply(A=@n22, B=@n5.G), B=ScalarParameter() {defaultValue=4, parameterName=Opacity Adjust}) {constB=4})
  n24 = MaterialFunctionCall(Fade Length=ScalarParameter() {defaultValue=100, parameterName=CameraFadeLength}) {materialFunction=&CameraDepthFade}
  Multiply(A=ParticleColor().A, B=DepthFade(Opacity=Multiply(A=@n23, B=@n24.Result)) {fadeDistanceDefault=5})
 ]
}
