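# Unlit translucent sprite material for Niagara and Cascade particles, drawing glints for a K-pop skydive effect, as its name indicates.
# An anime flare texture's red channel times particle alpha forms a mask that blends an OutSide tint toward an Inside color.
# The blend is multiplied by particle color and fed through MF_FX_EmissiveAdjustments.
# A CameraOffset function with a constant amount provides a camera-relative position offset for the sprites.
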Material /Game/Effects/Fort_Effects/Materials/M_Skydive_Kpop_Glints
graph {
 properties {
  bUsedWithNiagaraSprites=true
  bUsedWithParticleSprites=true
  blendMode=BLEND_Translucent
  pixelDepthOffsetMode=PDOM_Legacy
  shadingModel=MSM_Unlit
  usedShadingModels=Unlit
 }
 nodes [
  MaterialFunctionCall(Offset Amount=Constant() {r=20}) {materialFunction=/Engine/Functions/Engine_MaterialFunctions02/WorldPositionOffset/CameraOffset}
  n1 = ParticleColor()
  n2 = VectorParameter() {parameterName=OutSide, defaultValue=<rgba 0.8698 0.5818 0.6429 1>}
  n3 = VectorParameter() {parameterName=Inside, defaultValue=<rgba 1 1 1 1>}
  n4 = Multiply(A=@n1.A, B=TextureSample() {texture=/Game/Effects/Fort_Effects/Textures/Flares/T_Anime_Flare}.R)
  MaterialFunctionCall(Emissive=Multiply(A=@n1.RGB, B=LinearInterpolate(A=@n2.RGB, B=@n3.RGB, Alpha=@n4))) {materialFunction=&MF_FX_EmissiveAdjustments}
 ]
}
