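# Translucent unlit particle material for sparkle effects on sprite particles, including Niagara.
# Particle color tints a per-particle random blend of two colors (Color A only on mobile), with an optional HDR remap before the shared emissive adjustment function.
# The sparkle texture's red channel times its exponent-shaped green channel is scaled by particle alpha and a near-camera fade to form the transparency mask.
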Material /Game/Effects/Fort_Effects/Materials/M_Sparkle_01
graph {
 properties {
  bUsedWithNanite=true
  bUsedWithNiagaraSprites=true
  bUsedWithParticleSprites=true
  blendMode=BLEND_Translucent
  pixelDepthOffsetMode=PDOM_Legacy
  shadingModel=MSM_Unlit
  usedShadingModels=Unlit
 }
 nodes [
  n1 = ParticleColor()
  n2 = VectorParameter() {parameterName=Color A, defaultValue=<rgba 1 0.5534 0.295 0>}
  n3 = VectorParameter() {parameterName=Color B, defaultValue=<rgba 0.401 0.3782 1 0>}
  n4 = ParticleRandom()
  n5 = Multiply(
   A=@n1.RGB
   B=ShadingPathSwitch(Default=LinearInterpolate(A=@n2.RGB, B=@n3.RGB, Alpha=@n4), Mobile=@n2.RGB) {desc=Migrated from FeatureLevelSwitch}
  )
  n6 = Clamp(None=@n5)
  MaterialFunctionCall(Emissive=StaticSwitchParameter(
    True=MaterialFunctionCall(LDR Input=@n6, HDR Tint=@n1.RGB) {materialFunction=&DeriveHDRfromLDR}.RemappedRange
    False=@n5
   ) {parameterName=Use HDR}) {materialFunction=&MF_FX_EmissiveAdjustments}
  n7 = TextureSample() {texture=/Game/Effects/Fort_Effects/Textures/Sparkle/T_Sparkle_01}
  n8 = Power(Base=@n7.G, Exp=ScalarParameter() {defaultValue=1, parameterName=Exponent})
  Multiply(
   A=Multiply(A=@n1.A, B=Multiply(A=@n7.R, B=@n8))
   B=MaterialFunctionCall(Distance=ScalarParameter() {parameterName=Near Camera Fade Disance}) {materialFunction=&Near_Camera_Fade}.Result
  )
 ]
}
