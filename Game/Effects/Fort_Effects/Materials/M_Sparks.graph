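# M_Sparks is an unlit, two-sided translucent material for spark particles, usable on Niagara and Cascade sprites.
# It samples the T_Spark_SubUV texture and multiplies its RGB by particle color, feeding the result through MF_FX_EmissiveAdjustments.
# Texture alpha times particle alpha is clamped separately, so the particle system drives both tint and fade.
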Material /Game/Effects/Fort_Effects/Materials/M_Sparks
graph {
 properties {
  bUsedWithNiagaraSprites=true
  bUsedWithParticleSprites=true
  blendMode=BLEND_Translucent
  pixelDepthOffsetMode=PDOM_Legacy
  shadingModel=MSM_Unlit
  twoSided=true
  usedShadingModels=Unlit
 }
 nodes [
  n1 = TextureSample() {texture=/Game/Effects/Fort_Effects/Textures/T_Spark_SubUV}
  n2 = ParticleColor()
  Clamp(None=Multiply(A=@n1.A, B=@n2.A))
  MaterialFunctionCall(Emissive=Multiply(A=@n2.RGB, B=@n1.RGB)) {materialFunction=&MF_FX_EmissiveAdjustments}
  VertexColor()
 ]
}
