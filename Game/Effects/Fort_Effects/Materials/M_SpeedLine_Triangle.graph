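# Translucent, unlit, two-sided Niagara sprite material that draws a procedural triangular speed-line shape without textures.
# Two oppositely rotated, offset copies of V-flipped UVs are multiplied and sharpened into a wedge mask, softened by a short gradient along V.
# A squared near-camera fade with a tunable distance and particle alpha scale that mask.
# Particle color RGB feeds the shared FX emissive adjustment function.
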Material /Game/Effects/Fort_Effects/Materials/M_SpeedLine_Triangle
graph {
 properties {
  bUsedWithNiagaraSprites=true
  blendMode=BLEND_Translucent
  pixelDepthOffsetMode=PDOM_Legacy
  shadingModel=MSM_Unlit
  twoSided=true
  usedShadingModels=Unlit
 }
 nodes [
  n1 = ParticleColor()
  MaterialFunctionCall(Emissive=@n1.RGB) {materialFunction=&MF_FX_EmissiveAdjustments}
  n2 = TextureCoordinate() {vTiling=-1}
  n3 = MaterialFunctionCall(UVs=@n2, "Rotation Angle (0-1)"=Constant() {r=-0.17}) {materialFunction=&CustomRotator}
  n4 = Constant2Vector() {g=0.18, r=-0.5}
  n5 = Saturate(None=ComponentMask(None=Add(A=@n3.Rotated Values, B=@n4) {constB=-0.5}) {g=true})
  n6 = Constant() {r=0.17}
  n7 = ComponentMask(None=Add(
    A=@n4
    B=MaterialFunctionCall(UVs=@n2, "Rotation Angle (0-1)"=@n6) {materialFunction=&CustomRotator}.Rotated Values
   ) {constB=-0.5}) {g=true}
  n8 = Saturate(None=Multiply(A=Multiply(A=@n5, B=@n7)) {constB=25})
  n9 = TextureCoordinate()
  n10 = Saturate(None=Multiply(A=ComponentMask(None=@n9) {g=true}) {constB=5})
  n11 = MaterialFunctionCall(Distance=ScalarParameter() {defaultValue=80, parameterName=Near Fade}) {materialFunction=&Near_Camera_Fade}
  Multiply(A=@n1.A, B=Multiply(A=Multiply(A=@n8, B=@n10), B=Multiply(A=@n11.Result, B=@n11.Result)))
 ]
}
