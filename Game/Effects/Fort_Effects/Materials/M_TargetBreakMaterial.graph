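# M_TargetBreakMaterial is a masked, Default Lit material for mesh particles, named for a target-break effect.
# Its small graph holds a Particle Color node and three scalar constants, with no textures or parameters.
# The graph shows no links from these nodes to material outputs, so which surface inputs they drive is not visible.
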Material /Game/Effects/Fort_Effects/Materials/M_TargetBreakMaterial
graph {
 properties {bUsedWithMeshParticles=true, blendMode=BLEND_Masked, pixelDepthOffsetMode=PDOM_Legacy, usedShadingModels=Default Lit}
 nodes [
  Constant() {r=1}
  Constant() {r=1}
  Constant() {r=0.25}
  ParticleColor()
 ]
}
