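# Unlit, two-sided translucent material for mesh particles, apparently the flat top of a tornado effect.
# Two offset, panning samples of a tornado tile are weighted, summed and added to a scrolling sky texture to form moving noise.
# That noise is multiplied by a panning smoke-ring mask, then blended toward the plain mask by a faint vertical ramp and clamped.
# Separately, dimmed, clamped noise feeds MF_FX_EmissiveAdjustments as emissive.
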Material /Game/Effects/Fort_Effects/Materials/M_Tornado_Top_Flat
graph {
 properties {
  bUsedWithMeshParticles=true
  blendMode=BLEND_Translucent
  pixelDepthOffsetMode=PDOM_Legacy
  shadingModel=MSM_Unlit
  twoSided=true
  usedShadingModels=Unlit
 }
 nodes [
  n1 = Add(A=TextureCoordinate() {uTiling=2, vTiling=3}, B=AppendVector(A=Constant() {r=5.5}, B=Constant() {r=6.5}))
  n2 = Multiply(
   A=Constant() {r=0.4}
   B=TextureSample(UVs=Panner(Coordinate=@n1) {speedX=-0.02, speedY=0.12}) {texture=/Game/Global/Textures/Default/Tileable_Patterns/T_Tornado_Tile}.RGB
  )
  n3 = TextureSample(UVs=Panner(Coordinate=TextureCoordinate() {uTiling=2, vTiling=3}) {speedX=-0.02, speedY=0.12}) {texture=&T_Tornado_Tile}
  n4 = TextureCoordinate() {vTiling=3}
  n5 = Add(
   A=Multiply(A=Constant() {r=0.7}, B=Add(A=@n2, B=@n3.RGB))
   B=TextureSample(UVs=Panner(Coordinate=@n4) {speedX=-0.02, speedY=0.05}) {texture=/Game/Packages/Fortress_Sky/Materials/T_FORT_Sky}.RGB
  )
  n6 = TextureCoordinate() {uTiling=2}
  n7 = Multiply(
   A=TextureSample(UVs=Panner(Coordinate=@n6) {speedX=-0.05}) {texture=/Game/Effects/Fort_Effects/Textures/T_Smoke_Ring_Mask2}.RGB
   B=Constant() {r=2}
  )
  n8 = Multiply(A=TextureSample() {texture=/Game/Effects/Fort_Effects/Textures/Ramps/T_Ramp_up}.G, B=Constant() {r=0.2})
  Clamp(None=LinearInterpolate(A=Multiply(A=@n5, B=@n7), B=@n7, Alpha=Clamp(None=@n8)))
  n9 = Constant() {r=0.2}
  MaterialFunctionCall(Emissive=Clamp(None=Multiply(A=@n9, B=@n5))) {materialFunction=&MF_FX_EmissiveAdjustments}
 ]
}
