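# M_Zap is an unlit translucent particle sprite material, named for a zap effect.
# A smooth smoke mask texture is multiplied by particle color RGB and fed through MF_FX_EmissiveAdjustments for emissive.
# The same mask multiplied by particle alpha is clamped between zero and one, apparently for opacity.
# A Vertex Color node and a 0.5 constant appear unconnected.
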Material /Game/Effects/Fort_Effects/Materials/M_Zap
graph {
 properties {
  bUsedWithParticleSprites=true
  blendMode=BLEND_Translucent
  pixelDepthOffsetMode=PDOM_Legacy
  shadingModel=MSM_Unlit
  usedShadingModels=Unlit
 }
 nodes [
  n1 = TextureSample() {texture=/Game/Effects/Fort_Effects/Textures/T_Smooth_Smoke_mask}
  n2 = ParticleColor()
  Clamp(None=Multiply(A=@n1.RGB, B=@n2.A), Min=Constant(), Max=Constant() {r=1})
  Constant() {r=0.5}
  MaterialFunctionCall(Emissive=Multiply(A=@n1.RGB, B=@n2.RGB)) {materialFunction=&MF_FX_EmissiveAdjustments}
  VertexColor()
 ]
}
