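# Debug material for Niagara water effects, judging by its name, flagged for use on Niagara sprite and mesh particles.
# It is a masked, Default Lit surface material.
# The graph holds only a constant red color and a scalar constant, with no textures or parameters, and the file shows no connections from them to material outputs.
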Material /Game/Effects/Fort_Effects/Materials/Niagara_DebugWaterMaterial
graph {
 properties {
  bUsedWithNiagaraMeshParticles=true
  bUsedWithNiagaraSprites=true
  blendMode=BLEND_Masked
  pixelDepthOffsetMode=PDOM_Legacy
  usedShadingModels=Default Lit
 }
 nodes [
  Constant3Vector() {constant=<rgba 1 0 0 0>}
  Constant() {r=0.5}
 ]
}
